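# Animation Blueprint for the HonestWasp medium male cosmetic body, adding procedural bone fixes and secondary motion.
# Constraints and additive bone transforms adjust the CtrlRig pose into PreDyn.
# RigidBody cloth physics, then chained AnimDynamics with spherical thigh and chest limits, simulate jacket, ribbons, chains and necklace.
# bPbBlendEnabled selects between PostDyn and its bicep-jiggle variant, BicepDynamics, before a final zipper simulation drives the output.
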AnimBlueprint /BRCosmetics/Characters/Player/Male/Medium/Bodies/M_MED_HonestWasp/Meshes/M_MED_HonestWasp_AnimBP
target_skeleton=&/M_MED_HonestWasp_Skeleton1
graphs [
 AnimGraph {nodes [
   n1 = Use cached pose 'BicepDynamics'()
   n2 = Use cached pose 'PostDyn'()
   n3 = Get bPbBlendEnabled() -> (bPbBlendEnabled:bool)
   n4 = Anim Dynamics - Bone: spine_zipper_l(
    ComponentPose=Local To Component(LocalPose=Blend Poses by bool(BlendPose_0=@n1, BlendPose_1=@n2, bActiveValue=@n3))
    Alpha=1
    ExternalForce="0,0,0"
   ) {
    angularDampingOverride=1
    angularSpringConstant=500
    bAngularSpring=true
    bOverrideAngularDamping=true
    bUseGravityOverride=true
    linearDampingOverride=1
    linearSpringConstant=50
    simulationSpace=RootRelative
    alphaBoolBlend {blendOption=Linear}
    boundBone {boneName=spine_zipper_l}
    componentAppliedLinearAccClamp=<3000 3000 3000>
    componentLinearAccScale=<0.4 0.4 0.4>
    componentLinearVelScale=<0.4 0.4 0.4>
    gravityOverride {z=-980}
    physicsBodyDefinitions [{
      boundBone {boneName=spine_zipper_l}
      boxExtents=<10 10 10>
      constraintSetup {angularLimitsMax=<0 5 0>, angularLimitsMin=<0 -5 -70>, angularTarget=<1 0 0>}
      localJointOffset=<3 0 0>
     }]
   }
   Output Pose(Result=Component To Local(ComponentPose=@n4)) {layerGroup=None}
   n5 = "Transform (Modify) Bone - Bone: deform_bicep_l"(ComponentPose=Local To Component(LocalPose=Use cached pose 'PostDyn'()), Alpha=1, Translation="-1,0,0", Rotation="0,0,0", Scale="1,1,1") {translationMode=BMM_Additive, translationSpace=BCS_BoneSpace, boneToModify {boneName=deform_bicep_l}, translation {x=-1}}
   n6 = "Transform (Modify) Bone - Bone: deform_bicep_r"(ComponentPose=@n5, Alpha=1, Translation="1,0,0", Rotation="0,0,0", Scale="1,1,1") {translationMode=BMM_Additive, translationSpace=BCS_BoneSpace, boneToModify {boneName=deform_bicep_r}, translation {x=1}}
   n7 = Get Sim Space Settings() -> (Sim Space Settings:Anim Phys Sim Space Settings Structure)
   n8 = Anim Dynamics - Bone: deform_bicep_r!(ComponentPose=@n6, Alpha=1, AngularSpringConstant:float=5000, SimSpaceSettings:Anim Phys Sim Space Settings Structure=@n7) -> (Pose) {
    angularBiasOverride=1
    angularDampingOverride=0.9999
    angularSpringConstant=5000
    bAngularSpring=true
    bGravityOverrideInSimSpace=true
    bOverrideAngularBias=true
    bOverrideAngularDamping=true
    bUseGravityOverride=true
    linearDampingOverride=0.9999
    linearSpringConstant=100
    simulationSpace=RootRelative
    boundBone {boneName=deform_bicep_r}
    componentAppliedLinearAccClamp=<2 2 2>
    gravityOverride {z=-980}
    physicsBodyDefinitions [{
      boundBone {boneName=deform_bicep_r}
      boxExtents=<21 21 21>
      constraintSetup {angularLimitsMax=<5 5 5>, angularLimitsMin=<-5 -5 -5>, angularTarget=<1 0 0>}
      localJointOffset=<2 0 0>
     }]
    relativeSpaceBone {boneName=spine_04}
    simSpaceSettings {maxAngularAcceleration=45, maxAngularVelocity=8, simSpaceAngularAlpha=0.25}
   }
   n9 = Anim Dynamics - Bone: deform_bicep_l!(ComponentPose=@n8, Alpha=1, AngularSpringConstant:float=5000, SimSpaceSettings:Anim Phys Sim Space Settings Structure=@n7) -> (Pose) {
    angularBiasOverride=1
    angularDampingOverride=0.9999
    angularSpringConstant=5000
    bAngularSpring=true
    bGravityOverrideInSimSpace=true
    bOverrideAngularBias=true
    bOverrideAngularDamping=true
    bUseGravityOverride=true
    linearDampingOverride=0.9999
    linearSpringConstant=100
    simulationSpace=RootRelative
    boundBone {boneName=deform_bicep_l}
    componentAppliedLinearAccClamp=<2 2 2>
    gravityOverride {z=-980}
    physicsBodyDefinitions [{
      boundBone {boneName=deform_bicep_l}
      boxExtents=<20 20 20>
      constraintSetup {angularLimitsMax=<5 5 5>, angularLimitsMin=<-5 -5 -5>, angularTarget=<1 0 0>}
      localJointOffset=<-2 0 0>
     }]
    relativeSpaceBone {boneName=spine_04}
    simSpaceSettings {maxAngularAcceleration=45, maxAngularVelocity=8, simSpaceAngularAlpha=0.25}
   }
   n10 = Anim Dynamics - Bone: deform_bicep_r!(ComponentPose=@n9, Alpha=1, GravityOverride:Vector="0,0,-200", LinearSpringConstant:float=180) -> (Pose) {
    angularBiasOverride=1
    angularDampingOverride=0.99
    angularSpringConstant=10000
    bGravityOverrideInSimSpace=true
    bLinearSpring=true
    bOverrideAngularBias=true
    bOverrideLinearDamping=true
    bUseGravityOverride=true
    linearDampingOverride=1
    linearSpringConstant=180
    simulationSpace=RootRelative
    boundBone {boneName=deform_bicep_r}
    componentAppliedLinearAccClamp=<5 5 5>
    gravityOverride {z=-200}
    physicsBodyDefinitions [{
      boundBone {boneName=deform_bicep_r}
      boxExtents=<80 80 80>
      constraintSetup {angularLimitsMax=<0 0 0>, angularLimitsMin=<0 0 0>, angularTarget=<1 0 0>, linearAxesMax=<2 2 2>, linearAxesMin=<-2 -2 -2>}
      localJointOffset=<-0.1 0.1 0>
     }]
    relativeSpaceBone {boneName=spine_04}
    simSpaceSettings {maxAngularAcceleration=50, maxAngularVelocity=10, simSpaceAngularAlpha=0.1}
   }
   n11 = Anim Dynamics - Bone: deform_bicep_l!(ComponentPose=@n10, Alpha=1, GravityOverride:Vector="0,0,-200", LinearSpringConstant:float=180) -> (Pose) {
    angularBiasOverride=1
    angularDampingOverride=0.99
    angularSpringConstant=10000
    bGravityOverrideInSimSpace=true
    bLinearSpring=true
    bOverrideAngularBias=true
    bOverrideLinearDamping=true
    bUseGravityOverride=true
    linearDampingOverride=1
    linearSpringConstant=180
    simulationSpace=RootRelative
    boundBone {boneName=deform_bicep_l}
    componentAppliedLinearAccClamp=<5 5 5>
    gravityOverride {z=-200}
    physicsBodyDefinitions [{
      boundBone {boneName=deform_bicep_l}
      boxExtents=<90 90 90>
      constraintSetup {angularLimitsMax=<0 0 0>, angularLimitsMin=<0 0 0>, angularTarget=<1 0 0>, linearAxesMax=<2 2 2>, linearAxesMin=<-2 -2 -2>}
      localJointOffset=<0.1 -0.1 0>
     }]
    relativeSpaceBone {boneName=spine_04}
    simSpaceSettings {maxAngularAcceleration=50, maxAngularVelocity=10, simSpaceAngularAlpha=0.1}
   }
   Save cached pose 'BicepDynamics'(Pose=Component To Local(ComponentPose=@n11)) {cacheName=BicepDynamics}
   n12 = "Constraint - base_chain_arm (PPPPPPPP)"!(
    ComponentPose=Local To Component(LocalPose=Use cached pose 'CtrlRig'())
    Alpha=1
    ConstraintWeights_0=0.0682
    ConstraintWeights_1:float=0.2749
    ConstraintWeights_2:float=0.0769
    ConstraintWeights_3:float=0.4026
    ConstraintWeights_4:float=0.1503
    ConstraintWeights_5:float=0
    ConstraintWeights_6:float=0.004
    ConstraintWeights_7:float=0.0232
   ) -> (Pose) {
    boneToModify {boneName=base_chain_arm}
    constraintSetup [
     {targetBone {boneName=upperarm_r}}
     {targetBone {boneName=lowerarm_r}}
     {targetBone {boneName=lowerarm_twist_01_r}}
     {targetBone {boneName=lowerarm_twist_02_r}}
     {targetBone {boneName=deform_elbow_out_r}}
     {targetBone {boneName=deform_elbow_in_r}}
     {targetBone {boneName=upperarm_twist_02_r}}
     {targetBone {boneName=dyn_chain_arm_01}}
    ]
    constraintWeights [1,1,1,1,1,1,1,1]
   }
   n13 = "Constraint - spine_zipper_base (PPPPPPP)"!(
    ComponentPose=@n12
    Alpha=1
    ConstraintWeights_0=0
    ConstraintWeights_1:float=0
    ConstraintWeights_2:float=0
    ConstraintWeights_3:float=0
    ConstraintWeights_4:float=0
    ConstraintWeights_5:float=0.0214
    ConstraintWeights_6:float=0.9786
   ) -> (Pose) {
    boneToModify {boneName=spine_zipper_base}
    constraintSetup [
     {targetBone {boneName=pelvis}}
     {targetBone {boneName=thigh_twist_01_l}}
     {targetBone {boneName=deform_groin_l}}
     {targetBone {boneName=thigh_twist_01_r}}
     {targetBone {boneName=deform_groin_r}}
     {targetBone {boneName=dyn_jacket_l_a_02}}
     {targetBone {boneName=dyn_jacket_l_a_03}}
    ]
    constraintWeights [1,1,1,1,1,1,1]
   }
   n14 = "Transform (Modify) Bone - Bone: thorn_arm_01_r"!(ComponentPose=@n13, bAlphaBoolEnabled:bool=True, Translation="0,0,5", Rotation="0,0,0", Scale="0.01,0.01,0.01") -> (Pose) {
    alphaInputType=Bool
    rotationMode=BMM_Additive
    rotationSpace=BCS_BoneSpace
    scaleMode=BMM_Additive
    scaleSpace=BCS_BoneSpace
    translationMode=BMM_Additive
    translationSpace=BCS_BoneSpace
    boneToModify {boneName=thorn_arm_01_r}
    scale=<0.01 0.01 0.01>
    translation {z=5}
   }
   n15 = "Transform (Modify) Bone - Bone: thorn_arm_02_r"!(ComponentPose=@n14, bAlphaBoolEnabled:bool=True, Translation="0,0,-5", Rotation="0,0,0", Scale="0.01,0.01,0.01") -> (Pose) {
    alphaInputType=Bool
    rotationMode=BMM_Additive
    rotationSpace=BCS_BoneSpace
    scaleMode=BMM_Additive
    scaleSpace=BCS_BoneSpace
    translationMode=BMM_Additive
    translationSpace=BCS_BoneSpace
    boneToModify {boneName=thorn_arm_02_r}
    scale=<0.01 0.01 0.01>
    translation {z=-5}
   }
   n16 = "Transform (Modify) Bone - Bone: thorn_arm_03_r"!(ComponentPose=@n15, bAlphaBoolEnabled:bool=True, Translation="0,0,1", Rotation="0,0,0", Scale="0.01,0.01,0.01") -> (Pose) {
    alphaInputType=Bool
    rotationMode=BMM_Additive
    rotationSpace=BCS_BoneSpace
    scaleMode=BMM_Additive
    scaleSpace=BCS_BoneSpace
    translationMode=BMM_Additive
    translationSpace=BCS_BoneSpace
    boneToModify {boneName=thorn_arm_03_r}
    scale=<0.01 0.01 0.01>
    translation {z=1}
   }
   n17 = "Transform (Modify) Bone - Bone: thorn_arm_04_r"!(ComponentPose=@n16, bAlphaBoolEnabled:bool=True, Translation="0,5,0", Rotation="0,0,0", Scale="0.01,0.01,0.01") -> (Pose) {
    alphaInputType=Bool
    rotationMode=BMM_Additive
    rotationSpace=BCS_BoneSpace
    scaleMode=BMM_Additive
    scaleSpace=BCS_BoneSpace
    translationMode=BMM_Additive
    translationSpace=BCS_BoneSpace
    boneToModify {boneName=thorn_arm_04_r}
    scale=<0.01 0.01 0.01>
    translation {y=5}
   }
   n18 = "Transform (Modify) Bone - Bone: deform_elbow_in_base_r"!(ComponentPose=@n17, bAlphaBoolEnabled:bool=True, Translation="0,0,0", Rotation="0,0,0", Scale="0.01,0.01,0.01") -> (Pose) {
    alphaInputType=Bool
    rotationMode=BMM_Additive
    rotationSpace=BCS_BoneSpace
    scaleMode=BMM_Additive
    scaleSpace=BCS_BoneSpace
    translationMode=BMM_Additive
    translationSpace=BCS_BoneSpace
    boneToModify {boneName=deform_elbow_in_base_r}
    scale=<0.01 0.01 0.01>
   }
   n19 = "Transform (Modify) Bone - Bone: deform_elbow_in_base_l"!(ComponentPose=@n18, bAlphaBoolEnabled:bool=True, Translation="0,0,0", Rotation="0,0,0", Scale="0.01,0.01,0.01") -> (Pose) {
    alphaInputType=Bool
    rotationMode=BMM_Additive
    rotationSpace=BCS_BoneSpace
    scaleMode=BMM_Additive
    scaleSpace=BCS_BoneSpace
    translationMode=BMM_Additive
    translationSpace=BCS_BoneSpace
    boneToModify {boneName=deform_elbow_in_base_l}
    scale=<0.01 0.01 0.01>
   }
   n20 = "Transform (Modify) Bone - Bone: base_kneecap_01_l"!(ComponentPose=@n19, bAlphaBoolEnabled:bool=True, Translation="0,0,0", Rotation="0,0,0", Scale="0.01,0.01,0.01") -> (Pose) {
    alphaInputType=Bool
    rotationMode=BMM_Additive
    rotationSpace=BCS_BoneSpace
    scaleMode=BMM_Additive
    scaleSpace=BCS_BoneSpace
    translationMode=BMM_Additive
    translationSpace=BCS_BoneSpace
    boneToModify {boneName=base_kneecap_01_l}
    scale=<0.01 0.01 0.01>
   }
   n21 = "Transform (Modify) Bone - Bone: base_kneecap_01_r"!(ComponentPose=@n20, bAlphaBoolEnabled:bool=True, Translation="0,0,0", Rotation="0,0,0", Scale="0.01,0.01,0.01") -> (Pose) {
    alphaInputType=Bool
    rotationMode=BMM_Additive
    rotationSpace=BCS_BoneSpace
    scaleMode=BMM_Additive
    scaleSpace=BCS_BoneSpace
    translationMode=BMM_Additive
    translationSpace=BCS_BoneSpace
    boneToModify {boneName=base_kneecap_01_r}
    scale=<0.01 0.01 0.01>
   }
   n22 = "Constraint - base_chain_arm (P)"(ComponentPose=@n21, Alpha=1, ConstraintWeights_0=1) {boneToModify {boneName=base_chain_arm}, constraintSetup [{targetBone {boneName=lowerarm_r}}], constraintWeights [1]}
   n23 = "Transform (Modify) Bone - Bone: base_chain_arm"!(ComponentPose=@n22, bAlphaBoolEnabled:bool=True, Translation="-2,0,-5", Rotation="0,-50,0", Scale="0.1,0.1,0.1") -> (Pose) {
    alphaInputType=Bool
    rotationMode=BMM_Additive
    rotationSpace=BCS_BoneSpace
    scaleMode=BMM_Additive
    scaleSpace=BCS_BoneSpace
    translationMode=BMM_Additive
    translationSpace=BCS_BoneSpace
    boneToModify {boneName=base_chain_arm}
    rotation {yaw=-50}
    scale=<0.1 0.1 0.1>
    translation {x=-2, z=-5}
   }
   n24 = "Transform (Modify) Bone - Bone: base_chain_arm"(ComponentPose=@n23, bAlphaBoolEnabled=True, Translation="-2,0,-5", Rotation="0,-50,0", Scale="0.1,0.1,0.1") {
    alphaInputType=Bool
    rotationMode=BMM_Additive
    rotationSpace=BCS_BoneSpace
    scaleMode=BMM_Additive
    scaleSpace=BCS_BoneSpace
    translationMode=BMM_Additive
    translationSpace=BCS_BoneSpace
    boneToModify {boneName=base_chain_arm}
    rotation {yaw=-50}
    scale=<0.1 0.1 0.1>
    translation {x=-2, z=-5}
   }
   n25 = "Transform (Modify) Bone - Bone: base_chain_arm"(ComponentPose=@n24, bAlphaBoolEnabled=True, Translation="-2,0,-5", Rotation="0,-50,0", Scale="0.1,0.1,0.1") {
    alphaInputType=Bool
    rotationMode=BMM_Additive
    rotationSpace=BCS_BoneSpace
    scaleMode=BMM_Additive
    scaleSpace=BCS_BoneSpace
    translationMode=BMM_Additive
    translationSpace=BCS_BoneSpace
    boneToModify {boneName=base_chain_arm}
    rotation {yaw=-50}
    scale=<0.1 0.1 0.1>
    translation {x=-2, z=-5}
   }
   n26 = "Constraint - base_chain_arm (R)"(ComponentPose=@n25, Alpha=1, ConstraintWeights_0=1) {boneToModify {boneName=base_chain_arm}, constraintSetup [{transformType=Rotation, targetBone {boneName=pelvis}}], constraintWeights [1]}
   n27 = "Constraint - base_necklace_01 (R)"(ComponentPose=@n26, Alpha=1, ConstraintWeights_0=1) {boneToModify {boneName=base_necklace_01}, constraintSetup [{transformType=Rotation, targetBone {boneName=neck_01}}], constraintWeights [1]}
   Save cached pose 'PreDyn'(Pose=Component To Local(ComponentPose=@n27)) {cacheName=PreDyn}
   n28 = RigidBody!(ComponentPose=Local To Component(LocalPose=Use cached pose 'PreDyn'()), Alpha, OverrideWorldGravity:Vector="0,0,0", ExternalForce) -> (Pose) {
    bClampLinearTranslationLimitToRefPose=true
    bOverrideWorldGravity=true
    overridePhysicsAsset=&/M_MED_HonestWasp_cloth_Physics
    simulationSpace=BaseBoneSpace
    baseBoneRef {boneName=spine_03}
    overrideWorldGravity {z=-980}
    simSpaceSettings {maxAngularAcceleration=100, maxAngularVelocity=5, maxLinearAcceleration=300, maxLinearVelocity=100, velocityScaleZ=0, worldAlpha=1}
   }
   n29 = "Constraint - dyn_jacket_l_d_01 (P)"!(ComponentPose=@n28, bAlphaBoolEnabled:bool=True, ConstraintWeights_0=1) -> (Pose) {alphaInputType=Bool, boneToModify {boneName=dyn_jacket_l_d_01}, constraintSetup [{targetBone {boneName=spine_02}}], constraintWeights [1]}
   n30 = "Constraint - dyn_jacket_b_d_01 (P)"!(ComponentPose=@n29, bAlphaBoolEnabled:bool=True, ConstraintWeights_0=1) -> (Pose) {alphaInputType=Bool, boneToModify {boneName=dyn_jacket_b_d_01}, constraintSetup [{targetBone {boneName=spine_02}}], constraintWeights [1]}
   n31 = "Constraint - dyn_jacket_r_d_01 (P)"!(ComponentPose=@n30, bAlphaBoolEnabled:bool=True, ConstraintWeights_0=1) -> (Pose) {alphaInputType=Bool, boneToModify {boneName=dyn_jacket_r_d_01}, constraintSetup [{targetBone {boneName=spine_02}}], constraintWeights [1]}
   n32 = Anim Dynamics - Bone: dyn_thighribbon_A_01(ComponentPose=@n31, Alpha=1, ExternalForce="0,0,0") {
    angularDampingOverride=1
    angularSpringConstant=200
    bAngularSpring=true
    bOverrideAngularDamping=true
    bOverrideLinearDamping=true
    bUseGravityOverride=true
    bUseSphericalLimits=true
    linearDampingOverride=1
    linearSpringConstant=50
    simulationSpace=RootRelative
    alphaBoolBlend {blendOption=Linear}
    boundBone {boneName=dyn_thighribbon_A_01}
    componentAppliedLinearAccClamp=<3000 3000 3000>
    componentLinearAccScale=<0.4 0.4 0.4>
    componentLinearVelScale=<0.4 0.4 0.4>
    gravityOverride {z=-980}
    physicsBodyDefinitions [{
      boundBone {boneName=dyn_thighribbon_A_01}
      boxExtents=<10 10 10>
      constraintSetup {angularLimitsMax=<0 0 0>, angularLimitsMin=<0 -20 -10>, angularTarget=<1 0 0>}
      localJointOffset=<5 0 0>
     }]
    sphericalLimits [
     {limitRadius=10, drivingBone {boneName=thigh_r}, sphereLocalOffset=<20 0 0>}
     {limitRadius=10, drivingBone {boneName=thigh_r}, sphereLocalOffset=<35 0 0>}
    ]
   }
   n33 = Anim Dynamics - Bone: dyn_thighribbon_A_02(ComponentPose=@n32, Alpha=1, ExternalForce="0,0,0") {
    angularDampingOverride=1
    angularSpringConstant=200
    bAngularSpring=true
    bOverrideAngularDamping=true
    bOverrideLinearDamping=true
    bUseGravityOverride=true
    bUseSphericalLimits=true
    linearDampingOverride=1
    linearSpringConstant=50
    simulationSpace=RootRelative
    alphaBoolBlend {blendOption=Linear}
    boundBone {boneName=dyn_thighribbon_A_02}
    componentAppliedLinearAccClamp=<3000 3000 3000>
    componentLinearAccScale=<0.4 0.4 0.4>
    componentLinearVelScale=<0.4 0.4 0.4>
    gravityOverride {z=-980}
    physicsBodyDefinitions [{
      boundBone {boneName=dyn_thighribbon_A_02}
      boxExtents=<10 10 10>
      constraintSetup {angularLimitsMax=<0 0 10>, angularLimitsMin=<0 -20 -10>, angularTarget=<1 0 0>}
      localJointOffset=<5 0 0>
     }]
    sphericalLimits [
     {limitRadius=10, drivingBone {boneName=thigh_r}, sphereLocalOffset=<20 0 0>}
     {limitRadius=10, drivingBone {boneName=thigh_r}, sphereLocalOffset=<35 0 0>}
    ]
   }
   n34 = Anim Dynamics - Bone: dyn_thighribbon_A_03(ComponentPose=@n33, Alpha=1, ExternalForce="0,0,0") {
    angularDampingOverride=1
    angularSpringConstant=200
    bAngularSpring=true
    bOverrideAngularDamping=true
    bOverrideLinearDamping=true
    bUseGravityOverride=true
    bUseSphericalLimits=true
    linearDampingOverride=1
    linearSpringConstant=50
    simulationSpace=RootRelative
    alphaBoolBlend {blendOption=Linear}
    boundBone {boneName=dyn_thighribbon_A_03}
    componentAppliedLinearAccClamp=<3000 3000 3000>
    componentLinearAccScale=<0.4 0.4 0.4>
    componentLinearVelScale=<0.4 0.4 0.4>
    gravityOverride {z=-980}
    physicsBodyDefinitions [{
      boundBone {boneName=dyn_thighribbon_A_03}
      boxExtents=<10 10 10>
      constraintSetup {angularLimitsMax=<0 0 10>, angularLimitsMin=<0 -20 -10>, angularTarget=<1 0 0>}
      localJointOffset=<5 0 0>
     }]
    sphericalLimits [
     {limitRadius=10, drivingBone {boneName=thigh_r}, sphereLocalOffset=<20 0 0>}
     {limitRadius=10, drivingBone {boneName=thigh_r}, sphereLocalOffset=<35 0 0>}
    ]
   }
   n35 = Anim Dynamics - Bone: dyn_thighribbon_B_01(ComponentPose=@n34, Alpha=1, ExternalForce="0,0,0") {
    angularDampingOverride=1
    angularSpringConstant=200
    bAngularSpring=true
    bOverrideAngularDamping=true
    bOverrideLinearDamping=true
    bUseGravityOverride=true
    bUseSphericalLimits=true
    linearDampingOverride=1
    linearSpringConstant=50
    simulationSpace=RootRelative
    alphaBoolBlend {blendOption=Linear}
    boundBone {boneName=dyn_thighribbon_B_01}
    componentAppliedLinearAccClamp=<3000 3000 3000>
    componentLinearAccScale=<0.4 0.4 0.4>
    componentLinearVelScale=<0.4 0.4 0.4>
    gravityOverride {z=-980}
    physicsBodyDefinitions [{
      boundBone {boneName=dyn_thighribbon_B_01}
      boxExtents=<10 10 10>
      constraintSetup {angularLimitsMax=<0 0 10>, angularLimitsMin=<0 -20 0>, angularTarget=<1 0 0>}
      localJointOffset=<5 0 0>
     }]
    sphericalLimits [
     {limitRadius=10, drivingBone {boneName=thigh_l}, sphereLocalOffset=<-20 0 0>}
     {limitRadius=10, drivingBone {boneName=thigh_l}, sphereLocalOffset=<-35 0 0>}
    ]
   }
   n36 = Anim Dynamics - Bone: dyn_thighribbon_B_02(ComponentPose=@n35, Alpha=1, ExternalForce="0,0,0") {
    angularDampingOverride=1
    angularSpringConstant=200
    bAngularSpring=true
    bOverrideAngularDamping=true
    bOverrideLinearDamping=true
    bUseGravityOverride=true
    bUseSphericalLimits=true
    linearDampingOverride=1
    linearSpringConstant=50
    simulationSpace=RootRelative
    alphaBoolBlend {blendOption=Linear}
    boundBone {boneName=dyn_thighribbon_B_02}
    componentAppliedLinearAccClamp=<3000 3000 3000>
    componentLinearAccScale=<0.4 0.4 0.4>
    componentLinearVelScale=<0.4 0.4 0.4>
    gravityOverride {z=-980}
    physicsBodyDefinitions [{
      boundBone {boneName=dyn_thighribbon_B_02}
      boxExtents=<10 10 10>
      constraintSetup {angularLimitsMax=<0 0 10>, angularLimitsMin=<0 -20 0>, angularTarget=<1 0 0>}
      localJointOffset=<5 0 0>
     }]
    sphericalLimits [
     {limitRadius=10, drivingBone {boneName=thigh_l}, sphereLocalOffset=<-20 0 0>}
     {limitRadius=10, drivingBone {boneName=thigh_l}, sphereLocalOffset=<-35 0 0>}
    ]
   }
   n37 = Anim Dynamics - Bone: dyn_chain_A_01(ComponentPose=@n36, Alpha=1, ExternalForce="0,0,0") {
    angularDampingOverride=1
    angularSpringConstant=200
    bAngularSpring=true
    bOverrideAngularDamping=true
    bOverrideLinearDamping=true
    bUseGravityOverride=true
    bUseSphericalLimits=true
    linearDampingOverride=1
    linearSpringConstant=50
    simulationSpace=RootRelative
    alphaBoolBlend {blendOption=Linear}
    boundBone {boneName=dyn_chain_A_01}
    componentAppliedLinearAccClamp=<3000 3000 3000>
    componentLinearAccScale=<0.4 0.4 0.4>
    componentLinearVelScale=<0.4 0.4 0.4>
    gravityOverride {z=-980}
    physicsBodyDefinitions [{
      boundBone {boneName=dyn_chain_A_01}
      boxExtents=<10 10 10>
      constraintSetup {angularLimitsMax=<10 0 10>, angularLimitsMin=<-10 -10 -10>, angularTarget=<1 0 0>}
      localJointOffset=<5 0 0>
     }]
    sphericalLimits [
     {limitRadius=10, drivingBone {boneName=thigh_r}, sphereLocalOffset=<15 0 0>}
     {limitRadius=10, drivingBone {boneName=thigh_r}, sphereLocalOffset=<25 0 0>}
    ]
   }
   n38 = Anim Dynamics - Bone: dyn_chain_A_02(ComponentPose=@n37, Alpha=1, ExternalForce="0,0,0") {
    angularDampingOverride=1
    angularSpringConstant=200
    bAngularSpring=true
    bOverrideAngularDamping=true
    bOverrideLinearDamping=true
    bUseGravityOverride=true
    bUseSphericalLimits=true
    linearDampingOverride=1
    linearSpringConstant=50
    simulationSpace=RootRelative
    alphaBoolBlend {blendOption=Linear}
    boundBone {boneName=dyn_chain_A_02}
    componentAppliedLinearAccClamp=<3000 3000 3000>
    componentLinearAccScale=<0.4 0.4 0.4>
    componentLinearVelScale=<0.4 0.4 0.4>
    gravityOverride {z=-980}
    physicsBodyDefinitions [{
      boundBone {boneName=dyn_chain_A_02}
      boxExtents=<10 10 10>
      constraintSetup {angularLimitsMax=<10 5 10>, angularLimitsMin=<-10 -10 -10>, angularTarget=<1 0 0>}
      localJointOffset=<5 0 0>
     }]
    sphericalLimits [
     {limitRadius=10, drivingBone {boneName=thigh_r}, sphereLocalOffset=<15 0 0>}
     {limitRadius=10, drivingBone {boneName=thigh_r}, sphereLocalOffset=<25 0 0>}
    ]
   }
   n39 = Anim Dynamics - Bone: dyn_chain_A_03(ComponentPose=@n38, Alpha=1, ExternalForce="0,0,0") {
    angularDampingOverride=1
    angularSpringConstant=200
    bAngularSpring=true
    bOverrideAngularDamping=true
    bOverrideLinearDamping=true
    bUseGravityOverride=true
    bUseSphericalLimits=true
    linearDampingOverride=1
    linearSpringConstant=50
    simulationSpace=RootRelative
    alphaBoolBlend {blendOption=Linear}
    boundBone {boneName=dyn_chain_A_03}
    componentAppliedLinearAccClamp=<3000 3000 3000>
    componentLinearAccScale=<0.4 0.4 0.4>
    componentLinearVelScale=<0.4 0.4 0.4>
    gravityOverride {z=-980}
    physicsBodyDefinitions [{
      boundBone {boneName=dyn_chain_A_03}
      boxExtents=<10 10 10>
      constraintSetup {angularLimitsMax=<10 5 10>, angularLimitsMin=<-10 -10 -10>, angularTarget=<1 0 0>}
      localJointOffset=<5 0 0>
     }]
    sphericalLimits [
     {limitRadius=10, drivingBone {boneName=thigh_r}, sphereLocalOffset=<15 0 0>}
     {limitRadius=10, drivingBone {boneName=thigh_r}, sphereLocalOffset=<25 0 0>}
    ]
   }
   n40 = Anim Dynamics - Bone: dyn_chain_A_05(ComponentPose=@n39, Alpha=1, ExternalForce="0,0,0") {
    angularDampingOverride=1
    angularSpringConstant=200
    bAngularSpring=true
    bOverrideAngularDamping=true
    bOverrideLinearDamping=true
    bUseGravityOverride=true
    bUseSphericalLimits=true
    linearDampingOverride=1
    linearSpringConstant=50
    simulationSpace=RootRelative
    alphaBoolBlend {blendOption=Linear}
    boundBone {boneName=dyn_chain_A_05}
    componentAppliedLinearAccClamp=<3000 3000 3000>
    componentLinearAccScale=<0.4 0.4 0.4>
    componentLinearVelScale=<0.4 0.4 0.4>
    gravityOverride {z=-980}
    physicsBodyDefinitions [{
      boundBone {boneName=dyn_chain_A_05}
      boxExtents=<10 10 10>
      constraintSetup {angularLimitsMax=<10 5 10>, angularLimitsMin=<-10 -10 -10>, angularTarget=<1 0 0>}
      localJointOffset=<5 0 0>
     }]
    sphericalLimits [
     {limitRadius=10, drivingBone {boneName=thigh_r}, sphereLocalOffset=<15 0 0>}
     {limitRadius=10, drivingBone {boneName=thigh_r}, sphereLocalOffset=<25 0 0>}
    ]
   }
   n41 = Anim Dynamics - Bone: dyn_chain_A_06(ComponentPose=@n40, Alpha=1, ExternalForce="0,0,0") {
    angularDampingOverride=1
    angularSpringConstant=200
    bAngularSpring=true
    bOverrideAngularDamping=true
    bOverrideLinearDamping=true
    bUseGravityOverride=true
    bUseSphericalLimits=true
    linearDampingOverride=1
    linearSpringConstant=50
    simulationSpace=RootRelative
    alphaBoolBlend {blendOption=Linear}
    boundBone {boneName=dyn_chain_A_06}
    componentAppliedLinearAccClamp=<3000 3000 3000>
    componentLinearAccScale=<0.4 0.4 0.4>
    componentLinearVelScale=<0.4 0.4 0.4>
    gravityOverride {z=-980}
    physicsBodyDefinitions [{
      boundBone {boneName=dyn_chain_A_06}
      boxExtents=<10 10 10>
      constraintSetup {angularLimitsMax=<10 5 10>, angularLimitsMin=<-10 -10 -10>, angularTarget=<1 0 0>}
      localJointOffset=<5 0 0>
     }]
    sphericalLimits [
     {limitRadius=10, drivingBone {boneName=thigh_r}, sphereLocalOffset=<15 0 0>}
     {limitRadius=10, drivingBone {boneName=thigh_r}, sphereLocalOffset=<25 0 0>}
    ]
   }
   n42 = Anim Dynamics - Bone: dyn_chain_B_01(ComponentPose=@n41, Alpha=1, ExternalForce="0,0,0") {
    angularDampingOverride=1
    angularSpringConstant=200
    bAngularSpring=true
    bOverrideAngularDamping=true
    bOverrideLinearDamping=true
    bUseGravityOverride=true
    bUseSphericalLimits=true
    linearDampingOverride=1
    linearSpringConstant=50
    simulationSpace=RootRelative
    alphaBoolBlend {blendOption=Linear}
    boundBone {boneName=dyn_chain_B_01}
    componentAppliedLinearAccClamp=<3000 3000 3000>
    componentLinearAccScale=<0.4 0.4 0.4>
    componentLinearVelScale=<0.4 0.4 0.4>
    gravityOverride {z=-980}
    physicsBodyDefinitions [{
      boundBone {boneName=dyn_chain_B_01}
      boxExtents=<10 10 10>
      constraintSetup {angularLimitsMax=<10 0 10>, angularLimitsMin=<-10 -10 -10>, angularTarget=<1 0 0>}
      localJointOffset=<5 0 0>
     }]
    sphericalLimits [
     {limitRadius=10, drivingBone {boneName=thigh_l}, sphereLocalOffset=<-15 0 0>}
     {limitRadius=10, drivingBone {boneName=thigh_l}, sphereLocalOffset=<-25 0 0>}
    ]
   }
   n43 = Anim Dynamics - Bone: dyn_chain_B_02(ComponentPose=@n42, Alpha=1, ExternalForce="0,0,0") {
    angularDampingOverride=1
    angularSpringConstant=200
    bAngularSpring=true
    bOverrideAngularDamping=true
    bOverrideLinearDamping=true
    bUseGravityOverride=true
    bUseSphericalLimits=true
    linearDampingOverride=1
    linearSpringConstant=50
    simulationSpace=RootRelative
    alphaBoolBlend {blendOption=Linear}
    boundBone {boneName=dyn_chain_B_02}
    componentAppliedLinearAccClamp=<3000 3000 3000>
    componentLinearAccScale=<0.4 0.4 0.4>
    componentLinearVelScale=<0.4 0.4 0.4>
    gravityOverride {z=-980}
    physicsBodyDefinitions [{
      boundBone {boneName=dyn_chain_B_02}
      boxExtents=<10 10 10>
      constraintSetup {angularLimitsMax=<0 5 10>, angularLimitsMin=<0 -10 -10>, angularTarget=<1 0 0>}
      localJointOffset=<5 0 0>
     }]
    sphericalLimits [
     {limitRadius=10, drivingBone {boneName=thigh_l}, sphereLocalOffset=<-15 0 0>}
     {limitRadius=10, drivingBone {boneName=thigh_l}, sphereLocalOffset=<-25 0 0>}
    ]
   }
   n44 = Anim Dynamics - Bone: dyn_chain_B_03(ComponentPose=@n43, Alpha=1, ExternalForce="0,0,0") {
    angularDampingOverride=1
    angularSpringConstant=200
    bAngularSpring=true
    bOverrideAngularDamping=true
    bOverrideLinearDamping=true
    bUseGravityOverride=true
    bUseSphericalLimits=true
    linearDampingOverride=1
    linearSpringConstant=50
    simulationSpace=RootRelative
    alphaBoolBlend {blendOption=Linear}
    boundBone {boneName=dyn_chain_B_03}
    componentAppliedLinearAccClamp=<3000 3000 3000>
    componentLinearAccScale=<0.4 0.4 0.4>
    componentLinearVelScale=<0.4 0.4 0.4>
    gravityOverride {z=-980}
    physicsBodyDefinitions [{
      boundBone {boneName=dyn_chain_B_03}
      boxExtents=<10 10 10>
      constraintSetup {angularLimitsMax=<0 5 10>, angularLimitsMin=<0 -10 -10>, angularTarget=<1 0 0>}
      localJointOffset=<5 0 0>
     }]
    sphericalLimits [
     {limitRadius=10, drivingBone {boneName=thigh_l}, sphereLocalOffset=<-15 0 0>}
     {limitRadius=10, drivingBone {boneName=thigh_l}, sphereLocalOffset=<-25 0 0>}
    ]
   }
   n45 = Anim Dynamics - Bone: dyn_chain_arm_01(ComponentPose=@n44, Alpha=1, ExternalForce="0,0,0") {
    angularDampingOverride=1
    angularSpringConstant=1500
    bGravityOverrideInSimSpace=true
    bOverrideAngularDamping=true
    bUseGravityOverride=true
    bUseSphericalLimits=true
    linearDampingOverride=0.9999
    alphaBoolBlend {blendOption=Linear}
    boundBone {boneName=dyn_chain_arm_01}
    componentAppliedLinearAccClamp=<1100 1100 1100>
    componentLinearAccScale=<0.5 0.5 0.5>
    componentLinearVelScale=<0.5 0.5 0.5>
    gravityOverride {z=-980}
    physicsBodyDefinitions [{
      collisionType=CustomSphere
      sphereCollisionRadius=2
      boundBone {boneName=dyn_chain_arm_01}
      boxExtents=<8 15 15>
      constraintSetup {angularLimitsMax=<5 60 60>, angularLimitsMin=<-5 -60 -60>, angularTarget=<0 0 1>}
      localJointOffset=<4 0 0>
     }]
    sphericalLimits [
     {limitRadius=10, drivingBone {boneName=thigh_r}, sphereLocalOffset=<20 0 0>}
     {limitRadius=10, drivingBone {boneName=thigh_r}, sphereLocalOffset=<35 0 0>}
    ]
   }
   n46 = Anim Dynamics - Bone: dyn_chain_arm_02(ComponentPose=@n45, Alpha=1, ExternalForce="0,0,0") {
    angularDampingOverride=1
    angularSpringConstant=1500
    bGravityOverrideInSimSpace=true
    bOverrideAngularDamping=true
    bUseGravityOverride=true
    bUseSphericalLimits=true
    linearDampingOverride=0.9999
    alphaBoolBlend {blendOption=Linear}
    boundBone {boneName=dyn_chain_arm_02}
    componentAppliedLinearAccClamp=<1100 1100 1100>
    componentLinearAccScale=<0.5 0.5 0.5>
    componentLinearVelScale=<0.5 0.5 0.5>
    gravityOverride {z=-980}
    physicsBodyDefinitions [{
      collisionType=CustomSphere
      sphereCollisionRadius=2
      boundBone {boneName=dyn_chain_arm_02}
      boxExtents=<8 15 15>
      constraintSetup {angularLimitsMax=<-5 60 30>, angularLimitsMin=<-5 -60 -30>, angularTarget=<0 0 1>}
      localJointOffset=<4 0 0>
     }]
    sphericalLimits [
     {limitRadius=10, drivingBone {boneName=thigh_r}, sphereLocalOffset=<20 0 0>}
     {limitRadius=10, drivingBone {boneName=thigh_r}, sphereLocalOffset=<35 0 0>}
    ]
   }
   n47 = Anim Dynamics - Bone: dyn_chain_arm_03(ComponentPose=@n46, Alpha=1, ExternalForce="0,0,0") {
    angularDampingOverride=1
    angularSpringConstant=1500
    bGravityOverrideInSimSpace=true
    bOverrideAngularDamping=true
    bUseGravityOverride=true
    bUseSphericalLimits=true
    linearDampingOverride=0.9999
    alphaBoolBlend {blendOption=Linear}
    boundBone {boneName=dyn_chain_arm_03}
    componentAppliedLinearAccClamp=<1100 1100 1100>
    componentLinearAccScale=<0.5 0.5 0.5>
    componentLinearVelScale=<0.5 0.5 0.5>
    gravityOverride {z=-980}
    physicsBodyDefinitions [{
      collisionType=CustomSphere
      sphereCollisionRadius=2
      boundBone {boneName=dyn_chain_arm_03}
      boxExtents=<8 15 15>
      constraintSetup {angularLimitsMax=<-5 60 30>, angularLimitsMin=<-5 -60 -30>, angularTarget=<0 0 1>}
      localJointOffset=<4 0 0>
     }]
    sphericalLimits [
     {limitRadius=10, drivingBone {boneName=thigh_r}, sphereLocalOffset=<20 0 0>}
     {limitRadius=10, drivingBone {boneName=thigh_r}, sphereLocalOffset=<35 0 0>}
    ]
   }
   n48 = Anim Dynamics - Bone: dyn_chain_arm_04(ComponentPose=@n47, Alpha=1, ExternalForce="0,0,0") {
    angularDampingOverride=1
    angularSpringConstant=1500
    bGravityOverrideInSimSpace=true
    bOverrideAngularDamping=true
    bUseGravityOverride=true
    bUseSphericalLimits=true
    linearDampingOverride=0.9999
    alphaBoolBlend {blendOption=Linear}
    boundBone {boneName=dyn_chain_arm_04}
    componentAppliedLinearAccClamp=<1100 1100 1100>
    componentLinearAccScale=<0.5 0.5 0.5>
    componentLinearVelScale=<0.5 0.5 0.5>
    gravityOverride {z=-980}
    physicsBodyDefinitions [{
      collisionType=CustomSphere
      sphereCollisionRadius=2
      boundBone {boneName=dyn_chain_arm_04}
      boxExtents=<8 15 15>
      constraintSetup {angularLimitsMax=<-5 60 30>, angularLimitsMin=<-5 -60 -30>, angularTarget=<0 0 1>}
      localJointOffset=<4 0 0>
     }]
    sphericalLimits [
     {limitRadius=10, drivingBone {boneName=thigh_r}, sphereLocalOffset=<20 0 0>}
     {limitRadius=10, drivingBone {boneName=thigh_r}, sphereLocalOffset=<35 0 0>}
    ]
   }
   n49 = Anim Dynamics - Bone: dyn_chain_arm_05(ComponentPose=@n48, Alpha=1, ExternalForce="0,0,0") {
    angularDampingOverride=1
    angularSpringConstant=1500
    bGravityOverrideInSimSpace=true
    bOverrideAngularDamping=true
    bUseGravityOverride=true
    bUseSphericalLimits=true
    linearDampingOverride=0.9999
    alphaBoolBlend {blendOption=Linear}
    boundBone {boneName=dyn_chain_arm_05}
    componentAppliedLinearAccClamp=<1100 1100 1100>
    componentLinearAccScale=<0.5 0.5 0.5>
    componentLinearVelScale=<0.5 0.5 0.5>
    gravityOverride {z=-980}
    physicsBodyDefinitions [{
      collisionType=CustomSphere
      sphereCollisionRadius=2
      boundBone {boneName=dyn_chain_arm_05}
      boxExtents=<8 15 15>
      constraintSetup {angularLimitsMax=<-5 60 30>, angularLimitsMin=<-5 -60 -30>, angularTarget=<0 0 1>}
      localJointOffset=<4 0 0>
     }]
    sphericalLimits [
     {limitRadius=10, drivingBone {boneName=thigh_r}, sphereLocalOffset=<20 0 0>}
     {limitRadius=10, drivingBone {boneName=thigh_r}, sphereLocalOffset=<35 0 0>}
    ]
   }
   n50 = Anim Dynamics - Bone: dyn_chain_arm_06(ComponentPose=@n49, Alpha=1, ExternalForce="0,0,0") {
    angularDampingOverride=1
    angularSpringConstant=1500
    bGravityOverrideInSimSpace=true
    bOverrideAngularDamping=true
    bUseGravityOverride=true
    bUseSphericalLimits=true
    linearDampingOverride=0.9999
    alphaBoolBlend {blendOption=Linear}
    boundBone {boneName=dyn_chain_arm_06}
    componentAppliedLinearAccClamp=<1100 1100 1100>
    componentLinearAccScale=<0.5 0.5 0.5>
    componentLinearVelScale=<0.5 0.5 0.5>
    gravityOverride {z=-980}
    physicsBodyDefinitions [{
      collisionType=CustomSphere
      sphereCollisionRadius=2
      boundBone {boneName=dyn_chain_arm_06}
      boxExtents=<8 15 15>
      constraintSetup {angularLimitsMax=<-5 60 30>, angularLimitsMin=<-5 -60 -30>, angularTarget=<0 0 1>}
      localJointOffset=<4 0 0>
     }]
    sphericalLimits [
     {limitRadius=10, drivingBone {boneName=thigh_r}, sphereLocalOffset=<20 0 0>}
     {limitRadius=10, drivingBone {boneName=thigh_r}, sphereLocalOffset=<35 0 0>}
    ]
   }
   n51 = Anim Dynamics - Bone: dyn_chain_arm_07(ComponentPose=@n50, Alpha=1, ExternalForce="0,0,0") {
    angularDampingOverride=1
    angularSpringConstant=1500
    bGravityOverrideInSimSpace=true
    bOverrideAngularDamping=true
    bUseGravityOverride=true
    bUseSphericalLimits=true
    linearDampingOverride=0.9999
    alphaBoolBlend {blendOption=Linear}
    boundBone {boneName=dyn_chain_arm_07}
    componentAppliedLinearAccClamp=<1100 1100 1100>
    componentLinearAccScale=<0.5 0.5 0.5>
    componentLinearVelScale=<0.5 0.5 0.5>
    gravityOverride {z=-980}
    physicsBodyDefinitions [{
      collisionType=CustomSphere
      sphereCollisionRadius=2
      boundBone {boneName=dyn_chain_arm_07}
      boxExtents=<8 15 15>
      constraintSetup {angularLimitsMax=<-5 60 30>, angularLimitsMin=<-5 -60 -30>, angularTarget=<0 0 1>}
      localJointOffset=<4 0 0>
     }]
    sphericalLimits [
     {limitRadius=10, drivingBone {boneName=thigh_r}, sphereLocalOffset=<20 0 0>}
     {limitRadius=10, drivingBone {boneName=thigh_r}, sphereLocalOffset=<35 0 0>}
    ]
   }
   n52 = Anim Dynamics - Bone: dyn_necklace_01(ComponentPose=@n51, Alpha=1, ExternalForce="0,0,0") {
    angularDampingOverride=1
    angularSpringConstant=200
    bAngularSpring=true
    bOverrideAngularDamping=true
    bOverrideLinearDamping=true
    bUseGravityOverride=true
    bUseSphericalLimits=true
    linearDampingOverride=1
    linearSpringConstant=50
    simulationSpace=RootRelative
    alphaBoolBlend {blendOption=Linear}
    boundBone {boneName=dyn_necklace_01}
    componentAppliedLinearAccClamp=<3000 3000 3000>
    componentLinearAccScale=<0.4 0.4 0.4>
    componentLinearVelScale=<0.4 0.4 0.4>
    gravityOverride {z=-980}
    physicsBodyDefinitions [{
      boundBone {boneName=dyn_necklace_01}
      boxExtents=<10 10 10>
      constraintSetup {angularLimitsMax=<0 10 2>, angularLimitsMin=<0 -20 -2>, angularTarget=<1 0 0>}
      localJointOffset=<5 0 0>
     }]
    sphericalLimits [{limitRadius=34, drivingBone {boneName=spine_05}, sphereLocalOffset=<-20 -13 0>}]
   }
   n53 = Anim Dynamics - Bone: dyn_necklaceA_02(ComponentPose=@n52, Alpha=1, ExternalForce="0,0,0") {
    angularDampingOverride=1
    angularSpringConstant=200
    bAngularSpring=true
    bOverrideAngularDamping=true
    bOverrideLinearDamping=true
    bUseGravityOverride=true
    bUseSphericalLimits=true
    linearDampingOverride=1
    linearSpringConstant=50
    simulationSpace=RootRelative
    alphaBoolBlend {blendOption=Linear}
    boundBone {boneName=dyn_necklaceA_02}
    componentAppliedLinearAccClamp=<3000 3000 3000>
    componentLinearAccScale=<0.4 0.4 0.4>
    componentLinearVelScale=<0.4 0.4 0.4>
    gravityOverride {z=-980}
    physicsBodyDefinitions [{
      boundBone {boneName=dyn_necklaceA_02}
      boxExtents=<10 10 10>
      constraintSetup {angularLimitsMax=<0 10 2>, angularLimitsMin=<0 -10 -2>, angularTarget=<1 0 0>}
      localJointOffset=<5 0 0>
     }]
    sphericalLimits [{limitRadius=32, drivingBone {boneName=spine_05}, sphereLocalOffset=<-20 -13 0>}]
   }
   n54 = Anim Dynamics - Bone: dyn_necklaceA_03(ComponentPose=@n53, Alpha=1, ExternalForce="0,0,0") {
    angularDampingOverride=1
    angularSpringConstant=200
    bAngularSpring=true
    bOverrideAngularDamping=true
    bOverrideLinearDamping=true
    bUseGravityOverride=true
    linearDampingOverride=1
    linearSpringConstant=50
    simulationSpace=RootRelative
    alphaBoolBlend {blendOption=Linear}
    boundBone {boneName=dyn_necklaceA_03}
    componentAppliedLinearAccClamp=<3000 3000 3000>
    componentLinearAccScale=<0.4 0.4 0.4>
    componentLinearVelScale=<0.4 0.4 0.4>
    gravityOverride {z=-980}
    physicsBodyDefinitions [{
      boundBone {boneName=dyn_necklaceA_03}
      boxExtents=<10 10 10>
      constraintSetup {angularLimitsMax=<0 20 10>, angularLimitsMin=<0 -10 -10>, angularTarget=<1 0 0>}
      localJointOffset=<5 0 0>
     }]
    sphericalLimits [{limitRadius=32, drivingBone {boneName=spine_05}, sphereLocalOffset=<-20 -13 0>}]
   }
   n55 = Anim Dynamics - Bone: dyn_necklaceB_02(ComponentPose=@n54, Alpha=1, ExternalForce="0,0,0") {
    angularDampingOverride=1
    angularSpringConstant=200
    bAngularSpring=true
    bOverrideAngularDamping=true
    bOverrideLinearDamping=true
    bUseGravityOverride=true
    bUseSphericalLimits=true
    linearDampingOverride=1
    linearSpringConstant=50
    simulationSpace=RootRelative
    alphaBoolBlend {blendOption=Linear}
    boundBone {boneName=dyn_necklaceB_02}
    componentAppliedLinearAccClamp=<3000 3000 3000>
    componentLinearAccScale=<0.4 0.4 0.4>
    componentLinearVelScale=<0.4 0.4 0.4>
    gravityOverride {z=-980}
    physicsBodyDefinitions [{
      boundBone {boneName=dyn_necklaceB_02}
      boxExtents=<10 10 10>
      constraintSetup {angularLimitsMax=<0 5 5>, angularLimitsMin=<0 -10 -5>, angularTarget=<1 0 0>}
      localJointOffset=<5 0 0>
     }]
    sphericalLimits [{limitRadius=33, drivingBone {boneName=spine_05}, sphereLocalOffset=<-20 -13 0>}]
   }
   n56 = Anim Dynamics - Bone: dyn_necklaceB_03(ComponentPose=@n55, Alpha=1, ExternalForce="0,0,0") {
    angularDampingOverride=1
    angularSpringConstant=200
    bAngularSpring=true
    bOverrideAngularDamping=true
    bOverrideLinearDamping=true
    bUseGravityOverride=true
    bUseSphericalLimits=true
    linearDampingOverride=1
    linearSpringConstant=50
    simulationSpace=RootRelative
    alphaBoolBlend {blendOption=Linear}
    boundBone {boneName=dyn_necklaceB_03}
    componentAppliedLinearAccClamp=<3000 3000 3000>
    componentLinearAccScale=<0.4 0.4 0.4>
    componentLinearVelScale=<0.4 0.4 0.4>
    gravityOverride {z=-980}
    physicsBodyDefinitions [{
      boundBone {boneName=dyn_necklaceB_03}
      boxExtents=<10 10 10>
      constraintSetup {angularLimitsMax=<0 5 5>, angularLimitsMin=<0 -10 -5>, angularTarget=<1 0 0>}
      localJointOffset=<5 0 0>
     }]
    sphericalLimits [{limitRadius=33, drivingBone {boneName=spine_05}, sphereLocalOffset=<-20 -13 0>}]
   }
   Save cached pose 'PostDyn'(Pose=Component To Local(ComponentPose=@n56)) {cacheName=PostDyn}
   n57 = Control Rig!(
    Source=Copy Pose From Mesh() {bUseAttachedParent=true}
    Alpha
    "is Crouching ":double="To Float (Boolean)"(InBool=Get bIsCrouching())
    "is Crouch Moving ":bool="Get Crouch moving "() -> ("Crouch moving ":bool)
    is dbno:double="To Float (Boolean)"(InBool=Get bIsDBNO())
    is RBAN:bool=Get bRigidBodyEnabled()
   ) -> (Pose) {
    alphaBoolBlend {blendOption=Linear}
    controlRigAssetReference {blueprintRigClass=&/M_MED_HonestWasp_CtrlRig.M_MED_HonestWasp_CtrlRig_C}
   }
   Save cached pose 'CtrlRig'(Pose=@n57) {cacheName=CtrlRig}
  ]}
 EventGraph {nodes [
   Event BlueprintUpdateAnimation()
   > n1 = Sequence!(execute) -> (then_0, then_1, then_2:Exec, then_3:Exec, then_4:Exec)
   > "Set Crouch moving "("Crouch moving ":bool=AND Boolean(A=Get bIsCrouching(), B=Get bIsMoving2D())) -> (Output_Get:bool)
   n2 = "Gravity Override State Function (Parameter Asset) (Nativized)"(
    execute=@n1.then_1
    CharacterPartAnimInstance=Self-Reference()
    GameTimeInSeconds=GetGameTimeInSeconds()
    Parameters=&/M_MED_HonestWasp_Parameters
   )
   > Set GravityOverride(GravityOverride=@n2)
   n3 = PickaxeItemDefinitionHasAnyMetaTag(Tags=Make Array("[0]"=(TagName="Cosmetics.Metatags.HonestWasp")), bRequiresEquipped=false)
   n4 = Get bIsRightArmHidden() -> (bIsRightArmHidden:bool)
   Set Is HonestWasp Pickaxe(execute=@n1.then_2, Is HonestWasp Pickaxe:bool=NOT Boolean(A=OR Boolean(A=@n3, B=@n4))) -> (Output_Get:bool)
   > n5 = Branch(Condition=OR Boolean(A=Get bIsTacticalSprinting(), B=Get Is HonestWasp Pickaxe() -> (Is HonestWasp Pickaxe:bool)))
   > n6 = Branch(Condition=Get bIsRightArmHidden())
   n7 = Branch(execute=@n6.else, Condition=Get is Melee() -> (is Melee:bool))
   > Set is Chain Hide(is Chain Hide:bool=false) -> (Output_Get:bool)
   > Set Is HonestWasp Pickaxe(Is HonestWasp Pickaxe=true)
   Set is Chain Hide(execute=[@n7.else, @n9.else], is Chain Hide=false)
   > Set Is HonestWasp Pickaxe(Is HonestWasp Pickaxe=true)
   n8 = Branch(execute=@n5.else, Condition=Get bIsRightArmHidden())
   > Set is Chain Hide(is Chain Hide=true)
   > Set Is HonestWasp Pickaxe(Is HonestWasp Pickaxe=false)
   n9 = Branch(execute=@n8.else, Condition=Get is Melee())
   > Set is Chain Hide(is Chain Hide=true)
   > Set Is HonestWasp Pickaxe(Is HonestWasp Pickaxe=false)
   n10 = Cast To FortPlayerAnimInstance(execute=@n1.then_3, Object=GetAnimInstance(self=GetOwnerSkeletalMeshComponent()))
   > n11 = Switch on EFortWeaponCoreAnimation!(execute, Selection:EFortWeaponCoreAnimation Enum=Get WeaponCoreAnim(self=@n10)) -> (
    Melee:Exec
    Pistol:Exec
    Shotgun:Exec
    PaperBlueprint:Exec
    Rifle:Exec
    MeleeOneHand:Exec
    MachinePistol:Exec
    RocketLauncher:Exec
    GrenadeLauncher:Exec
    GoingCommando:Exec
    AssaultRifle:Exec
    TacticalShotgun:Exec
    SniperRifle:Exec
    TrapPlacement:Exec
    ShoulderLauncher:Exec
    AbilityDecoTool:Exec
    Crossbow:Exec
    C4:Exec
    RemoteControl:Exec
    DualWield:Exec
    AR_BullPup:Exec
    AR_ForwardGrip:Exec
    MedPackPaddles:Exec
    SMG_P90:Exec
    AR_DrumGun:Exec
    Consumable_Small:Exec
    Consumable_Large:Exec
    Balloon:Exec
    MountedTurret:Exec
    CreativeTool:Exec
    ExplosiveBow:Exec
    AshtonIndigo:Exec
    AshtonChicago:Exec
    MeleeDualWield:Exec
    Unarmed:Exec
   )
   Set is Melee(
    execute=[
     @n11.AssaultRifle
     @n11.GrenadeLauncher
     @n11.MachinePistol
     @n11.PaperBlueprint
     @n11.Pistol
     @n11.Rifle
     @n11.RocketLauncher
     @n11.SMG_P90
     @n11.Shotgun
     @n11.SniperRifle
     @n11.TacticalShotgun
    ]
    is Melee:bool=false
   ) -> (Output_Get:bool)
   Set is Melee(execute=@n11.Melee, is Melee=true)
   Set is LODs(execute=@n1.then_4, is LODs:bool="integer >= integer"(A=Get CurrentLOD(), B=2)) -> (Output_Get:bool)
   > Set is Chain Constraint(is Chain Constraint:bool=OR Boolean(A=Get is Chain Hide() -> (is Chain Hide:bool), B=Get is LODs() -> (is LODs:bool))) -> (Output_Get:bool)
   > n12 = Branch(Condition=Get bIsTacticalSprinting())
   > Set Alpha NeckLace(Alpha NeckLace:double=0.5) -> (Output_Get:double)
   Set Alpha NeckLace(execute=@n12.else, Alpha NeckLace=0)
   Event OnCharacterCustomizationLoaded()
   n13 = TryGetPawnOwner()
   > Is Valid(InputObject=@n13)
   > n14 = Cast To FortPlayerPawnAthena(Object=@n13)
   > Set As Fort Player Pawn Athena(As Fort Player Pawn Athena:Fort Player Pawn Athena Object Reference=@n14) -> (Output_Get:Fort Player Pawn Athena Object Reference)
   n15 = CheckForBackbling()
   > n16 = Set isWearingBackbling(isWearingBackbling=AND Boolean!(A=@n15, B=NOT Boolean(A=@n15.bIsBackblingACape), C:bool=NOT Boolean(A=@n15.bIsBackblingLong)) -> (ReturnValue))
   > n17 = Set Is Wearing Cape(Is Wearing Cape=OR Boolean(A=@n15.bIsBackblingACape, B=@n15.bIsBackblingLong))
   > Set is Wearing cape and backpack(is Wearing cape and backpack:bool=OR Boolean(A=@n16, B=@n17)) -> (Output_Get:bool)
   Get bIsInFrontEnd()
  ]}
]
variables [
 Alpha NeckLace {type=number}
 As Fort Player Pawn Athena {type=FortPlayerPawnAthena}
 "Crouch moving " {type=boolean}
 GravityOverride {type=Vector}
 is Chain Constraint {type=boolean}
 is Chain Hide {type=boolean}
 Is HonestWasp Pickaxe {type=boolean}
 is LODs {type=boolean}
 is Melee {type=boolean}
 Is Wearing Cape {type=boolean}
 is Wearing cape and backpack {type=boolean}
 isWearingBackbling {type=boolean}
 Sim Space Settings {
  type=AnimPhysSimSpaceSettings
  default {maxAngularAcceleration=60, maxAngularVelocity=12, simSpaceAngularAlpha=0.7, externalAngularVelocity=<0 0 0>}
 }
]
