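# Physics asset for the right-arm chain on the HonestWasp medium male player body, used for physics-driven chain motion.
# A zero-size base capsule is locked in swing and twist to the right lower arm, and seven small capsule links are constrained in series from it.
# Right upper arm, lower arm and thigh capsules provide body volumes around the chain.
# Shape names referencing thigh ribbons and calves suggest a setup copied from another asset.
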PhysicsAsset /BRCosmetics/Characters/Player/Male/Medium/Bodies/M_MED_HonestWasp/Meshes/M_MED_HonestWasp_arm_chain_Physics
bodies {
 base_chain_arm [{shapeType=Capsule, length=0, radius=0, shapeName=base_thighribbon_A_01_capsule}]
 dyn_chain_arm_01 [{shapeType=Capsule, length=5.5, radius=2.5, shapeName=dyn_thighribbon_A_01_capsule, center=<4 0 0>, rotation=<rot 88.0611 -0.0001 -0.0003>}]
 dyn_chain_arm_02 [{shapeType=Capsule, length=5, radius=2.8, shapeName=dyn_thighribbon_A_02_capsule, center=<4 0 0>, rotation=<rot 84.8641 0 0>}]
 dyn_chain_arm_03 [{shapeType=Capsule, length=4.5, radius=3.3, shapeName=dyn_thighribbon_A_03_capsule, center=<4 0 0>, rotation=<rot 86.9253 -179.9998 -180>}]
 dyn_chain_arm_04 [{shapeType=Capsule, length=4.5, radius=3.3, shapeName=dyn_thighribbon_A_04_capsule, center=<4 0 0>, rotation=<rot 88.871 -179.9999 180>}]
 dyn_chain_arm_05 [{shapeType=Capsule, length=4.5, radius=3.3, shapeName=dyn_thighribbon_A_04_capsule, center=<4 0 0>, rotation=<rot 88.871 -179.9999 180>}]
 dyn_chain_arm_06 [{shapeType=Capsule, length=4.5, radius=3.3, shapeName=dyn_thighribbon_A_04_capsule, center=<4 0 0>, rotation=<rot 88.871 -179.9999 180>}]
 dyn_chain_arm_07 [{shapeType=Capsule, length=4.5, radius=3.3, shapeName=dyn_thighribbon_A_04_capsule, center=<4 0 0>, rotation=<rot 88.871 -179.9999 180>}]
 lowerarm_r [{shapeType=Capsule, length=34.5, radius=8.5, shapeName=calf_l_capsule, center=<-18.5109 -1.6308 0>, rotation=<rot 90 0 0>}]
 thigh_r [{
   shapeType=Capsule
   length=25.5
   radius=9
   shapeName=thigh_l_capsule
   center=<21.8542 0.5837 0.1317>
   rotation=<rot 88.8195 172.6059 170.2331>
  }]
 upperarm_r [{shapeType=Capsule, length=30, radius=8.5, shapeName=calf_l_capsule, center=<-9.0497 -1.6308 0>, rotation=<rot 90 0 0>}]
}
constraints [
 {
  bone1Name=base_chain_arm
  bone2Name=lowerarm_r
  swing1LimitDegrees=40
  swing1Motion=Locked
  swing2LimitDegrees=20
  swing2Motion=Locked
  twistLimitDegrees=1
  twistMotion=Locked
 }
 {bone1Name=dyn_chain_arm_01, bone2Name=base_chain_arm, swing1LimitDegrees=60, swing2LimitDegrees=60, twistLimitDegrees=1}
 {bone1Name=dyn_chain_arm_02, bone2Name=dyn_chain_arm_01, swing1LimitDegrees=40, swing2LimitDegrees=40, twistLimitDegrees=1}
 {bone1Name=dyn_chain_arm_03, bone2Name=dyn_chain_arm_02, swing1LimitDegrees=40, swing2LimitDegrees=40, twistLimitDegrees=1}
 {bone1Name=dyn_chain_arm_04, bone2Name=dyn_chain_arm_03, swing1LimitDegrees=40, swing2LimitDegrees=40, twistLimitDegrees=1}
 {bone1Name=dyn_chain_arm_05, bone2Name=dyn_chain_arm_04, swing1LimitDegrees=40, swing2LimitDegrees=40, twistLimitDegrees=1}
 {bone1Name=dyn_chain_arm_06, bone2Name=dyn_chain_arm_05, swing1LimitDegrees=40, swing2LimitDegrees=40, twistLimitDegrees=1}
 {bone1Name=dyn_chain_arm_07, bone2Name=dyn_chain_arm_06, swing1LimitDegrees=40, swing2LimitDegrees=40, twistLimitDegrees=1}
]
