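# Physics asset for the leg chains on the M_MED_HonestWasp male body cosmetic, with capsules on the pelvis, thighs, calves and chain bones.
# Chain B hangs from the left thigh as one strand, while chain A hangs from the right and forks, its tips joined by an all-free angular constraint.
# Chain joints have narrow twist limits and zero-size base capsules.
# Only the left leg has locked body constraints; right calf and thigh have none.
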PhysicsAsset /BRCosmetics/Characters/Player/Male/Medium/Bodies/M_MED_HonestWasp/Meshes/M_MED_HonestWasp_leg_chain_Physics
bodies {
 base_chain_A_01 [{shapeType=Capsule, length=0, radius=0, shapeName=base_thighribbon_A_01_capsule}]
 base_chain_B_01 [{shapeType=Capsule, length=0, radius=0, shapeName=base_thighribbon_A_01_capsule}]
 calf_l [{shapeType=Capsule, length=34.5, radius=8.5, shapeName=calf_l_capsule, center=<-18.5109 -1.6308 0>, rotation=<rot 90 0 0>}]
 calf_r [{shapeType=Capsule, length=34.5, radius=8.5, shapeName=calf_l_capsule, center=<18.5109 1.6308 0>, rotation=<rot 90 0 0>}]
 dyn_chain_A_01 [{
   shapeType=Capsule
   length=1.5
   radius=4.5
   shapeName=dyn_thighribbon_B_01_capsule
   center=<3.2164 0.009 1.9267>
   rotation=<rot 87.7268 179.9993 -3.2444>
  }]
 dyn_chain_A_02 [{
   shapeType=Capsule
   length=5
   radius=2.8
   shapeName=dyn_thighribbon_A_02_capsule
   center=<4.291 0 1.7353>
   rotation=<rot 84.8642 0.0002 -89.9997>
  }]
 dyn_chain_A_03 [{
   shapeType=Capsule
   length=4.5
   radius=3.3
   shapeName=dyn_thighribbon_A_03_capsule
   center=<4.1596 0 1.7328>
   rotation=<rot 86.9252 179.9997 89.9996>
  }]
 dyn_chain_A_05 [{
   shapeType=Capsule
   length=5
   radius=2.8
   shapeName=dyn_thighribbon_A_02_capsule
   center=<4.291 0 1.7353>
   rotation=<rot 84.8642 0.0002 -89.9997>
  }]
 dyn_chain_A_06 [{
   shapeType=Capsule
   length=4.5
   radius=3.3
   shapeName=dyn_thighribbon_A_03_capsule
   center=<4.1596 0 1.7328>
   rotation=<rot 86.9252 179.9994 89.9994>
  }]
 dyn_chain_B_01 [{
   shapeType=Capsule
   length=5.5
   radius=2.5
   shapeName=dyn_thighribbon_A_01_capsule
   center=<4.224 0.2661 1.8663>
   rotation=<rot 88.0612 -0.0008 -90.0011>
  }]
 dyn_chain_B_02 [{
   shapeType=Capsule
   length=5
   radius=2.8
   shapeName=dyn_thighribbon_A_02_capsule
   center=<4.291 0 1.7353>
   rotation=<rot 84.8642 -0.0002 -90.0003>
  }]
 dyn_chain_B_03 [{
   shapeType=Capsule
   length=4.5
   radius=3.3
   shapeName=dyn_thighribbon_A_03_capsule
   center=<4.1596 0 1.7328>
   rotation=<rot 86.9252 -179.9992 90.0005>
  }]
 pelvis [{shapeType=Capsule, length=7.0633, radius=13.4164, shapeName=pelvis_capsule, center=<-7.5717 3.6488 0>}]
 thigh_l [{
   shapeType=Capsule
   length=25.5
   radius=9
   shapeName=thigh_l_capsule
   center=<-21.8542 -0.5837 0.1317>
   rotation=<rot 88.8195 172.6059 170.2331>
  }]
 thigh_r [{
   shapeType=Capsule
   length=25.5
   radius=9
   shapeName=thigh_l_capsule
   center=<21.8542 0.5837 0.1317>
   rotation=<rot 88.8195 172.6059 170.2331>
  }]
}
constraints [
 {
  bone1Name=calf_l
  bone2Name=thigh_l
  swing1LimitDegrees=45
  swing1Motion=Locked
  swing2LimitDegrees=45
  swing2Motion=Locked
  twistLimitDegrees=45
  twistMotion=Locked
 }
 {
  bone1Name=thigh_l
  bone2Name=pelvis
  swing1LimitDegrees=45
  swing1Motion=Locked
  swing2LimitDegrees=45
  swing2Motion=Locked
  twistLimitDegrees=45
  twistMotion=Locked
 }
 {bone1Name=base_chain_B_01, bone2Name=thigh_l, swing1LimitDegrees=20, swing2LimitDegrees=40, twistLimitDegrees=1}
 {bone1Name=dyn_chain_B_01, bone2Name=base_chain_B_01, swing1LimitDegrees=20, swing2LimitDegrees=40, twistLimitDegrees=20}
 {bone1Name=dyn_chain_B_02, bone2Name=dyn_chain_B_01, swing1LimitDegrees=20, swing2LimitDegrees=40, twistLimitDegrees=1}
 {bone1Name=dyn_chain_B_03, bone2Name=dyn_chain_B_02, swing1LimitDegrees=20, swing2LimitDegrees=40, twistLimitDegrees=1}
 {bone1Name=base_chain_A_01, bone2Name=thigh_r, swing1LimitDegrees=20, swing2LimitDegrees=40, twistLimitDegrees=1}
 {bone1Name=dyn_chain_A_01, bone2Name=base_chain_A_01, swing1LimitDegrees=20, swing2LimitDegrees=40, twistLimitDegrees=20}
 {bone1Name=dyn_chain_A_02, bone2Name=dyn_chain_A_01, swing1LimitDegrees=20, swing2LimitDegrees=40, twistLimitDegrees=1}
 {bone1Name=dyn_chain_A_03, bone2Name=dyn_chain_A_02, swing1LimitDegrees=20, swing2LimitDegrees=40, twistLimitDegrees=1}
 {bone1Name=dyn_chain_A_05, bone2Name=dyn_chain_A_01, swing1LimitDegrees=20, swing2LimitDegrees=40, twistLimitDegrees=1}
 {bone1Name=dyn_chain_A_06, bone2Name=dyn_chain_A_05, swing1LimitDegrees=20, swing2LimitDegrees=40, twistLimitDegrees=1}
 {
  bone1Name=dyn_chain_A_03
  bone2Name=dyn_chain_A_06
  swing1LimitDegrees=45
  swing1Motion=Free
  swing2LimitDegrees=45
  swing2Motion=Free
  twistLimitDegrees=45
  twistMotion=Free
 }
]
