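# Physics asset for the thigh ribbons of the M_MED_HonestWasp male body cosmetic, defining their collision shapes and joint limits.
# Two capsule chains, A with three dynamic segments and B with two, hang from zero-size base bodies locked to the left thigh.
# Chain joints allow limited swing and almost no twist.
# Pelvis, thigh and calf capsules form the body shapes, with locked constraints only on the left leg.
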PhysicsAsset /BRCosmetics/Characters/Player/Male/Medium/Bodies/M_MED_HonestWasp/Meshes/M_MED_HonestWasp_ribbon_Physics
bodies {
 base_thighribbon_A_01 [{shapeType=Capsule, length=0, radius=0, shapeName=base_thighribbon_A_01_capsule}]
 base_thighribbon_B_01 [{shapeType=Capsule, length=0, radius=0, shapeName=base_thighribbon_A_01_capsule}]
 calf_l [{shapeType=Capsule, length=34.5, radius=8.5, shapeName=calf_l_capsule, center=<-18.5109 -1.6308 0>, rotation=<rot 90 0 0>}]
 calf_r [{shapeType=Capsule, length=34.5, radius=8.5, shapeName=calf_l_capsule, center=<18.5109 1.6308 0>, rotation=<rot 90 0 0>}]
 dyn_thighribbon_A_01 [{
   shapeType=Capsule
   length=5.5
   radius=2.5
   shapeName=dyn_thighribbon_A_01_capsule
   center=<4.224 0.2661 0>
   rotation=<rot 88.0611 -0.0001 -0.0003>
  }]
 dyn_thighribbon_A_02 [{shapeType=Capsule, length=5, radius=2.8, shapeName=dyn_thighribbon_A_02_capsule, center=<4.291 0 0>, rotation=<rot 84.8641 0 0>}]
 dyn_thighribbon_A_03 [{
   shapeType=Capsule
   length=4.5
   radius=3.3
   shapeName=dyn_thighribbon_A_03_capsule
   center=<4.1596 0 1.3>
   rotation=<rot 86.9253 -179.9998 -180>
  }]
 dyn_thighribbon_B_01 [{
   shapeType=Capsule
   length=5.5
   radius=2.5
   shapeName=dyn_thighribbon_A_01_capsule
   center=<4.224 0.2661 0>
   rotation=<rot 88.0611 -0.0001 -0.0003>
  }]
 dyn_thighribbon_B_02 [{shapeType=Capsule, length=5, radius=2.8, shapeName=dyn_thighribbon_A_02_capsule, center=<4.291 0 0>, rotation=<rot 84.8641 0 0>}]
 pelvis [{shapeType=Capsule, length=7.0633, radius=13.4164, shapeName=pelvis_capsule, center=<-7.5717 3.6488 0>}]
 thigh_l [{shapeType=Capsule, length=41.5, radius=9.5, center=<-21.8081 0.097 -0.042>, rotation=<rot 86.8191 178.2336 -3.6745>}]
 thigh_r [{shapeType=Capsule, length=41.5, radius=9.5, center=<21.7429 -0.097 -0.042>, rotation=<rot -86.8191 3.7526 3.6745>}]
}
constraints [
 {bone1Name=dyn_thighribbon_A_01, bone2Name=base_thighribbon_A_01, swing1LimitDegrees=20, swing2LimitDegrees=40, twistLimitDegrees=1}
 {bone1Name=dyn_thighribbon_A_02, bone2Name=dyn_thighribbon_A_01, swing1LimitDegrees=20, swing2LimitDegrees=40, twistLimitDegrees=1}
 {bone1Name=dyn_thighribbon_A_03, bone2Name=dyn_thighribbon_A_02, swing1LimitDegrees=20, swing2LimitDegrees=40, twistLimitDegrees=1}
 {bone1Name=dyn_thighribbon_B_01, bone2Name=base_thighribbon_B_01, swing1LimitDegrees=20, swing2LimitDegrees=30, twistLimitDegrees=1}
 {bone1Name=dyn_thighribbon_B_02, bone2Name=dyn_thighribbon_B_01, swing1LimitDegrees=20, swing2LimitDegrees=30, twistLimitDegrees=1}
 {
  bone1Name=base_thighribbon_A_01
  bone2Name=thigh_l
  swing1LimitDegrees=45
  swing1Motion=Locked
  swing2LimitDegrees=45
  swing2Motion=Locked
  twistLimitDegrees=45
  twistMotion=Locked
 }
 {
  bone1Name=base_thighribbon_B_01
  bone2Name=thigh_l
  swing1LimitDegrees=45
  swing1Motion=Locked
  swing2LimitDegrees=45
  swing2Motion=Locked
  twistLimitDegrees=45
  twistMotion=Locked
 }
 {
  bone1Name=calf_l
  bone2Name=thigh_l
  swing1LimitDegrees=45
  swing1Motion=Locked
  swing2LimitDegrees=45
  swing2Motion=Locked
  twistLimitDegrees=45
  twistMotion=Locked
 }
 {
  bone1Name=thigh_l
  bone2Name=pelvis
  swing1LimitDegrees=45
  swing1Motion=Locked
  swing2LimitDegrees=45
  swing2Motion=Locked
  twistLimitDegrees=45
  twistMotion=Locked
 }
]
