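# Control Rig for the hair and pigtail bones on HonestWasp's face accessory, posing them in response to body pose.
# Spherical pose readers on head, clavicles and upper arms weight additive bone offsets and move targets that hair bones aim at.
# Smoothed pickaxe-swing, targeting and front-end flags gate extra corrections, and a backbling cape flag suppresses one helper offset.
# While targeting, front hair skin bones scale down as the upper arms near the head.
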ControlRigBlueprint /BRCosmetics/Characters/Player/Male/Medium/Bodies/M_MED_HonestWasp/Meshes/Parts/M_MED_HonestWasp_FaceAcc_CtrlRig
graphs [
 CollapseNode_ContainedGraph {nodes [
   n1 = RigVMFunctionEntryNode()
   n2 = &RigVMFunction_Sequence(ExecuteContext=@n1.ExecuteContext)
   RigVMFunction_Sequence_2 = RigVMAggregateNode(ExecuteContext=@n2.A)
   n3 = &RigUnit_SphericalPoseReader(
    ActiveRegionScaleFactors=(PositiveWidth=1.35,NegativeWidth=1.35,PositiveHeight=0.45,NegativeHeight=1.9)
    ActiveRegionSize=0.1
    Debug=(bDrawDebug=false,bDraw2D=False,bDrawLocalAxes=False,DebugScale=25,DebugSegments=20,DebugThickness=0.25)
    DriverAxis=(X=1,Y=1,Z=0)
    DriverItem=(Type=Bone,Name="head")
    FalloffRegionScaleFactors=(PositiveWidth=1,NegativeWidth=1,PositiveHeight=1.0266,NegativeHeight=1)
    FalloffSize=0.235
    FlipHeightScaling=False
    FlipWidthScaling=False
    OptionalParentItem=(Type=Bone,Name="None")
    RotationOffset=(X=0,Y=-90,Z=0)
    ExecutePin=@RigVMFunction_Sequence_2.A
   )
   > n4 = &RigUnit_SphericalPoseReader(
    ActiveRegionScaleFactors=(PositiveWidth=1.35,NegativeWidth=1.35,PositiveHeight=1,NegativeHeight=1.9)
    ActiveRegionSize=0.1968
    Debug=(bDrawDebug=false,bDraw2D=False,bDrawLocalAxes=False,DebugScale=25,DebugSegments=20,DebugThickness=0.25)
    DriverAxis=(X=1,Y=1,Z=0)
    DriverItem=(Type=Bone,Name="head")
    FalloffRegionScaleFactors=(PositiveWidth=1,NegativeWidth=1,PositiveHeight=1.0266,NegativeHeight=1)
    FalloffSize=0.235
    FlipHeightScaling=False
    FlipWidthScaling=False
    OptionalParentItem=(Type=Bone,Name="None")
    RotationOffset=(X=-37,Y=-90,Z=0)
   )
   > n5 = &RigUnit_SphericalPoseReader(
    ActiveRegionScaleFactors=(PositiveWidth=0.21,NegativeWidth=1,PositiveHeight=1,NegativeHeight=1)
    ActiveRegionSize=0.5
    Debug=(bDrawDebug=True,bDraw2D=False,bDrawLocalAxes=False,DebugScale=25,DebugSegments=20,DebugThickness=0.25)
    DriverAxis=(X=1,Y=0,Z=0)
    DriverItem=(Type=Bone,Name="head")
    FalloffRegionScaleFactors=(PositiveWidth=1,NegativeWidth=1,PositiveHeight=1,NegativeHeight=1)
    FalloffSize=0.5345
    FlipHeightScaling=False
    FlipWidthScaling=False
    OptionalParentItem=(Type=Bone,Name="None")
    RotationOffset=(X=90,Y=0,Z=-10)
   )
   > n6 = &RigUnit_SphericalPoseReader(
    ActiveRegionScaleFactors=(PositiveWidth=1,NegativeWidth=1,PositiveHeight=1,NegativeHeight=1)
    ActiveRegionSize=0.15
    Debug=(bDrawDebug=false,bDraw2D=false,DebugScale=25,DebugSegments=20,DebugThickness=0.25)
    DriverAxis=(X=0,Y=1,Z=0)
    DriverItem=(Type=Bone,Name="head")
    FalloffRegionScaleFactors=(PositiveWidth=1,NegativeWidth=1,PositiveHeight=1,NegativeHeight=1)
    FalloffSize=0.25
    OptionalParentItem=(Type=Bone)
    RotationOffset=(X=150,Y=0,Z=0)
   )
   > n7 = &RigUnit_SphericalPoseReader(
    ActiveRegionScaleFactors=(PositiveWidth=1,NegativeWidth=1,PositiveHeight=1,NegativeHeight=1)
    ActiveRegionSize=0.15
    Debug=(bDrawDebug=false,bDraw2D=false,DebugScale=25,DebugSegments=20,DebugThickness=0.25)
    DriverAxis=(X=0,Y=1,Z=0)
    DriverItem=(Type=Bone,Name="head")
    FalloffRegionScaleFactors=(PositiveWidth=1,NegativeWidth=1,PositiveHeight=1,NegativeHeight=1)
    FalloffSize=0.25
    FlipHeightScaling=true
    FlipWidthScaling=false
    OptionalParentItem=(Type=Bone)
    RotationOffset=(X=30,Y=0,Z=0)
   )
   &RigUnit_ModifyTransforms(
    ItemToModify=((Item=(Type=Bone,Name="head_aim_tgt"),Transform=(Rotation=(X=0,Y=0,Z=0,W=1),Translation=(X=-10,Y=5,Z=0),Scale3D=(X=1,Y=1,Z=1))))
    WeightMinimum=0
    ExecutePin=@RigVMFunction_Sequence_2.B
    Weight=@n3.OutputParam
   )
   > &RigUnit_ModifyTransforms(
    ItemToModify=((Item=(Type=Bone,Name="head_aim_tgt"),Transform=(Rotation=(X=0,Y=0,Z=0,W=1),Translation=(X=-10,Y=-5,Z=0),Scale3D=(X=1,Y=1,Z=1))))
    WeightMinimum=0
    Weight=@n4.OutputParam
   )
   > &RigUnit_ModifyTransforms(
    ItemToModify=((Item=(Type=Bone,Name="head_aim_tgt"),Transform=(Rotation=(X=0,Y=0,Z=0,W=1),Translation=(X=0,Y=-10,Z=0),Scale3D=(X=1,Y=1,Z=1))))
    WeightMinimum=0
    Weight=@n5.OutputParam
   )
   > &RigUnit_ModifyTransforms(
    ItemToModify=((Item=(Type=Bone,Name="head_aim_tgt"),Transform=(Rotation=(X=0,Y=0,Z=0,W=1),Translation=(X=0,Y=0,Z=-3),Scale3D=(X=1,Y=1,Z=1))))
    WeightMinimum=0
    Weight=@n6.OutputParam
   )
   > &RigUnit_ModifyTransforms(
    ItemToModify=((Item=(Type=Bone,Name="head_aim_tgt"),Transform=(Rotation=(X=0,Y=0,Z=0,W=1),Translation=(X=0,Y=0,Z=-5),Scale3D=(X=1,Y=1,Z=1))))
    WeightMinimum=0
    Weight=@n7.OutputParam
   )
   &RigUnit_AimConstraintLocalSpaceOffset(
    Child=(Type=Bone,Name="aim_hair_lf")
    Filter=(bX=true,bY=true,bZ=true)
    Parents=((Item=(Type=Bone,Name="head_aim_tgt"),Weight=1))
    UpAxis=(X=0,Y=-1,Z=0)
    Weight=1
    WorldUp=(Target=(X=0,Y=-1,Z=0),Kind=Location,Space=(Type=Bone,Name="head_aim_tgt"))
    bMaintainOffset=true
    ExecutePin=@RigVMFunction_Sequence_2.D
   )
   > &RigUnit_AimConstraintLocalSpaceOffset(
    Child=(Type=Bone,Name="aim_hair_cr")
    Filter=(bX=true,bY=true,bZ=true)
    Parents=((Item=(Type=Bone,Name="head_aim_tgt"),Weight=1))
    UpAxis=(X=0,Y=-1,Z=0)
    Weight=1
    WorldUp=(Target=(X=0,Y=-1,Z=0),Kind=Location,Space=(Type=Bone,Name="head_aim_tgt"))
    bMaintainOffset=true
   )
   > &RigUnit_AimConstraintLocalSpaceOffset(
    Child=(Type=Bone,Name="aim_hair_rt")
    Filter=(bX=true,bY=true,bZ=true)
    Parents=((Item=(Type=Bone,Name="head_aim_tgt"),Weight=1))
    UpAxis=(X=0,Y=-1,Z=0)
    Weight=1
    WorldUp=(Target=(X=0,Y=-1,Z=0),Kind=Location,Space=(Type=Bone,Name="head_aim_tgt"))
    bMaintainOffset=true
   )
   &RigUnit_ModifyTransforms(
    ItemToModify="((Item=(Type=Bone,Name=\"base_hair_rt\"),Transform=(Rotation=(X=0,Y=-0.0872,Z=0,W=0.9962),Translation=(X=-5,Y=0,Z=0),Scale3D=(X=1,Y=1,Z=1))),(Item=(Type=Bone,Name=\"base_hair_cr\"),Transform=(Rotation=(X=0,Y=0,Z=0,W=1),Translation=(X=0,Y=0,Z=0),Scale3D=(X=1,Y= ...(truncated)"
    WeightMinimum=0
    ExecutePin=@RigVMFunction_Sequence_2.E
    Weight=@n3.OutputParam
   )
   > n8 = &RigUnit_SphericalPoseReader(
    ActiveRegionScaleFactors=(PositiveWidth=0.235,NegativeWidth=1,PositiveHeight=1,NegativeHeight=1)
    ActiveRegionSize=0.4
    Debug=(bDrawDebug=false,bDraw2D=False,bDrawLocalAxes=False,DebugScale=25,DebugSegments=20,DebugThickness=0.25)
    DriverAxis=(X=1,Y=0,Z=0)
    DriverItem=(Type=Bone,Name="clavicle_l")
    FalloffRegionScaleFactors=(PositiveWidth=1,NegativeWidth=1,PositiveHeight=1,NegativeHeight=1)
    FalloffSize=0.5
    FlipHeightScaling=False
    FlipWidthScaling=False
    OptionalParentItem=(Type=Bone,Name="None")
    RotationOffset=(X=-90,Y=0,Z=0)
   )
   n9 = &RigVMFunction_AlphaInterp(InterpSpeedDecreasing=10, InterpSpeedIncreasing=10, Scale=1, bClampResult=False, Value=@n8.OutputParam)
   Cast = &RigVMFunction_MathBoolToFloat(Value=RigVMVariableNode() {variable=swingPickaxe}.Value)
   n10 = &RigVMFunction_AlphaInterp(InterpSpeedDecreasing=10, InterpSpeedIncreasing=10, Scale=1, bClampResult=False, Value=@Cast.Result)
   RigVMFunction_AnimEvalRichCurve = &RigVMFunction_AnimEvalRichCurve(
    Curve="(EditorCurveData=(Keys=((InterpMode=RCIM_Cubic,TangentMode=RCTM_Auto,TangentWeightMode=RCTWM_WeightedNone,Time=0,Value=0,ArriveTangent=0,ArriveTangentWeight=0,LeaveTangent=0,LeaveTangentWeight=0),(InterpMode=RCIM_Cubic,TangentMode=RCTM_Auto,TangentWeightMo ...(truncated)"
    SourceMinimum=0
    TargetMinimum=0
    Value=@n10.Result
   )
   n11 = &RigVMFunction_AlphaInterp(
    InterpSpeedDecreasing=10
    InterpSpeedIncreasing=10
    Scale=1
    bClampResult=False
    Value=&RigVMFunction_MathDoubleMul(A=@n9.Result, B=@RigVMFunction_AnimEvalRichCurve.Result).Result
   )
   > &RigUnit_ModifyTransforms(
    ItemToModify="((Item=(Type=Bone,Name=\"aim_hair_lf\"),Transform=(Rotation=(X=-0.4467,Y=-0.1419,Z=0.254,W=0.8461),Translation=(X=0,Y=0,Z=0),Scale3D=(X=1,Y=1,Z=1))),(Item=(Type=Bone,Name=\"aim_hair_cr\"),Transform=(Rotation=(X=-0.0856,Y=-0.1914,Z=0.3994,W=0.8925),Translation= ...(truncated)"
    WeightMinimum=0
    Weight=@n11.Result
   )
   > n12 = &RigUnit_SphericalPoseReader(
    ActiveRegionScaleFactors=(PositiveWidth=0.235,NegativeWidth=1,PositiveHeight=1,NegativeHeight=1)
    ActiveRegionSize=0.4
    Debug=(bDrawDebug=false,bDraw2D=False,bDrawLocalAxes=False,DebugScale=25,DebugSegments=20,DebugThickness=0.25)
    DriverAxis=(X=-1,Y=0,Z=0)
    DriverItem=(Type=Bone,Name="clavicle_r")
    FalloffRegionScaleFactors=(PositiveWidth=1,NegativeWidth=1,PositiveHeight=1,NegativeHeight=1)
    FalloffSize=0.5
    FlipHeightScaling=False
    FlipWidthScaling=False
    OptionalParentItem=(Type=Bone,Name="None")
    RotationOffset=(X=90,Y=180,Z=0)
   )
   n13 = &RigVMFunction_AlphaInterp(InterpSpeedDecreasing=10, InterpSpeedIncreasing=10, Scale=1, bClampResult=False, Value=@n12.OutputParam)
   n14 = &RigVMFunction_AlphaInterp(
    InterpSpeedDecreasing=10
    InterpSpeedIncreasing=10
    Scale=1
    bClampResult=False
    Value=&RigVMFunction_MathDoubleMul(A=@n13.Result, B=@RigVMFunction_AnimEvalRichCurve.Result).Result
   )
   > &RigUnit_ModifyTransforms(
    ItemToModify="((Item=(Type=Bone,Name=\"aim_hair_lf\"),Transform=(Rotation=(X=-0.4702,Y=0.1943,Z=0.1793,W=-0.842),Translation=(X=0,Y=0,Z=0),Scale3D=(X=1,Y=1,Z=1))),(Item=(Type=Bone,Name=\"aim_hair_cr\"),Transform=(Rotation=(X=-0.2449,Y=0.0471,Z=0.3915,W=-0.8858),Translation= ...(truncated)"
    WeightMinimum=0
    Weight=@n14.Result
   )
   n15 = &RigUnit_SphericalPoseReader(
    ActiveRegionScaleFactors=(PositiveWidth=1,NegativeWidth=1,PositiveHeight=1,NegativeHeight=1)
    ActiveRegionSize=0.1
    Debug=(bDrawDebug=false,bDraw2D=false,bDrawLocalAxes=False,DebugScale=25,DebugSegments=20,DebugThickness=0.25)
    DriverAxis=(X=0,Y=1,Z=0)
    DriverItem=(Type=Bone,Name="head")
    FalloffRegionScaleFactors=(PositiveWidth=1,NegativeWidth=1,PositiveHeight=1,NegativeHeight=1)
    FalloffSize=0.2
    FlipHeightScaling=False
    FlipWidthScaling=False
    OptionalParentItem=(Type=Bone,Name="None")
    RotationOffset=(X=90,Y=0,Z=15)
    ExecutePin=@RigVMFunction_Sequence_2.F
   )
   > n16 = &RigUnit_SphericalPoseReader(
    ActiveRegionScaleFactors=(PositiveWidth=1,NegativeWidth=1.5,PositiveHeight=1,NegativeHeight=1)
    ActiveRegionSize=0.15
    Debug=(bDrawDebug=false,bDraw2D=false,bDrawLocalAxes=False,DebugScale=25,DebugSegments=20,DebugThickness=0.25)
    DriverAxis=(X=-1,Y=0,Z=0)
    DriverItem=(Type=Bone,Name="upperarm_r")
    FalloffRegionScaleFactors=(PositiveWidth=1,NegativeWidth=2,PositiveHeight=1,NegativeHeight=1)
    FalloffSize=0.25
    FlipHeightScaling=False
    FlipWidthScaling=False
    OptionalParentItem=(Type=Bone,Name="None")
    RotationOffset=(X=-50,Y=0,Z=-40)
   )
   n17 = &RigVMFunction_AlphaInterp(
    InterpSpeedDecreasing=10
    InterpSpeedIncreasing=10
    Scale=1
    bClampResult=false
    bMapRange=false
    Value=&RigVMFunction_MathBoolToFloat(Value=RigVMVariableNode() {variable=IsTargeting}.Value).Result
   )
   n18 = &RigVMFunction_AnimEvalRichCurve(
    Curve="(EditorCurveData=(Keys=((InterpMode=RCIM_Cubic,TangentMode=RCTM_Auto,TangentWeightMode=RCTWM_WeightedNone,Time=0,Value=0,ArriveTangent=0,ArriveTangentWeight=0,LeaveTangent=0,LeaveTangentWeight=0),(InterpMode=RCIM_Cubic,TangentMode=RCTM_Auto,TangentWeightMo ...(truncated)"
    SourceMinimum=0
    TargetMinimum=0
    Value=@n17.Result
   )
   Multiply = RigVMAggregateNode(C=@n18.Result, B=@n16.OutputParam, A=@n15.OutputParam)
   > &RigUnit_ModifyTransforms(
    ItemToModify=((Item=(Type=Bone,Name="offset_hair_rt"),Transform=(Rotation=(X=0,Y=0,Z=0,W=1),Translation=(X=0,Y=0,Z=3),Scale3D=(X=1,Y=1,Z=1))))
    WeightMinimum=0
    Weight=@Multiply.Result
   )
   > Delta Rotation = RigVMFunctionReferenceNode(Child=head, Parent=neck_02, Source Maximum=-21, Source Minimum=20)
   > n19 = &RigUnit_SphericalPoseReader(
    ActiveRegionScaleFactors=(PositiveWidth=1,NegativeWidth=1,PositiveHeight=1,NegativeHeight=1)
    ActiveRegionSize=0.4
    Debug=(bDrawDebug=false,bDraw2D=False,bDrawLocalAxes=False,DebugScale=25,DebugSegments=20,DebugThickness=0.25)
    DriverAxis=(X=-1,Y=0,Z=0)
    DriverItem=(Type=Bone,Name="upperarm_r")
    FalloffRegionScaleFactors=(PositiveWidth=1,NegativeWidth=1,PositiveHeight=1,NegativeHeight=1)
    FalloffSize=0.5
    FlipHeightScaling=False
    FlipWidthScaling=False
    OptionalParentItem=(Type=Bone,Name="None")
    RotationOffset=(X=-30,Y=40,Z=40)
   )
   n20 = RigVMVariableNode() {variable=isWearingBackblingCape}
   n21 = &RigVMFunction_AlphaInterp(
    InterpSpeedDecreasing=10
    InterpSpeedIncreasing=10
    Scale=1
    bClampResult=false
    bMapRange=false
    Value=&RigVMFunction_MathBoolToFloat(Value=RigVMVariableNode() {variable=IsTargeting}.Value).Result
   )
   n22 = &RigVMFunction_AnimEvalRichCurve(
    Curve="(EditorCurveData=(Keys=((InterpMode=RCIM_Cubic,TangentMode=RCTM_Auto,TangentWeightMode=RCTWM_WeightedNone,Time=0,Value=0,ArriveTangent=0,ArriveTangentWeight=0,LeaveTangent=0,LeaveTangentWeight=0),(InterpMode=RCIM_Cubic,TangentMode=RCTM_Auto,TangentWeightMo ...(truncated)"
    SourceMinimum=0
    TargetMinimum=0
    Value=@n21.Result
   )
   Multiply_2 = RigVMAggregateNode(C=@n22.Result, B=@n19.OutputParam, A=@"Delta Rotation".Delta)
   > &RigUnit_ModifyTransforms(
    ItemToModify=((Item=(Type=Bone,Name="dyn_hair_02_rt_helper"),Transform=(Rotation=(X=0,Y=0.1736,Z=0,W=0.9848),Translation=(X=-5,Y=0,Z=2),Scale3D=(X=1,Y=1,Z=0.6))))
    WeightMinimum=0
    Weight=&RigVMFunction_MathDoubleRemap(
     SourceMaximum=1
     SourceMinimum=0
     TargetMaximum=0
     Value=&RigVMFunction_MathBoolToFloat(Value=@n20.Value).Result
     TargetMinimum=@Multiply_2.Result
    ).Result
   )
   Remap = &RigVMFunction_MathDoubleRemap(SourceMaximum=0.25, SourceMinimum=0, TargetMaximum=1, TargetMinimum=0, Value=RigVMVariableNode() {variable=Head Down}.Value)
   > &RigUnit_ModifyTransforms(
    ItemToModify=((Item=(Type=Bone,Name="dyn_hair_02_rt_helper"),Transform=(Rotation=(X=0,Y=0,Z=0,W=1),Translation=(X=0,Y=0,Z=1),Scale3D=(X=1,Y=1,Z=1))))
    WeightMinimum=0
    Weight=&RigVMFunction_MathDoubleMul(A=@Remap.Result, B=@n22.Result).Result
   )
   RigVMFunctionReturnNode(
    Result=(X=0,Y=0,Z=0,W=1)
    Head Right=@n7.OutputParam
    Head Left=@n6.OutputParam
    Head Up=@n3.OutputParam
    Head Up OverShoot=@n4.OutputParam
    Head Down=@n5.OutputParam
    ExecuteContext=@n2.B
   )
  ]}
 Multiply_1_ContainedGraph {nodes [
   n1 = RigVMFunctionEntryNode()
   Multiply = &RigVMFunction_MathDoubleMul(A=@n1.A, B=@n1.B)
   RigVMFunctionReturnNode(Result=&RigVMFunction_MathDoubleMul(B=@n1.C, A=@Multiply.Result).Result)
  ]}
 Multiply_2_ContainedGraph {nodes [
   n1 = RigVMFunctionEntryNode()
   Multiply = &RigVMFunction_MathDoubleMul(A=@n1.A, B=@n1.B)
   RigVMFunctionReturnNode(Result=&RigVMFunction_MathDoubleMul(B=@n1.C, A=@Multiply.Result).Result)
  ]}
 Multiply_2_ContainedGraph {nodes [
   n1 = RigVMFunctionEntryNode()
   Multiply = &RigVMFunction_MathDoubleMul(A=@n1.A, B=@n1.B)
   RigVMFunctionReturnNode(Result=&RigVMFunction_MathDoubleMul(B=@n1.C, A=@Multiply.Result).Result)
  ]}
 New Function_ContainedGraph {nodes [
   n1 = RigVMFunctionEntryNode()
   RigVMFunction_MathQuaternionToEuler = &RigVMFunction_MathQuaternionToEuler(
    RotationOrder=ZYX
    Value=&RigUnit_GetRelativeTransformForItem(
     Child=(Type=Bone,Name="hand_l")
     Parent=(Type=Bone,Name="lowerarm_l")
     bChildInitial=False
     bParentInitial=False
     Child.Name=@n1.Child
     Parent.Name=@n1.Parent
    ).RelativeTransform.Rotation
   )
   Remap = &RigVMFunction_MathDoubleRemap(
    TargetMaximum=1
    TargetMinimum=0
    Value=@RigVMFunction_MathQuaternionToEuler.Result.Z
    SourceMaximum=@n1.Source Maximum
    SourceMinimum=@n1.Source Minimum
   )
   > RigVMFunctionReturnNode(Delta=@Remap.Result)
  ]}
 New Function_ContainedGraph {nodes [
   n1 = RigVMFunctionEntryNode()
   RigVMFunction_MathQuaternionToEuler = &RigVMFunction_MathQuaternionToEuler(
    RotationOrder=ZYX
    Value=&RigUnit_GetRelativeTransformForItem(
     Child=(Type=Bone,Name="hand_l")
     Parent=(Type=Bone,Name="lowerarm_l")
     bChildInitial=false
     bParentInitial=false
     Child.Name=@n1.Child
     Parent.Name=@n1.Parent
    ).RelativeTransform.Rotation
   )
   Remap = &RigVMFunction_MathDoubleRemap(
    TargetMaximum=1
    TargetMinimum=0
    Value=@RigVMFunction_MathQuaternionToEuler.Result.Y
    SourceMaximum=@n1.Source Maximum
    SourceMinimum=@n1.Source Minimum
   )
   > RigVMFunctionReturnNode(Delta=@Remap.Result)
  ]}
 New Function_ContainedGraph {nodes [
   n1 = RigVMFunctionEntryNode()
   RigVMFunction_MathBoolToInteger = &RigVMFunction_MathBoolToInteger(Value=@n1.onlyTwist)
   RigVMFunction_MathQuaternionSwingTwist = &RigVMFunction_MathQuaternionSwingTwist(
    twistAxis=@n1.twistAxis
    Input=&RigUnit_GetRelativeTransformForItem(
     bChildInitial=False
     bParentInitial=False
     child=(Type=Bone,Name="None")
     parent=(Type=Bone,Name="None")
     child.Name=@n1.child
     parent.Name=@n1.parent
     parent.Type=@n1.parentType
    ).RelativeTransform.Rotation
   )
   RigVMFunction_MathQuaternionToEuler = &RigVMFunction_MathQuaternionToEuler(
    Value=RigVMDispatchNode(
     Values=((),())
     Index=@RigVMFunction_MathBoolToInteger.Result
     Values.0=@RigVMFunction_MathQuaternionSwingTwist.Swing
     Values.1=@RigVMFunction_MathQuaternionSwingTwist.Twist
    ) {template="DISPATCH_RigVMDispatch_SelectInt32(in Index,in Values,out Result)"}.Result
    RotationOrder=@n1.Rotation Order
   )
   > DISPATCH_RigVMDispatch_Print = RigVMDispatchNode(ScreenColor=(R=1,G=1,B=1,A=1), ScreenDuration=1.5, Enabled=@n1.debugRotation, Value=@RigVMFunction_MathQuaternionToEuler.Result) {template="DISPATCH_RigVMDispatch_Print(in Prefix,in Value,in Enabled,in ScreenDuration,in ScreenColor)"}
   Clamp = &RigVMFunction_MathIntClamp(Value=@n1.primAxis)
   DISPATCH_RigVMDispatch_SelectInt32 = RigVMDispatchNode(
    Values=(0,0,0)
    Index=@Clamp.Result
    Values.0=@RigVMFunction_MathQuaternionToEuler.Result.X
    Values.1=@RigVMFunction_MathQuaternionToEuler.Result.Y
    Values.2=@RigVMFunction_MathQuaternionToEuler.Result.Z
   ) {template="DISPATCH_RigVMDispatch_SelectInt32(in Index,in Values,out Result)"}
   Remap = &RigVMFunction_MathDoubleRemap(
    TargetMaximum=1
    TargetMinimum=0
    bClamp=False
    Value=@DISPATCH_RigVMDispatch_SelectInt32.Result
    SourceMaximum=@n1.max
    SourceMinimum=@n1.min
   )
   Multiply = &RigVMFunction_MathDoubleMul(A=&RigVMFunction_MathDoubleClamp(Value=@Remap.Result).Result, B=@n1.premult)
   > RigVMFunctionReturnNode(outputWeight=@Multiply.Result)
  ]}
 New Function_ContainedGraph {nodes [
   n1 = RigVMFunctionEntryNode()
   RigVMFunction_MathBoolToInteger = &RigVMFunction_MathBoolToInteger(Value=@n1.onlyTwist)
   RigVMFunction_MathQuaternionSwingTwist = &RigVMFunction_MathQuaternionSwingTwist(
    twistAxis=@n1.twistAxis
    Input=&RigUnit_GetRelativeTransformForItem(
     bChildInitial=False
     bParentInitial=False
     child=(Type=Bone,Name="None")
     parent=(Type=Bone,Name="None")
     child.Name=@n1.child
     parent.Name=@n1.parent
     parent.Type=@n1.parentType
    ).RelativeTransform.Rotation
   )
   RigVMFunction_MathQuaternionToEuler = &RigVMFunction_MathQuaternionToEuler(
    Value=RigVMDispatchNode(
     Values=((),())
     Index=@RigVMFunction_MathBoolToInteger.Result
     Values.0=@RigVMFunction_MathQuaternionSwingTwist.Swing
     Values.1=@RigVMFunction_MathQuaternionSwingTwist.Twist
    ) {template="DISPATCH_RigVMDispatch_SelectInt32(in Index,in Values,out Result)"}.Result
    RotationOrder=@n1.Rotation Order
   )
   > DISPATCH_RigVMDispatch_Print = RigVMDispatchNode(ScreenColor=(R=1,G=1,B=1,A=1), ScreenDuration=1.5, Enabled=@n1.debugRotation, Value=@RigVMFunction_MathQuaternionToEuler.Result) {template="DISPATCH_RigVMDispatch_Print(in Prefix,in Value,in Enabled,in ScreenDuration,in ScreenColor)"}
   Clamp = &RigVMFunction_MathIntClamp(Value=@n1.primAxis)
   DISPATCH_RigVMDispatch_SelectInt32 = RigVMDispatchNode(
    Values=(0,0,0)
    Index=@Clamp.Result
    Values.0=@RigVMFunction_MathQuaternionToEuler.Result.X
    Values.1=@RigVMFunction_MathQuaternionToEuler.Result.Y
    Values.2=@RigVMFunction_MathQuaternionToEuler.Result.Z
   ) {template="DISPATCH_RigVMDispatch_SelectInt32(in Index,in Values,out Result)"}
   Remap = &RigVMFunction_MathDoubleRemap(
    TargetMaximum=1
    TargetMinimum=0
    bClamp=False
    Value=@DISPATCH_RigVMDispatch_SelectInt32.Result
    SourceMaximum=@n1.max
    SourceMinimum=@n1.min
   )
   Multiply = &RigVMFunction_MathDoubleMul(A=&RigVMFunction_MathDoubleClamp(Value=@Remap.Result).Result, B=@n1.premult)
   > RigVMFunctionReturnNode(outputWeight=@Multiply.Result)
  ]}
 New Function_ContainedGraph {nodes [
   n1 = RigVMFunctionEntryNode()
   RigVMFunction_MathQuaternionToEuler = &RigVMFunction_MathQuaternionToEuler(
    RotationOrder=ZYX
    Value=&RigUnit_GetRelativeTransformForItem(
     Child=(Type=Bone,Name="head")
     Parent=(Type=Null,Name="neck_01")
     bChildInitial=False
     bParentInitial=False
     Child.Name=@n1.Child
     Parent.Name=@n1.Parent
    ).RelativeTransform.Rotation
   )
   Remap = &RigVMFunction_MathDoubleRemap(
    TargetMaximum=1
    TargetMinimum=0
    SourceMaximum=@n1.Source Max
    SourceMinimum=@n1.Source Min
    Value=@RigVMFunction_MathQuaternionToEuler.Result.Z
   )
   > RigVMFunctionReturnNode(
    Delta_z=@Remap.Result
    Delta_X_Negative=&RigVMFunction_MathDoubleRemap(SourceMaximum=-60, SourceMinimum=0, TargetMaximum=1, TargetMinimum=0, Value=@RigVMFunction_MathQuaternionToEuler.Result.X).Result
    Delta_X_Positive=&RigVMFunction_MathDoubleRemap(SourceMaximum=70, SourceMinimum=0, TargetMaximum=1, TargetMinimum=0, Value=@RigVMFunction_MathQuaternionToEuler.Result.X).Result
   )
  ]}
 RigVMFunctionLibrary {nodes [
   Delta Rotation = RigVMCollapseNode(ExecuteContext=())
   Delta Rotation_clavicle_up = RigVMCollapseNode(ExecuteContext=())
   DeltaRotation = RigVMCollapseNode(ExecuteContext=(), Rotation Order=ZYX, premult=1, twistAxis=(X=1,Y=0,Z=0))
   DeltaRotation_1 = RigVMCollapseNode(ExecuteContext=(), Rotation Order=ZYX, premult=1, twistAxis=(X=1,Y=0,Z=0))
   DeltaRotation_2 = RigVMCollapseNode(ExecuteContext=())
  ]}
 RigVMFunction_Sequence_1_ContainedGraph {nodes [
   n1 = RigVMFunctionEntryNode()
   RigVMFunction_Sequence = &RigVMFunction_Sequence(ExecuteContext=@n1.ExecuteContext)
   n2 = &RigVMFunction_Sequence(ExecuteContext=@RigVMFunction_Sequence.B)
   RigVMFunctionReturnNode(A=@RigVMFunction_Sequence.A, B=@n2.A, C=@n2.B)
  ]}
 RigVMFunction_Sequence_3_ContainedGraph {nodes [
   n1 = RigVMFunctionEntryNode()
   n2 = &RigVMFunction_Sequence(ExecuteContext=@n1.ExecuteContext)
   n3 = &RigVMFunction_Sequence(ExecuteContext=@n2.B)
   n4 = &RigVMFunction_Sequence(ExecuteContext=@n3.B)
   n5 = &RigVMFunction_Sequence(ExecuteContext=@n4.B)
   RigVMFunctionReturnNode(B=@n3.A, A=@n2.A, E=@n5.A, F=@n5.B, D=@n4.A)
  ]}
 RigVMModel {nodes [
   RigUnit_BeginExecution = &RigUnit_BeginExecution()
   RigVMFunction_Sequence = RigVMAggregateNode(ExecuteContext=@RigUnit_BeginExecution.ExecutePin)
   "HairSetup " = RigVMCollapseNode(ExecuteContext=@RigVMFunction_Sequence.A)
   n1 = &RigVMFunction_Sequence(ExecuteContext=@"HairSetup ".ExecuteContext)
   &RigUnit_ParentConstraint(
    AdvancedSettings=(InterpolationType=Shortest,RotationOrderForFilter=XYZ)
    Child=(Type=Bone,Name="base_hair_front_l_aim_tgt")
    Filter=(TranslationFilter=(bX=True,bY=True,bZ=True),RotationFilter=(bX=true,bY=true,bZ=true),ScaleFilter=(bX=True,bY=True,bZ=True))
    Parents=((Item=(Type=Bone,Name="spine_05"),Weight=1),(Item=(Type=Bone,Name="clavicle_l"),Weight=1))
    Weight=1
    ExecutePin=@n1.A
   )
   > &RigUnit_ModifyTransforms(
    ItemToModify=((Item=(Type=Bone,Name="hair_front_l_aim_tgt"),Transform=(Rotation=(X=0,Y=0,Z=0,W=1),Translation=(X=5,Y=0,Z=0.9271),Scale3D=(X=1,Y=1,Z=1))))
    WeightMinimum=0
    Weight=@"HairSetup ".Head Up
   )
   n2 = &RigVMFunction_AlphaInterp(
    InterpSpeedDecreasing=10
    InterpSpeedIncreasing=10
    Scale=1
    bClampResult=False
    Value=&RigVMFunction_MathBoolToFloat(Value=RigVMVariableNode() {variable=IsFrontEnd}.Value).Result
   )
   > &RigUnit_ModifyTransforms(
    ItemToModify=((Item=(Type=Bone,Name="hair_front_l_aim_tgt"),Transform=(Rotation=(X=0,Y=0,Z=0,W=1),Translation=(X=5,Y=0,Z=-5),Scale3D=(X=1,Y=1,Z=1))))
    WeightMinimum=0
    Weight=&RigVMFunction_MathDoubleMul(
     B=@n2.Result
     A=&RigVMFunction_MathDoubleRemap(SourceMaximum=0.5, SourceMinimum=0, TargetMaximum=1, TargetMinimum=0, Value=@"HairSetup ".Head Up).Result
    ).Result
   )
   > &RigUnit_ModifyTransforms(
    ItemToModify=((Item=(Type=Bone,Name="hair_front_l_aim_tgt"),Transform=(Rotation=(X=0,Y=0,Z=0,W=1),Translation=(X=0,Y=-10,Z=10),Scale3D=(X=1,Y=1,Z=1))))
    WeightMinimum=0
    Weight=@"HairSetup ".Head Up
   )
   > &RigUnit_AimConstraintLocalSpaceOffset(
    Child=(Type=Bone,Name="aim_hair_front_lf")
    Filter=(bX=true,bY=true,bZ=true)
    Parents=((Item=(Type=Bone,Name="hair_front_l_aim_tgt"),Weight=1))
    Weight=1
    WorldUp=(Target=(X=0,Y=0,Z=1),Kind=Location,Space=(Type=Bone,Name="hair_front_l_aim_tgt"))
   )
   > DeltaRotation_2 = RigVMFunctionReferenceNode(Child=head, Parent=head_null, Source Max=0, Source Min=0)
   > &RigUnit_ModifyTransforms(
    ItemToModify=((Item=(Type=Bone,Name="offset_hair_front_lf"),Transform=(Rotation=(X=-0.0013,Y=-0.0465,Z=-0.0276,W=0.9985),Translation=(X=-2.9005,Y=0.4992,Z=0.334),Scale3D=(X=1,Y=1,Z=1))))
    WeightMinimum=0
    Weight=@DeltaRotation_2.Delta_X_Negative
   )
   > n3 = &RigUnit_SphericalPoseReader(
    ActiveRegionScaleFactors=(PositiveWidth=1,NegativeWidth=1,PositiveHeight=1,NegativeHeight=1)
    ActiveRegionSize=0.15
    Debug=(bDrawDebug=false,bDraw2D=False,bDrawLocalAxes=False,DebugScale=25,DebugSegments=20,DebugThickness=0.25)
    DriverAxis=(X=1,Y=0,Z=0)
    DriverItem=(Type=Bone,Name="clavicle_l")
    FalloffRegionScaleFactors=(PositiveWidth=1,NegativeWidth=1,PositiveHeight=1,NegativeHeight=1)
    FalloffSize=0.25
    FlipHeightScaling=False
    FlipWidthScaling=False
    OptionalParentItem=(Type=Bone,Name="None")
    RotationOffset=(X=50,Y=0,Z=-90)
   )
   RigVMFunction_MathBoolToFloat = &RigVMFunction_MathBoolToFloat(Value=RigVMVariableNode() {variable=IsFrontEnd}.Value)
   AlphaInterp = &RigVMFunction_AlphaInterp(InterpSpeedDecreasing=10, InterpSpeedIncreasing=10, Scale=1, bClampResult=False, Value=@RigVMFunction_MathBoolToFloat.Result)
   Multiply_3 = RigVMAggregateNode(
    C=@AlphaInterp.Result
    A=&RigVMFunction_MathDoubleRemap(SourceMaximum=0.5, SourceMinimum=0, TargetMaximum=1, TargetMinimum=0, Value=@"HairSetup ".Head Up).Result
    B=@n3.OutputParam
   )
   > &RigUnit_ModifyTransforms(
    ItemToModify=((Item=(Type=Bone,Name="base_hair_front_01_lf"),Transform=(Rotation=(X=0.0855,Y=0.0465,Z=0.0309,W=0.9948),Translation=(X=0.5,Y=-1.7,Z=2.6),Scale3D=(X=1,Y=1,Z=1))))
    WeightMinimum=0
    Weight=@Multiply_3.Result
   )
   &RigUnit_ParentConstraint(
    AdvancedSettings=(InterpolationType=Average,RotationOrderForFilter=XYZ)
    Child=(Type=Bone,Name="base_hair_front_r_aim_tgt")
    Filter=(TranslationFilter=(bX=True,bY=True,bZ=True),RotationFilter=(bX=true,bY=true,bZ=true),ScaleFilter=(bX=True,bY=True,bZ=True))
    Parents=((Item=(Type=Bone,Name="spine_05"),Weight=1),(Item=(Type=Bone,Name="clavicle_r"),Weight=1))
    Weight=1
    ExecutePin=@n1.B
   )
   > &RigUnit_ModifyTransforms(
    ItemToModify=((Item=(Type=Bone,Name="offset_hair_front_rt"),Transform=(Rotation=(X=0,Y=0,Z=0,W=1),Translation=(X=-1,Y=0,Z=0),Scale3D=(X=1,Y=1,Z=1))))
    WeightMinimum=0
    Weight=@"HairSetup ".Head Up
   )
   > &RigUnit_ModifyTransforms(
    ItemToModify=((Item=(Type=Bone,Name="hair_front_r_aim_tgt"),Transform=(Rotation=(X=0,Y=0,Z=0,W=1),Translation=(X=0,Y=-10,Z=0),Scale3D=(X=1,Y=1,Z=1))))
    WeightMinimum=0
    Weight=@"HairSetup ".Head Up
   )
   > &RigUnit_ModifyTransforms(
    ItemToModify="((Item=(Type=Bone,Name=\"hair_front_r_aim_tgt\"),Transform=(Rotation=(X=0,Y=0,Z=0,W=1),Translation=(X=0,Y=-3,Z=0.9271),Scale3D=(X=1,Y=1,Z=1))),(Item=(Type=Bone,Name=\"offset_hair_front_rt\"),Transform=(Rotation=(X=0,Y=0,Z=0,W=1),Translation=(X=0,Y=1,Z=0),Scale ...(truncated)"
    Weight=1
    WeightMinimum=0
   )
   > &RigUnit_ModifyTransforms(
    ItemToModify="((Item=(Type=Bone,Name=\"hair_front_r_aim_tgt\"),Transform=(Rotation=(X=0,Y=0,Z=0,W=1),Translation=(X=0,Y=0,Z=2),Scale3D=(X=1,Y=1,Z=1))),(Item=(Type=Bone,Name=\"offset_hair_front_rt\"),Transform=(Rotation=(X=0,Y=0,Z=0,W=1),Translation=(X=0,Y=0,Z=0),Scale3D=(X= ...(truncated)"
    WeightMinimum=0
    Weight=@"HairSetup ".Head Up
   )
   > &RigUnit_AimConstraintLocalSpaceOffset(
    Child=(Type=Bone,Name="aim_hair_front_rt")
    Filter=(bX=True,bY=True,bZ=True)
    Parents=((Item=(Type=Bone,Name="hair_front_r_aim_tgt"),Weight=1))
    Weight=1
    WorldUp=(Target=(X=0,Y=0,Z=1),Kind=Direction,Space=(Type=Bone,Name="hair_front_r_aim_tgt"))
   )
   > &RigUnit_ModifyTransforms(
    ItemToModify="((Item=(Type=Bone,Name=\"offset_hair_front_rt\"),Transform=(Rotation=(X=0,Y=-0.0436,Z=0,W=0.999),Translation=(X=-2,Y=0,Z=-2),Scale3D=(X=1,Y=1,Z=1))),(Item=(Type=Bone,Name=\"dyn_hair_front_02_rt_helper\"),Transform=(Rotation=(X=0,Y=0,Z=0,W=1),Translation=(X=-2, ...(truncated)"
    WeightMinimum=0
    Weight=@"HairSetup ".Head Right
   )
   > &RigUnit_ModifyTransforms(
    ItemToModify=((Item=(Type=Bone,Name="offset_hair_front_rt"),Transform=(Rotation=(X=0,Y=0,Z=0,W=1),Translation=(X=0,Y=0,Z=0),Scale3D=(X=1,Y=1,Z=1))))
    WeightMinimum=0
    Weight=@"HairSetup ".Head Left
   )
   > n4 = &RigUnit_SphericalPoseReader(
    ActiveRegionScaleFactors=(PositiveWidth=1,NegativeWidth=1,PositiveHeight=1,NegativeHeight=1)
    ActiveRegionSize=0.15
    Debug=(bDrawDebug=false,bDraw2D=False,bDrawLocalAxes=False,DebugScale=25,DebugSegments=20,DebugThickness=0.25)
    DriverAxis=(X=1,Y=0,Z=0)
    DriverItem=(Type=Bone,Name="clavicle_l")
    FalloffRegionScaleFactors=(PositiveWidth=1,NegativeWidth=1,PositiveHeight=1,NegativeHeight=1)
    FalloffSize=0.25
    FlipHeightScaling=False
    FlipWidthScaling=False
    OptionalParentItem=(Type=Bone,Name="None")
    RotationOffset=(X=50,Y=0,Z=-90)
   )
   Remap = &RigVMFunction_MathDoubleRemap(SourceMaximum=0.75, SourceMinimum=0, TargetMaximum=1, TargetMinimum=0, Value=@"HairSetup ".Head Up)
   Multiply = &RigVMFunction_MathDoubleMul(A=@Remap.Result, B=@n4.OutputParam)
   > &RigUnit_ModifyTransforms(
    ItemToModify=((Item=(Type=Bone,Name="base_hair_front_01_rt"),Transform=(Rotation=(X=0,Y=0,Z=-0.0872,W=0.9962),Translation=(X=0.1592,Y=-2.3686,Z=1.8722),Scale3D=(X=1,Y=1,Z=1))))
    WeightMinimum=0
    Weight=@Multiply.Result
   )
   > &RigUnit_ModifyTransforms(
    ItemToModify=((Item=(Type=Bone,Name="dyn_hair_front_02_rt_helper"),Transform=(Rotation=(X=0,Y=0,Z=0,W=1),Translation=(X=0,Y=1,Z=3),Scale3D=(X=1,Y=1,Z=1))))
    WeightMinimum=0
    Weight=@"HairSetup ".Head Up
   )
   RigVMFunction_ControlFlowBranch = &RigVMFunction_ControlFlowBranch(ExecuteContext=@RigVMFunction_Sequence.C, Condition=RigVMVariableNode() {variable=IsFrontEnd}.Value)
   Get Delta Ref = RigVMFunctionReferenceNode(Item=(Type=Bone,Name="neck_01"), ExecuteContext=@RigVMFunction_ControlFlowBranch.True)
   n5 = &RigVMFunction_MathDoubleRemap(SourceMaximum=40, SourceMinimum=0, TargetMaximum=1, TargetMinimum=0, Value=@"Get Delta Ref".Result.Rotation.Yaw)
   > &RigUnit_ModifyTransforms(
    ItemToModify="((Item=(Type=Bone,Name=\"offset_hair_front_rt\"),Transform=(Rotation=(X=0,Y=0.0872,Z=0,W=0.9962),Translation=(X=0,Y=0,Z=0),Scale3D=(X=1,Y=1,Z=1))),(Item=(Type=Bone,Name=\"offset_hair_front_lf\"),Transform=(Rotation=(X=0,Y=0.1045,Z=0,W=0.9945),Translation=(X=0, ...(truncated)"
    WeightMinimum=0
    Weight=@n5.Result
   )
   > &RigUnit_ModifyTransforms(
    ItemToModify="((Item=(Type=Bone,Name=\"offset_hair_cr\"),Transform=(Rotation=(X=0,Y=-0.2588,Z=0,W=0.9659),Translation=(X=0,Y=0,Z=0),Scale3D=(X=1,Y=1,Z=1))),(Item=(Type=Bone,Name=\"offset_hair_rt\"),Transform=(Rotation=(X=0,Y=-0.1736,Z=0,W=0.9848),Translation=(X=0,Y=0,Z=0),S ...(truncated)"
    WeightMinimum=0
    Weight=@n5.Result
   )
   n6 = &RigVMFunction_AlphaInterp(
    InterpSpeedDecreasing=10
    InterpSpeedIncreasing=10
    Scale=1
    bClampResult=False
    Value=&RigVMFunction_MathBoolToFloat(Value=RigVMVariableNode() {variable=IsTargeting}.Value).Result
   )
   RigUnit_GetTransform = &RigUnit_GetTransform(Item=(Type=Bone,Name="upperarm_r"), bInitial=False)
   RigVMFunction_MathVectorDistance = &RigVMFunction_MathVectorDistance(
    B=@RigUnit_GetTransform.Transform.Translation
    A=&RigUnit_GetTransform(Item=(Type=Bone,Name="head"), bInitial=False).Transform.Translation
   )
   &RigUnit_ModifyTransforms(
    ItemToModify="((Item=(Type=Bone,Name=\"skn_hair_front_02_rt\"),Transform=(Rotation=(X=0,Y=0,Z=0,W=-1),Translation=(X=0,Y=0,Z=0),Scale3D=(X=1,Y=0.1,Z=1))),(Item=(Type=Bone,Name=\"skn_hair_front_01_rt\"),Transform=(Rotation=(X=0,Y=0,Z=0,W=1),Translation=(X=0,Y=0,Z=0),Scale3D= ...(truncated)"
    WeightMinimum=0
    Weight=&RigVMFunction_MathDoubleMul(
     A=@n6.Result
     B=&RigVMFunction_MathDoubleRemap(SourceMaximum=17, SourceMinimum=21.423, TargetMaximum=1, TargetMinimum=0, Value=@RigVMFunction_MathVectorDistance.Result).Result
    ).Result
    ExecutePin=@RigVMFunction_Sequence.B
   )
   n7 = &RigVMFunction_AlphaInterp(
    InterpSpeedDecreasing=10
    InterpSpeedIncreasing=10
    Scale=1
    bClampResult=False
    Value=&RigVMFunction_MathBoolToFloat(Value=RigVMVariableNode() {variable=IsTargeting}.Value).Result
   )
   n8 = &RigUnit_GetTransform(Item=(Type=Bone,Name="upperarm_l"), bInitial=False)
   n9 = &RigUnit_GetTransform(Item=(Type=Bone,Name="head"), bInitial=False)
   n10 = &RigVMFunction_MathDoubleMul(
    A=@n7.Result
    B=&RigVMFunction_MathDoubleRemap(
     SourceMaximum=17
     SourceMinimum=21.423
     TargetMaximum=1
     TargetMinimum=0
     Value=&RigVMFunction_MathVectorDistance(B=@n8.Transform.Translation, A=@n9.Transform.Translation).Result
    ).Result
   )
   > &RigUnit_ModifyTransforms(
    ItemToModify="((Item=(Type=Bone,Name=\"skn_hair_front_02_lf\"),Transform=(Rotation=(X=0,Y=0,Z=0,W=-1),Translation=(X=0,Y=0,Z=0),Scale3D=(X=1,Y=0.1,Z=1))),(Item=(Type=Bone,Name=\"skn_hair_front_01_lf\"),Transform=(Rotation=(X=0,Y=0,Z=0,W=1),Translation=(X=0,Y=0,Z=0),Scale3D= ...(truncated)"
    WeightMinimum=0
    Weight=@n10.Result
   )
   > &RigUnit_ModifyTransforms(
    ItemToModify="((Item=(Type=Bone,Name=\"base_hair_front_01_lf\"),Transform=(Rotation=(X=0,Y=0,Z=0,W=1),Translation=(X=0,Y=5,Z=0),Scale3D=(X=1,Y=1,Z=1))),(Item=(Type=Bone,Name=\"offset_pigtail_front_lf\"),Transform=(Rotation=(X=0,Y=0,Z=0,W=1),Translation=(X=0,Y=-5,Z=0),Scale3 ...(truncated)"
    WeightMinimum=0
    Weight=@n10.Result
   )
   > &RigUnit_ModifyTransforms(
    ItemToModify=((Item=(Type=Bone,Name="offset_pigtail_front_lf"),Transform=(Rotation=(X=0,Y=0,Z=0,W=1),Translation=(X=0,Y=-1.5,Z=0),Scale3D=(X=1,Y=1,Z=1))))
    Weight=1
    WeightMinimum=0
   )
  ]}
]
hierarchy [
 root {loc=<0 0 0>, rot=<0 0 0 1>}
 pelvis {parent=root, loc=<0 -0.7071 93.1191>, rot=<0.0127 0.707 -0.0127 -0.707>}
 spine_01 {parent=pelvis, loc=<6.7494 -0.6388 0>, rot=<0 0 0.034 -0.9994>}
 spine_02 {parent=spine_01, loc=<7.2518 -0.035 0>, rot=<0 0 0.0238 -0.9997>}
 spine_03 {parent=spine_02, loc=<6.0104 0 0>, rot=<0 0 0.0149 -0.9999>}
 spine_04 {parent=spine_03, loc=<8.1024 0 0>, rot=<0 0 -0.0132 -0.9999>}
 spine_05 {parent=spine_04, loc=<10.0146 0 0>, rot=<0 0 -0.0015 -1>}
 neck_01 {parent=spine_05, loc=<11.3138 0.1529 0>, rot=<0 0 -0.2003 -0.9797>}
 neck_02 {parent=neck_01, loc=<4.5627 0 0>, rot=<0 0 0.0504 -0.9987>}
 head {parent=neck_02, loc=<5.3914 0 0>, rot=<0 0 0.115 -0.9934>}
 base_hair_lf {parent=head, loc=<12.8179 -3.7749 -5.9095>, rot=<0.1278 -0.6908 0.7008 -0.124>}
 aim_hair_lf {parent=base_hair_lf, loc=<0 0 0>, rot=<0 0 0 1>}
 offset_hair_lf {parent=aim_hair_lf, loc=<0 0 0>, rot=<0 0 0 1>}
 dyn_hair_01_lf {parent=offset_hair_lf, loc=<0 0 0>, rot=<0 0 0 1>}
 dyn_hair_02_lf {parent=dyn_hair_01_lf, loc=<13.5092 0 0>, rot=<0 0.093 0 -0.9957>}
 dyn_hair_03_lf {parent=dyn_hair_02_lf, loc=<7.0701 0 0>, rot=<0 -0.0342 0 -0.9994>}
 dyn_hair_04_lf {parent=dyn_hair_03_lf, loc=<7.0629 0 0>, rot=<0 -0.0514 0 -0.9987>}
 dyn_hair_05_lf {parent=dyn_hair_04_lf, loc=<7.0605 0 0>, rot=<0 -0.1054 0 -0.9944>}
 base_pigtail_B_lf {parent=dyn_hair_03_lf, loc=<0.1402 1.1166 2.7717>, rot=<-0.0061 -0.0341 0.0337 -0.9988>}
 aim_pigtail_B_lf {parent=base_pigtail_B_lf, loc=<0 0 0>, rot=<0 0 0 1>}
 offset_pigtail_B_lf {parent=aim_pigtail_B_lf, loc=<0 0 0>, rot=<0 0 0 1>}
 dyn_pigtail_01_B_lf {parent=offset_pigtail_B_lf, loc=<0 0 0>, rot=<0 0 0 1>}
 dyn_pigtail_02_B_lf {parent=dyn_pigtail_01_B_lf, loc=<4 0 0>, rot=<0 0.0579 0 -0.9983>}
 dyn_pigtail_03_B_lf {parent=dyn_pigtail_02_B_lf, loc=<4 0 0>, rot=<0 0 0 1>}
 base_pigtail_A_lf {parent=dyn_hair_02_lf, loc=<0.1743 4.842 1.163>, rot=<0.009 -0.0061 -0.1238 -0.9922>}
 aim_pigtail_A_lf {parent=base_pigtail_A_lf, loc=<0 0 0>, rot=<0 0 0 1>}
 offset_pigtail_A_lf {parent=aim_pigtail_A_lf, loc=<0 0 0>, rot=<0 0 0 1>}
 dyn_pigtail_01_A_lf {parent=offset_pigtail_A_lf, loc=<0 0 0>, rot=<0 0 0 1>}
 dyn_pigtail_02_A_lf {parent=dyn_pigtail_01_A_lf, loc=<5 0 0>, rot=<0 -0.1654 0 -0.9862>}
 dyn_pigtail_03_A_lf {parent=dyn_pigtail_02_A_lf, loc=<5 0 0>, rot=<0 -0.1185 0 -0.9929>}
 base_hair_cr {parent=head, loc=<13.1825 -4.1457 -0.0414>, rot=<0.1394 -0.6833 0.7037 -0.1355>}
 aim_hair_cr {parent=base_hair_cr, loc=<0 0 0>, rot=<0 0 0 1>}
 offset_hair_cr {parent=aim_hair_cr, loc=<0 0 0>, rot=<0 0 0 1>}
 dyn_hair_01_cr {parent=offset_hair_cr, loc=<0 0 0>, rot=<0 0 0 1>}
 dyn_hair_02_cr {parent=dyn_hair_01_cr, loc=<13.7912 0 0>, rot=<0 0.0489 0 -0.9988>}
 dyn_hair_03_cr {parent=dyn_hair_02_cr, loc=<7.0701 0 0>, rot=<0 0.0189 0 -0.9998>}
 dyn_hair_04_cr {parent=dyn_hair_03_cr, loc=<7.0629 0 0>, rot=<0 -0.0833 0 -0.9965>}
 dyn_hair_05_cr {parent=dyn_hair_04_cr, loc=<7.0605 0 0>, rot=<0 -0.1054 0 -0.9944>}
 base_hair_rt {parent=head, loc=<12.8499 -3.6016 5.791>, rot=<0.1278 -0.6908 0.7008 -0.124>}
 aim_hair_rt {parent=base_hair_rt, loc=<0 0 0>, rot=<0 0 0 1>}
 offset_hair_rt {parent=aim_hair_rt, loc=<0 0 0>, rot=<0 0 0 1>}
 dyn_hair_01_rt {parent=offset_hair_rt, loc=<0 0 0>, rot=<0 0 0 1>}
 dyn_hair_02_rt {parent=dyn_hair_01_rt, loc=<13.5092 0 0>, rot=<0 0.093 0 -0.9957>}
 dyn_hair_03_rt {parent=dyn_hair_02_rt, loc=<7.0701 0 0>, rot=<0 -0.0342 0 -0.9994>}
 dyn_hair_04_rt {parent=dyn_hair_03_rt, loc=<7.0629 0 0>, rot=<0 -0.0514 0 -0.9987>}
 dyn_hair_05_rt {parent=dyn_hair_04_rt, loc=<7.0605 0 0>, rot=<0 -0.1054 0 -0.9944>}
 base_pigtail_cr {parent=dyn_hair_02_rt, loc=<2.1794 0.7694 4.4746>, rot=<-0.007 -0.0148 0.0024 -0.9999>}
 aim_pigtail_cr {parent=base_pigtail_cr, loc=<0 0 0>, rot=<0 0 0 1>}
 offset_pigtail_cr {parent=aim_pigtail_cr, loc=<0 0 0>, rot=<0 0 0 1>}
 dyn_pigtail_01_cr {parent=offset_pigtail_cr, loc=<0 0 0>, rot=<0 0 0 1>}
 dyn_pigtail_02_cr {parent=dyn_pigtail_01_cr, loc=<7.0359 0 0>, rot=<0 0 -0.2104 -0.9776>}
 dyn_pigtail_03_cr {parent=dyn_pigtail_02_cr, loc=<5 0 0>, rot=<0 0 0.1069 -0.9943>}
 base_hair_02_rt_helper {parent=dyn_hair_02_rt, loc=<4.1056 -5.6431 -2.7638>, rot=<0 0 0 1>}
 dyn_hair_02_rt_helper {parent=base_hair_02_rt_helper, loc=<0 0 0>, rot=<0 0.1499 0 -0.9887>}
 base_hair_front_01_rt {parent=head, loc=<5.7485 1.7418 9.345>, rot=<0.0264 -0.7126 -0.7005 0.0291>}
 aim_hair_front_rt {parent=base_hair_front_01_rt, loc=<0 0 0>, rot=<0 0 0 -1>}
 offset_hair_front_rt {parent=aim_hair_front_rt, loc=<0 0 0>, rot=<0 0 0 -1>}
 dyn_hair_front_01_rt {parent=offset_hair_front_rt, loc=<0 0 0>, rot=<0.0236 0.0082 -0.0002 -0.9997>}
 dyn_hair_front_02_rt {parent=dyn_hair_front_01_rt, loc=<10.7479 -0.0165 -0.0008>, rot=<-0.0217 0.3796 0.0093 -0.9248>}
 dyn_hair_front_03_rt {parent=dyn_hair_front_02_rt, loc=<7.561 0 0.0001>, rot=<0 0 0 -1>}
 dyn_hair_front_04_rt {parent=dyn_hair_front_03_rt, loc=<6.3721 0 0>, rot=<0 -0.1305 0 -0.9914>}
 skn_hair_front_04_rt {parent=dyn_hair_front_04_rt, loc=<0 0 0>, rot=<0 0 0 1>}
 dyn_hair_front_05_rt {parent=dyn_hair_front_04_rt, loc=<6.372 0 0>, rot=<0 0 0 1>}
 skn_hair_front_05_rt {parent=dyn_hair_front_05_rt, loc=<0 0 0>, rot=<0 0 0 1>}
 skn_hair_front_03_rt {parent=dyn_hair_front_03_rt, loc=<0 0 0>, rot=<0 0 0 1>}
 base_pigtail_front_rt {parent=dyn_hair_front_03_rt, loc=<6.0444 2.7026 0.9231>, rot=<0.0045 -0.2088 -0.0213 -0.9777>}
 aim_pigtail_front_rt {parent=base_pigtail_front_rt, loc=<0 0 0>, rot=<0 0 0 -1>}
 offset_pigtail_front_rt {parent=aim_pigtail_front_rt, loc=<0 0 0>, rot=<0 0 0 -1>}
 dyn_pigtail_front_01_rt {parent=offset_pigtail_front_rt, loc=<0 0 0>, rot=<0.0236 0.0082 -0.0002 -0.9997>}
 skn_pigtail_front_01_rt {parent=dyn_pigtail_front_01_rt, loc=<0 0 0>, rot=<0 0 0 1>}
 dyn_pigtail_front_02_rt {parent=dyn_pigtail_front_01_rt, loc=<3.9662 0 0>, rot=<-0.0235 -0.0819 -0.0015 -0.9964>}
 skn_pigtail_front_02_rt {parent=dyn_pigtail_front_02_rt, loc=<0 0 0>, rot=<0 0 0 1>}
 dyn_pigtail_front_03_rt {parent=dyn_pigtail_front_02_rt, loc=<5.9845 0 0>, rot=<0 0 0 1>}
 skn_pigtail_front_03_rt {parent=dyn_pigtail_front_03_rt, loc=<0 0 0>, rot=<0 0 0 1>}
 skn_hair_front_02_rt {parent=dyn_hair_front_02_rt, loc=<0 0 0>, rot=<0 0 0 1>}
 dyn_pigtail_front_04_rt {parent=dyn_hair_front_02_rt, loc=<2.2619 2.0006 0.4281>, rot=<0.0364 0.0287 -0.1881 -0.9811>}
 skn_pigtail_front_04_rt {parent=dyn_pigtail_front_04_rt, loc=<0 0 0>, rot=<0 0 0 1>}
 dyn_pigtail_front_05_rt {parent=dyn_pigtail_front_04_rt, loc=<6.6859 0 0>, rot=<-0.0229 -0.2597 -0.0058 -0.9654>}
 skn_pigtail_front_05_rt {parent=dyn_pigtail_front_05_rt, loc=<0 0 0>, rot=<0 0 0 1>}
 base_hair_front_02_rt_helper {parent=dyn_hair_front_02_rt, loc=<-0.0222 3.1059 -3.6441>, rot=<0 0 0 1>}
 dyn_hair_front_02_rt_helper {parent=base_hair_front_02_rt_helper, loc=<0 0 0>, rot=<0 0.1499 0 -0.9887>}
 skn_hair_front_01_rt {parent=dyn_hair_front_01_rt, loc=<0 0 0>, rot=<0 0 0 1>}
 FACIAL_C_FacialRoot {parent=head, loc=<2.1524 -0.746 -0.0002>, rot=<0.4879 -0.5118 -0.4879 -0.5118>}
 FACIAL_C_Forehead {parent=FACIAL_C_FacialRoot, loc=<-0.003 -2.9059 10.1375>, rot=<-0.0404 0.0001 -0.0001 -0.9992>}
 FACIAL_L_ForeheadIn {parent=FACIAL_C_FacialRoot, loc=<1.4909 -3.0107 9.9913>, rot=<-0.0577 -0.0695 0.0043 -0.9959>}
 FACIAL_R_ForeheadIn {parent=FACIAL_C_FacialRoot, loc=<-1.4896 -2.9951 9.9916>, rot=<-0.0577 0.0697 -0.0046 -0.9959>}
 FACIAL_L_ForeheadMid {parent=FACIAL_C_FacialRoot, loc=<2.8361 -3.3662 9.4249>, rot=<-0.068 -0.1756 -0.0202 -0.9819>}
 FACIAL_R_ForeheadMid {parent=FACIAL_C_FacialRoot, loc=<-2.8295 -3.3605 9.4229>, rot=<-0.0681 0.1758 0.0199 -0.9819>}
 FACIAL_L_ForeheadOut {parent=FACIAL_C_FacialRoot, loc=<4.0941 -3.2738 8.6376>, rot=<-0.1185 -0.2557 -0.0785 -0.9562>}
 FACIAL_R_ForeheadOut {parent=FACIAL_C_FacialRoot, loc=<-4.0812 -3.2735 8.6288>, rot=<-0.1186 0.2559 0.0782 -0.9562>}
 FACIAL_L_CheekOuter {parent=FACIAL_C_FacialRoot, loc=<3.7366 0.4781 6.8567>, rot=<0.058 -0.3765 0.1014 -0.919>}
 FACIAL_R_CheekOuter {parent=FACIAL_C_FacialRoot, loc=<-3.7065 0.49 6.9121>, rot=<0.0579 0.3768 -0.1016 -0.9189>}
 FACIAL_L_NasolabialBulge {parent=FACIAL_C_FacialRoot, loc=<3.7496 3.6813 6.7091>, rot=<0.0607 -0.3771 0.0248 -0.9238>}
 FACIAL_R_NasolabialBulge {parent=FACIAL_C_FacialRoot, loc=<-3.7815 3.6863 6.7215>, rot=<0.0606 0.3774 -0.0251 -0.9237>}
 FACIAL_L_NasolabialFurrow {parent=FACIAL_C_FacialRoot, loc=<3.2666 2.5407 8.8836>, rot=<-0.0304 -0.1268 -0.3634 -0.9225>}
 FACIAL_R_NasolabialFurrow {parent=FACIAL_C_FacialRoot, loc=<-3.2521 2.5435 8.93>, rot=<-0.0306 0.127 0.3631 -0.9225>}
 FACIAL_L_CheekLower {parent=FACIAL_C_FacialRoot, loc=<4.184 3.4841 6.0958>, rot=<0.0373 -0.4266 0.0589 -0.9018>}
 FACIAL_R_CheekLower {parent=FACIAL_C_FacialRoot, loc=<-4.21 3.4901 6.091>, rot=<0.0372 0.4268 -0.0591 -0.9016>}
 FACIAL_C_Nose {parent=FACIAL_C_FacialRoot, loc=<-0.023 -0.3106 10.1971>, rot=<-0.1274 0.0001 -0.0001 -0.9919>}
 FACIAL_C_NoseLower {parent=FACIAL_C_Nose, loc=<-0.0316 2.5574 -0.3068>, rot=<0.1305 0 0 -0.9914>}
 FACIAL_L_Nostril {parent=FACIAL_C_Nose, loc=<0.8495 1.6771 -0.5323>, rot=<0.0173 0.1305 0.0023 -0.9913>}
 FACIAL_R_Nostril {parent=FACIAL_C_Nose, loc=<-0.9434 1.5648 -0.5354>, rot=<0.0173 -0.1305 -0.0023 -0.9913>}
 FACIAL_C_NoseTip {parent=FACIAL_C_Nose, loc=<-0.0075 1.8928 0.476>, rot=<0.0872 0 0 -0.9962>}
 FACIAL_C_MouthUpper {parent=FACIAL_C_FacialRoot, loc=<-0.0479 3.9832 7.7245>, rot=<0.003 0.0001 -0.0001 -1>}
 FACIAL_C_LipUpper {parent=FACIAL_C_MouthUpper, loc=<-0.0086 -0.6106 2.4105>, rot=<-0.0872 0 0 -0.9962>}
 FACIAL_L_LipUpper {parent=FACIAL_C_MouthUpper, loc=<1.3632 -0.4795 2.0413>, rot=<-0.0778 -0.1601 -0.0379 -0.9833>}
 FACIAL_R_LipUpper {parent=FACIAL_C_MouthUpper, loc=<-1.4417 -0.48 2.0139>, rot=<-0.0778 0.1601 0.0379 -0.9833>}
 FACIAL_L_LipUpperOuter {parent=FACIAL_C_MouthUpper, loc=<2.2022 -0.0457 1.3583>, rot=<-0.0984 -0.2293 -0.0983 -0.9634>}
 FACIAL_R_LipUpperOuter {parent=FACIAL_C_MouthUpper, loc=<-2.3128 -0.0315 1.3508>, rot=<-0.0984 0.2293 0.0983 -0.9634>}
 FACIAL_L_LipCorner {parent=FACIAL_C_MouthUpper, loc=<2.6473 0.3257 0.732>, rot=<0 -0.2078 0 -0.9782>}
 FACIAL_R_LipCorner {parent=FACIAL_C_MouthUpper, loc=<-2.767 0.3539 0.7468>, rot=<0 0.2078 0 -0.9782>}
 FACIAL_L_Masseter {parent=FACIAL_C_FacialRoot, loc=<6.0087 1.139 4.6584>, rot=<0.0723 0.1131 0.0074 -0.9909>}
 FACIAL_R_Masseter {parent=FACIAL_C_FacialRoot, loc=<-5.9974 1.1313 4.6649>, rot=<0.0723 -0.1128 -0.0077 -0.991>}
 FACIAL_C_UnderChin {parent=FACIAL_C_FacialRoot, loc=<0.0012 8.8986 5.9482>, rot=<0.5025 0.0002 0 -0.8646>}
 FACIAL_L_UnderChin {parent=FACIAL_C_FacialRoot, loc=<3.9935 7.2708 3.8341>, rot=<0.4051 -0.3544 0.0833 -0.8386>}
 FACIAL_R_UnderChin {parent=FACIAL_C_FacialRoot, loc=<-3.9945 7.2289 3.8228>, rot=<0.405 0.3548 -0.0834 -0.8385>}
 FACIAL_C_TeethUpper {parent=FACIAL_C_FacialRoot, loc=<-0.0578 4.2155 9.1616>, rot=<0.003 0.0001 -0.0001 -1>}
 FACIAL_C_LowerLipRotation {parent=FACIAL_C_FacialRoot, loc=<0 1.9789 2.9513>, rot=<0.3449 0.0002 -0.0001 -0.9387>}
 FACIAL_C_MouthLower {parent=FACIAL_C_LowerLipRotation, loc=<-0.0461 -1.5626 4.9354>, rot=<-0.342 0 0 -0.9397>}
 FACIAL_C_LipLower {parent=FACIAL_C_MouthLower, loc=<0.0298 0.8622 2.3218>, rot=<0.0872 0 0 -0.9962>}
 FACIAL_L_LipLower {parent=FACIAL_C_MouthLower, loc=<1.2229 0.8406 2.0975>, rot=<0.0927 -0.1294 0.0272 -0.9869>}
 FACIAL_R_LipLower {parent=FACIAL_C_MouthLower, loc=<-1.23 0.8356 2.0795>, rot=<0.0927 0.1294 -0.0272 -0.9869>}
 FACIAL_L_LipLowerOuter {parent=FACIAL_C_MouthLower, loc=<2.0655 0.5268 1.3633>, rot=<0.0913 -0.2321 0.0318 -0.9679>}
 FACIAL_R_LipLowerOuter {parent=FACIAL_C_MouthLower, loc=<-2.1672 0.5493 1.3277>, rot=<0.0913 0.2321 -0.0318 -0.9679>}
 FACIAL_C_TeethLower {parent=FACIAL_C_LowerLipRotation, loc=<-0.056 -2.2592 6.1256>, rot=<-0.342 0 0 -0.9397>}
 FACIAL_C_Tongue1 {parent=FACIAL_C_TeethLower, loc=<0.0472 0.1934 -5.1481>, rot=<-0.2588 0 0 -0.9659>}
 FACIAL_C_Tongue2 {parent=FACIAL_C_Tongue1, loc=<-0.0104 0.4521 1.4917>, rot=<0.2588 0 0 -0.9659>}
 FACIAL_C_Tongue3 {parent=FACIAL_C_Tongue2, loc=<-0.0137 0.1965 1.3835>, rot=<0.1305 0 0 -0.9914>}
 FACIAL_C_Tongue4 {parent=FACIAL_C_Tongue3, loc=<-0.013 -0.2151 1.1363>, rot=<0.0436 0 0 -0.999>}
 FACIAL_C_TongueUpper3 {parent=FACIAL_C_Tongue3, loc=<-0.0007 -0.3952 -0.1229>, rot=<0 0 0 1>}
 FACIAL_C_TongueLower3 {parent=FACIAL_C_Tongue3, loc=<0.0045 0.8098 0.1603>, rot=<0 0 0 1>}
 FACIAL_L_TongueSide3 {parent=FACIAL_C_Tongue3, loc=<1.7197 0.0937 0.223>, rot=<0 0 0 1>}
 FACIAL_R_TongueSide3 {parent=FACIAL_C_Tongue3, loc=<-1.7332 0.1007 0.2218>, rot=<0 0 0 1>}
 FACIAL_L_TongueSide2 {parent=FACIAL_C_Tongue2, loc=<1.959 0.2448 0.518>, rot=<0 0 0 1>}
 FACIAL_C_TongueUpper2 {parent=FACIAL_C_Tongue2, loc=<-0.0009 -0.5697 0.1334>, rot=<0 0 0 1>}
 FACIAL_R_TongueSide2 {parent=FACIAL_C_Tongue2, loc=<-1.9641 0.2444 0.5156>, rot=<0 0 0 1>}
 FACIAL_C_Jaw {parent=FACIAL_C_FacialRoot, loc=<0 1.9789 2.9513>, rot=<0.3449 0.0002 -0.0001 -0.9387>}
 FACIAL_C_Jawline {parent=FACIAL_C_Jaw, loc=<-0.0012 2.3126 8.111>, rot=<0.2586 0 0 -0.966>}
 FACIAL_L_ChinSide {parent=FACIAL_C_Jaw, loc=<3.5329 0.3977 7.0902>, rot=<-0.2208 -0.3738 0.0499 -0.8995>}
 FACIAL_R_ChinSide {parent=FACIAL_C_Jaw, loc=<-3.5546 0.424 7.0937>, rot=<-0.2208 0.3738 -0.0499 -0.8995>}
 FACIAL_C_Chin {parent=FACIAL_C_Jaw, loc=<-0.0089 0.682 7.8311>, rot=<-0.2591 0 0 -0.9659>}
 FACIAL_L_Ear {parent=FACIAL_C_FacialRoot, loc=<6.7369 -0.5734 1.7777>, rot=<-0.0233 -0.0006 -0.0263 -0.9994>}
 FACIAL_R_Ear {parent=FACIAL_C_FacialRoot, loc=<-6.741 -0.5765 1.7704>, rot=<-0.0233 0.0009 0.026 -0.9994>}
 FACIAL_L_Jawline {parent=FACIAL_C_FacialRoot, loc=<4.7228 5.0723 4.9873>, rot=<0.2122 0.1577 -0.0552 -0.9628>}
 FACIAL_R_Jawline {parent=FACIAL_C_FacialRoot, loc=<-4.7107 5.0436 4.9872>, rot=<0.2122 -0.1574 0.055 -0.9629>}
 FACIAL_L_EyesackUpper {parent=FACIAL_C_FacialRoot, loc=<2.9129 -2.2825 9.8286>, rot=<0.0518 -0.0843 -0.0644 -0.993>}
 FACIAL_R_EyesackUpper {parent=FACIAL_C_FacialRoot, loc=<-2.9174 -2.2867 9.8364>, rot=<0.0518 0.0846 0.0641 -0.993>}
 FACIAL_L_EyelidUpperFurrow {parent=FACIAL_C_FacialRoot, loc=<2.6137 -2.1698 9.5584>, rot=<-0.0271 -0.1509 -0.0793 -0.985>}
 FACIAL_R_EyelidUpperFurrow {parent=FACIAL_C_FacialRoot, loc=<-2.65 -2.1748 9.5651>, rot=<-0.0271 0.1512 0.0791 -0.985>}
 FACIAL_L_EyelidUpperA {parent=FACIAL_C_FacialRoot, loc=<2.6127 -1.7761 8.1751>, rot=<-0.1449 0.0001 -0.0002 -0.9895>}
 FACIAL_R_EyelidUpperA {parent=FACIAL_C_FacialRoot, loc=<-2.6156 -1.7709 8.1817>, rot=<-0.1449 0.0001 -0.0002 -0.9895>}
 FACIAL_L_Eye {parent=FACIAL_C_FacialRoot, loc=<2.6065 -1.7611 8.1765>, rot=<0.003 0.0001 -0.0001 -1>}
 FACIAL_R_Eye {parent=FACIAL_C_FacialRoot, loc=<-2.6088 -1.7612 8.1802>, rot=<0.003 0.0001 -0.0001 -1>}
 FACIAL_L_EyelidLowerA {parent=FACIAL_C_FacialRoot, loc=<2.6127 -1.7761 8.1751>, rot=<0.2869 0.0002 -0.0001 -0.958>}
 FACIAL_R_EyelidLowerA {parent=FACIAL_C_FacialRoot, loc=<-2.6156 -1.7709 8.1817>, rot=<0.2869 0.0002 -0.0001 -0.958>}
 FACIAL_L_EyeCornerInner {parent=FACIAL_C_FacialRoot, loc=<1.7071 -1.4956 8.9113>, rot=<0.003 0.0001 -0.0001 -1>}
 FACIAL_R_EyeCornerInner {parent=FACIAL_C_FacialRoot, loc=<-1.753 -1.4782 8.9064>, rot=<0.003 0.0001 -0.0001 -1>}
 FACIAL_L_EyeCornerOuter {parent=FACIAL_C_FacialRoot, loc=<3.6919 -1.6493 8.7099>, rot=<0.0028 -0.3825 0.001 -0.924>}
 FACIAL_R_EyeCornerOuter {parent=FACIAL_C_FacialRoot, loc=<-3.7355 -1.6865 8.6318>, rot=<0.0027 0.3827 -0.0013 -0.9239>}
 FACIAL_L_EyesackLower {parent=FACIAL_C_FacialRoot, loc=<2.7877 -0.4877 8.3544>, rot=<-0.0193 -0.1396 -0.022 -0.9898>}
 FACIAL_R_EyesackLower {parent=FACIAL_C_FacialRoot, loc=<-2.8323 -0.5109 8.369>, rot=<-0.0194 0.1399 0.0217 -0.9897>}
 FACIAL_L_CheekInner {parent=FACIAL_C_FacialRoot, loc=<1.9543 0.2209 8.4497>, rot=<-0.1203 -0.0885 -0.197 -0.969>}
 FACIAL_R_CheekInner {parent=FACIAL_C_FacialRoot, loc=<-2.0095 0.1913 8.4598>, rot=<-0.1204 0.0887 0.1967 -0.969>}
 FACIAL_C_NoseBridge {parent=FACIAL_C_FacialRoot, loc=<-0.0003 -1.4116 11.0753>, rot=<0.0468 0.0001 -0.0001 -0.9989>}
 FACIAL_L_EyelidUpperB {parent=FACIAL_C_FacialRoot, loc=<2.6127 -1.7761 8.1751>, rot=<-0.222 0.0001 -0.0002 -0.975>}
 FACIAL_L_EyelidLowerB {parent=FACIAL_C_FacialRoot, loc=<2.6127 -1.7761 8.1751>, rot=<0.3201 0.0002 -0.0001 -0.9474>}
 FACIAL_R_EyelidLowerB {parent=FACIAL_C_FacialRoot, loc=<-2.6156 -1.7709 8.1817>, rot=<0.3284 0.0002 -0.0001 -0.9445>}
 FACIAL_R_EyelidUpperB {parent=FACIAL_C_FacialRoot, loc=<-2.6156 -1.7709 8.1817>, rot=<-0.2221 0.0001 -0.0002 -0.975>}
 base_hair_front_01_lf {parent=head, loc=<5.7485 1.7418 -9.3451>, rot=<0.0264 -0.7126 -0.7005 0.0291>}
 aim_hair_front_lf {parent=base_hair_front_01_lf, loc=<0 0 0>, rot=<0 0 0 1>}
 offset_hair_front_lf {parent=aim_hair_front_lf, loc=<0 0 0>, rot=<0 0 0 1>}
 dyn_hair_front_01_lf {parent=offset_hair_front_lf, loc=<0 0 0>, rot=<0 0 0 1>}
 dyn_hair_front_02_lf {parent=dyn_hair_front_01_lf, loc=<10.7465 -0.0083 0.1751>, rot=<0 0.3873 0 -0.922>}
 dyn_hair_front_03_lf {parent=dyn_hair_front_02_lf, loc=<7.5611 0 0>, rot=<0 0 0 1>}
 dyn_hair_front_04_lf {parent=dyn_hair_front_03_lf, loc=<6.372 0 0>, rot=<0 -0.1305 0 -0.9914>}
 dyn_pigtail_front_04_lf {parent=dyn_hair_front_04_lf, loc=<0.4355 2.5434 0.021>, rot=<0 -0.1031 0 -0.9947>}
 dyn_pigtail_front_05_lf {parent=dyn_pigtail_front_04_lf, loc=<6.372 0.3009 0>, rot=<0 0 0 1>}
 skn_hair_front_04_lf {parent=dyn_hair_front_04_lf, loc=<0 0 0>, rot=<0 0 0 1>}
 dyn_hair_front_05_lf {parent=dyn_hair_front_04_lf, loc=<6.372 0 0>, rot=<0 0 0 1>}
 skn_hair_front_05_lf {parent=dyn_hair_front_05_lf, loc=<0 0 0>, rot=<0 0 0 1>}
 skn_hair_front_03_lf {parent=dyn_hair_front_03_lf, loc=<0 0 0>, rot=<0 0 0 1>}
 base_pigtail_front_01_lf {parent=dyn_hair_front_02_lf, loc=<0 3.2593 1.3654>, rot=<0 -0.149 0 -0.9888>}
 aim_pigtail_front_lf {parent=base_pigtail_front_01_lf, loc=<0 0 0>, rot=<0 0 0 1>}
 offset_pigtail_front_lf {parent=aim_pigtail_front_lf, loc=<0 0 0>, rot=<0 0 0 1>}
 dyn_pigtail_front_01_lf {parent=offset_pigtail_front_lf, loc=<0 0 0>, rot=<0 0 0 1>}
 dyn_pigtail_front_02_lf {parent=dyn_pigtail_front_01_lf, loc=<5.147 1.3684 0.4498>, rot=<0 0.0297 0 -0.9996>}
 dyn_pigtail_front_03_lf {parent=dyn_pigtail_front_02_lf, loc=<5 0 0>, rot=<0 0 0 1>}
 skn_hair_front_02_lf {parent=dyn_hair_front_02_lf, loc=<0 0 0>, rot=<0 0 0 1>}
 skn_hair_front_01_lf {parent=dyn_hair_front_01_lf, loc=<0 0 0>, rot=<0 0 0 1>}
 head_null {parent=neck_02, type=Null, loc=<5.3914 0 0>, rot=<0 0 0.115 -0.9934>}
 clavicle_l {parent=spine_05, loc=<4.3575 -0.1466 -2.3223>, rot=<0.0465 -0.7289 -0.009 -0.6829>}
 upperarm_l {parent=clavicle_l, loc=<15.5111 0 0>, rot=<0.0226 -0.3234 -0.0088 -0.9459>}
 upperarm_l_null {parent=upperarm_l, type=Null, loc=<0 0 0>, rot=<0 0 0 1>}
 upperarm_twist_01_l {parent=upperarm_l, loc=<0.0041 0 0>, rot=<0 0 0 1>}
 deform_upperarm_bk_base_l {parent=upperarm_twist_01_l, loc=<1.0813 -7.2193 -1.9453>, rot=<0 0 0 1>}
 deform_upperarm_bk_l {parent=deform_upperarm_bk_base_l, loc=<0 0 0>, rot=<0 0 0 1>}
 deform_upperarm_fr_base_l {parent=upperarm_twist_01_l, loc=<-0.3691 7.0506 -2.765>, rot=<0 0 0 1>}
 deform_upperarm_fr_l {parent=deform_upperarm_fr_base_l, loc=<0 0 0>, rot=<0 0 0 1>}
 deform_upperarm_up_base_l {parent=upperarm_twist_01_l, loc=<-3.2216 -1.1391 3.3311>, rot=<0 0 0 1>}
 deform_upperarm_up_l {parent=deform_upperarm_up_base_l, loc=<0 0 0>, rot=<0 0 0 1>}
 clavicle_l_null {parent=clavicle_l, type=Null, loc=<0 0 0>, rot=<0 0 0 1>}
 clavicle_r {parent=spine_05, loc=<4.3575 -0.1466 2.3223>, rot=<0.7289 0.0465 -0.6829 0.009>}
 upperarm_r {parent=clavicle_r, loc=<-15.5111 0 0>, rot=<0.0226 -0.3234 -0.0088 -0.9459>}
 upperarm_r_null {parent=upperarm_r, type=Null, loc=<0 0 0>, rot=<0 0 0 1>}
 upperarm_twist_01_r {parent=upperarm_r, loc=<-0.0041 0 0>, rot=<0 0 0 1>}
 deform_upperarm_bk_base_r {parent=upperarm_twist_01_r, loc=<-1.0814 7.2192 1.9456>, rot=<0 0 0 1>}
 deform_upperarm_bk_r {parent=deform_upperarm_bk_base_r, loc=<0 0 0>, rot=<0 0 0 1>}
 deform_upperarm_fr_base_r {parent=upperarm_twist_01_r, loc=<0.3689 -7.0506 2.7652>, rot=<0 0 0 1>}
 deform_upperarm_fr_r {parent=deform_upperarm_fr_base_r, loc=<0 0 0>, rot=<0 0 0 1>}
 deform_upperarm_up_base_r {parent=upperarm_twist_01_r, loc=<3.2214 1.1391 -3.3308>, rot=<0 0 0 1>}
 deform_upperarm_up_r {parent=deform_upperarm_up_base_r, loc=<0 0 0>, rot=<0 0 0 -1>}
 clavicle_r_null {parent=clavicle_r, type=Null, loc=<0 0 0>, rot=<0 0 0 1>}
 base_head_aim_tgt {parent=spine_05, loc=<9.0029 -16.7403 0.2354>, rot=<0.0075 -0.0008 0.0004 -1>}
 head_aim_tgt {parent=base_head_aim_tgt, loc=<0 0 0>, rot=<0 0 0 -1>}
 base_hair_front_r_aim_tgt {parent=spine_05, loc=<-0.7279 12.698 9.2939>, rot=<0.2599 0.639 0.707 0.156>}
 hair_front_r_aim_tgt {parent=base_hair_front_r_aim_tgt, loc=<0 0 0>, rot=<0 0 0 1>}
 base_hair_front_l_aim_tgt {parent=spine_05, loc=<-0.7283 12.698 -9.2939>, rot=<-0.156 -0.707 -0.639 -0.2599>}
 hair_front_l_aim_tgt {parent=base_hair_front_l_aim_tgt, loc=<0 0 0>, rot=<0 0 0 1>}
 deform_pec_aim_base_l {parent=spine_04, loc=<7.8732 12.6388 -2.4123>, rot=<0.0102 -0.4602 0.3979 -0.7936>}
 deform_pec_aim_l {parent=deform_pec_aim_base_l, loc=<0 0 0>, rot=<0 0 0 1>}
 deform_pec_base_l {parent=deform_pec_aim_l, loc=<10.766 0 0>, rot=<0 0 0 1>}
 deform_pec_l {parent=deform_pec_base_l, loc=<0.055 0 0>, rot=<0 0 0 1>}
 deform_pec_aim_base_r {parent=spine_04, loc=<7.8727 12.6388 2.4123>, rot=<-0.4602 -0.0102 0.7936 0.3979>}
 deform_pec_aim_r {parent=deform_pec_aim_base_r, loc=<0 0 0>, rot=<0 0 0 1>}
 deform_pec_base_r {parent=deform_pec_aim_r, loc=<-10.7662 -0.0006 -0.0007>, rot=<0 0 0 1>}
 deform_pec_r {parent=deform_pec_base_r, loc=<0 0 0>, rot=<0 0 0 1>}
]
variables [
 Head Down {type=double, default=0}
 Head Left {type=double, default=0}
 Head Right {type=double, default=0}
 Head Up {type=double, default=0}
 Head Up OverShoot {type=double, default=0}
 IsFrontEnd {type=bool, default=False}
 IsTargeting {type=bool, default=False}
 isWearingBackblingCape {type=bool, default=False}
 swingPickaxe {type=bool, default=False}
]
