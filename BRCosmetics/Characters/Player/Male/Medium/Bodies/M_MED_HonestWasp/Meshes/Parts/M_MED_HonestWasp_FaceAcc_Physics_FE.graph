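# Physics asset for the HonestWasp male character's FaceAcc mesh part, defining capsule bodies and constraints for dynamic hair bones.
# Three five-bone chains (center, left, right) and two four-bone front chains each hang from an offset_hair anchor.
# Swing limits generally loosen toward the tips, and most joints lock twist.
# Larger capsules on head, neck, upper spine, clavicles and upper arms form an unconstrained body volume around the hair.
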PhysicsAsset /BRCosmetics/Characters/Player/Male/Medium/Bodies/M_MED_HonestWasp/Meshes/Parts/M_MED_HonestWasp_FaceAcc_Physics_FE
bodies {
 clavicle_l [{
   shapeType=Capsule
   length=8.4228
   radius=9.7886
   shapeName=clavicle_l_capsule
   center=<6.3703 -0.6819 -1.6719>
   rotation=<rot 90 -179.9999 0>
  }]
 clavicle_r [{
   shapeType=Capsule
   length=8.7873
   radius=8.9241
   shapeName=clavicle_r_capsule
   center=<-6.1862 1.4739 1.2036>
   rotation=<rot 87.5685 -179.9942 -179.9944>
  }]
 dyn_hair_01_cr [{shapeType=Capsule, length=6.9364, radius=2, shapeName=dyn_hair_01_rt_capsule, center=<5 0 0>, rotation=<rot 81.7452 0 0>}]
 dyn_hair_01_lf [{
   shapeType=Capsule
   length=6.834
   radius=3.673
   shapeName=dyn_hair_01_rt_capsule
   center=<5 3.0355 -2.6972>
   rotation=<rot 74.4491 179.9997 179.9998>
  }]
 dyn_hair_01_rt [{
   shapeType=Capsule
   length=18.1145
   radius=2.1222
   shapeName=dyn_hair_01_rt_capsule
   center=<13.7563 0 0.4645>
   rotation=<rot 76.8299 180 180>
  }]
 dyn_hair_02_cr [{shapeType=Capsule, length=5.8123, radius=1.941, shapeName=dyn_hair_02_rt_capsule, center=<2 0 1.0253>, rotation=<rot 83.9553 180 180>}]
 dyn_hair_02_lf [{
   shapeType=Capsule
   length=2.9388
   radius=3.0193
   shapeName=dyn_hair_02_rt_capsule
   center=<5.2385 2.3311 -0.7579>
   rotation=<rot 55.1851 179.9994 179.9996>
  }]
 dyn_hair_02_rt [{
   shapeType=Capsule
   length=5.82
   radius=2.816
   shapeName=dyn_hair_02_rt_capsule
   center=<7.6785 -2.0552 0.7394>
   rotation=<rot 78.449 180 180>
  }]
 dyn_hair_03_cr [{shapeType=Capsule, length=5.9471, radius=2.0529, shapeName=dyn_hair_03_cr_capsule, center=<5.077 0 0.9926>, rotation=<rot 87.9195 0 0>}]
 dyn_hair_03_lf [{shapeType=Capsule, length=7.5885, radius=2.631, shapeName=dyn_hair_03_lf_capsule, center=<8.2057 5.218 -1.6694>, rotation=<rot 80.4786 0 0>}]
 dyn_hair_03_rt [{shapeType=Capsule, length=5.342, radius=2.8685, shapeName=dyn_hair_03_rt_capsule, center=<3.6032 0 1.3327>, rotation=<rot 78.1838 180 180>}]
 dyn_hair_04_cr [{shapeType=Capsule, length=8.7239, radius=2.1494, shapeName=dyn_hair_04_cr_capsule, center=<7.4287 0 0.8046>, rotation=<rot 88.5975 0 0>}]
 dyn_hair_04_lf [{shapeType=Capsule, length=9.3061, radius=2.5852, shapeName=dyn_hair_04_lf_capsule, center=<-0.0493 0 -1.1743>, rotation=<rot 87.5939 0 0>}]
 dyn_hair_04_rt [{
   shapeType=Capsule
   length=8.0312
   radius=2.8421
   shapeName=dyn_hair_04_rt_capsule
   center=<7.3479 0 0.12>
   rotation=<rot 89.3926 0.0003 0.0003>
  }]
 dyn_hair_05_cr [{shapeType=Capsule, length=3, radius=4, shapeName=dyn_hair_05_cr_capsule, center=<5 0 1.2654>, rotation=<rot 79.4625 180 -180>}]
 dyn_hair_05_lf [{
   shapeType=Capsule
   length=5.8098
   radius=3.1902
   shapeName=dyn_hair_05_lf_capsule
   center=<4.9859 0 0.0844>
   rotation=<rot 89.4711 0.0001 0.0001>
  }]
 dyn_hair_05_rt [{shapeType=Capsule, length=5.3607, radius=2.6393, shapeName=dyn_hair_05_rt_capsule, center=<4.8976 0 0.494>, rotation=<rot 78.2878 180 180>}]
 dyn_hair_front_01_lf [
  {shapeType=Capsule, length=8.728, radius=1.7438, shapeName=dyn_hair_front_01_lf_capsule, center=<2 0 3.503>, rotation=<rot 90 0 0>}
  {
   shapeType=Capsule
   length=7.3005
   radius=1.6338
   shapeName=dyn_hair_front_01_lf_capsule
   center=<9.3322 -1.4016 -2.4>
   rotation=<rot 86.425 0 0>
  }
 ]
 dyn_hair_front_01_rt [
  {shapeType=Capsule, length=8.728, radius=1.7438, shapeName=dyn_hair_front_01_lf_capsule, center=<2 0 3.503>, rotation=<rot 90 0 0>}
  {
   shapeType=Capsule
   length=7.3005
   radius=1.6338
   shapeName=dyn_hair_front_01_lf_capsule
   center=<9.3322 -1.4016 -2.4>
   rotation=<rot 86.425 0 0>
  }
 ]
 dyn_hair_front_02_lf [{shapeType=Capsule, length=5, radius=2, shapeName=dyn_hair_front_02_lf_capsule, center=<5 0 -0.8466>, rotation=<rot -90 0 0>}]
 dyn_hair_front_02_rt [{
   shapeType=Capsule
   length=6.0171
   radius=2.0108
   shapeName=dyn_hair_front_02_lf_capsule
   center=<4.9919 -1.1913 -0.608>
   rotation=<rot 79.9996 180 0>
  }]
 dyn_hair_front_03_lf [{
   shapeType=Capsule
   length=6.7311
   radius=1.7875
   shapeName=dyn_hair_front_03_lf_capsule
   center=<4.3387 -0.339 -0.2988>
   rotation=<rot 78.5244 -179.9984 -180>
  }]
 dyn_hair_front_03_rt [{
   shapeType=Capsule
   length=6.8667
   radius=3.4615
   shapeName=dyn_hair_front_03_lf_capsule
   center=<5.8334 -0.8761 1.144>
   rotation=<rot 85.6531 -179.9984 0>
  }]
 dyn_hair_front_04_lf [{
   shapeType=Capsule
   length=6.2311
   radius=2.2875
   shapeName=dyn_hair_front_03_lf_capsule
   center=<4.2575 -0.339 -0.8404>
   rotation=<rot 81.4762 0.0016 0>
  }]
 dyn_hair_front_04_rt [{
   shapeType=Capsule
   length=2.8667
   radius=2.9615
   shapeName=dyn_hair_front_03_lf_capsule
   center=<5.0809 -0.7889 1.3078>
   rotation=<rot 85.6532 -179.9984 -90>
  }]
 head [{shapeType=Capsule, length=7.1907, radius=5.7787, shapeName=head_capsule, center=<0.9288 3.8536 -0.2703>, rotation=<rot 90 6.1106 0>}]
 neck_01 [
  {shapeType=Capsule, length=8.5, radius=6, shapeName=neck_01_capsule, rotation=<rot 0 -90.0003 -90>}
  {shapeType=Capsule, length=30, radius=18, shapeName=neck_01_capsule, center=<-2.3554 -21.5375 0>}
 ]
 neck_02 [{shapeType=Capsule, length=38.3892, radius=12.3653, shapeName=neck_02_capsule, center=<-1.5348 14.5958 0>, rotation=<rot 0 -90 0>}]
 offset_hair_cr [{shapeType=Capsule, length=0.505, radius=0.505, shapeName=offset_hair_cr_capsule, rotation=<rot 0 -90 0>}]
 offset_hair_front_lf [{shapeType=Capsule, length=0.505, radius=0.505, shapeName=offset_hair_front_lf_capsule, rotation=<rot 0 -90 0>}]
 offset_hair_front_rt [{shapeType=Capsule, length=0.505, radius=0.505, shapeName=offset_hair_front_rt_capsule, rotation=<rot 0 -90 0>}]
 offset_hair_lf [{shapeType=Capsule, length=0.505, radius=0.505, shapeName=offset_hair_lf_capsule, rotation=<rot 0 -90 0>}]
 offset_hair_rt [{shapeType=Capsule, length=0.505, radius=0.505, shapeName=offset_hair_rt_capsule, rotation=<rot 0 -90 0>}]
 spine_04 [
  {shapeType=Capsule, length=18.631, radius=11.3554, shapeName=spine_05_capsule, center=<-6.204 2.8253 0>, rotation=<rot 0 -90 0>}
  {shapeType=Capsule, length=21.8384, radius=11.1413, shapeName=spine_04_capsule, center=<4.1473 -3.2659 0>}
  {shapeType=Capsule, length=9.8481, radius=8.8535, shapeName=spine_04_capsule, center=<0.9022 6.6652 0>}
 ]
 spine_05 [
  {shapeType=Capsule, length=20, radius=10.5551, shapeName=spine_05_capsule, center=<2.4023 -2.12 0>, rotation=<rot 0 -90 0>}
  {shapeType=Capsule, length=15, radius=11.5, shapeName=spine_05_capsule, center=<-2.0386 1.3575 0>}
 ]
 upperarm_l [{
   shapeType=Capsule
   length=24.8995
   radius=7.4721
   shapeName=upperarm_l_capsule
   center=<12.95 -0.7645 -0.5337>
   rotation=<rot -0.0003 -89.9997 -92.7028>
  }]
 upperarm_r [{shapeType=Capsule, length=22.5453, radius=6.5231, shapeName=upperarm_l_capsule, center=<-13.9807 0.6006 0.5342>, rotation=<rot 0 -90 -90>}]
}
constraints [
 {
  bone1Name=dyn_hair_01_cr
  bone2Name=offset_hair_cr
  swing1LimitDegrees=0
  swing1Motion=Locked
  swing2LimitDegrees=3
  twistLimitDegrees=0
  twistMotion=Locked
 }
 {bone1Name=dyn_hair_03_cr, bone2Name=dyn_hair_02_cr, swing1LimitDegrees=2, swing2LimitDegrees=15, twistLimitDegrees=0, twistMotion=Locked}
 {bone1Name=dyn_hair_04_cr, bone2Name=dyn_hair_03_cr, swing1LimitDegrees=7, swing2LimitDegrees=20, twistLimitDegrees=0, twistMotion=Locked}
 {bone1Name=dyn_hair_05_cr, bone2Name=dyn_hair_04_cr, swing1LimitDegrees=10, swing2LimitDegrees=30, twistLimitDegrees=45, twistMotion=Locked}
 {
  bone1Name=dyn_hair_01_lf
  bone2Name=offset_hair_lf
  swing1LimitDegrees=0
  swing1Motion=Locked
  swing2LimitDegrees=3
  twistLimitDegrees=0
  twistMotion=Locked
 }
 {bone1Name=dyn_hair_02_lf, bone2Name=dyn_hair_01_lf, swing1LimitDegrees=2, swing2LimitDegrees=10, twistLimitDegrees=0, twistMotion=Locked}
 {bone1Name=dyn_hair_03_lf, bone2Name=dyn_hair_02_lf, swing1LimitDegrees=2, swing2LimitDegrees=15, twistLimitDegrees=0, twistMotion=Locked}
 {bone1Name=dyn_hair_04_lf, bone2Name=dyn_hair_03_lf, swing1LimitDegrees=7, swing2LimitDegrees=20, twistLimitDegrees=0, twistMotion=Locked}
 {bone1Name=dyn_hair_05_lf, bone2Name=dyn_hair_04_lf, swing1LimitDegrees=10, swing2LimitDegrees=30, twistLimitDegrees=45, twistMotion=Locked}
 {
  bone1Name=dyn_hair_01_rt
  bone2Name=offset_hair_rt
  swing1LimitDegrees=0
  swing1Motion=Locked
  swing2LimitDegrees=3
  twistLimitDegrees=0
  twistMotion=Locked
 }
 {bone1Name=dyn_hair_02_rt, bone2Name=dyn_hair_01_rt, swing1LimitDegrees=2, swing2LimitDegrees=10, twistLimitDegrees=0, twistMotion=Locked}
 {bone1Name=dyn_hair_03_rt, bone2Name=dyn_hair_02_rt, swing1LimitDegrees=2, swing2LimitDegrees=15, twistLimitDegrees=0, twistMotion=Locked}
 {bone1Name=dyn_hair_04_rt, bone2Name=dyn_hair_03_rt, swing1LimitDegrees=7, swing2LimitDegrees=20, twistLimitDegrees=0, twistMotion=Locked}
 {bone1Name=dyn_hair_05_rt, bone2Name=dyn_hair_04_rt, swing1LimitDegrees=10, swing2LimitDegrees=30, twistLimitDegrees=45, twistMotion=Locked}
 {bone1Name=dyn_hair_02_cr, bone2Name=dyn_hair_01_cr, swing1LimitDegrees=2, swing2LimitDegrees=10, twistLimitDegrees=0, twistMotion=Locked}
 {bone1Name=dyn_hair_front_04_rt, bone2Name=dyn_hair_front_03_rt, swing1LimitDegrees=20, swing2LimitDegrees=20, twistLimitDegrees=5}
 {bone1Name=dyn_hair_front_01_rt, bone2Name=offset_hair_front_rt, swing1LimitDegrees=5, swing2LimitDegrees=5, twistLimitDegrees=0}
 {
  bone1Name=dyn_hair_front_02_rt
  bone2Name=dyn_hair_front_01_rt
  swing1LimitDegrees=10
  swing2LimitDegrees=15
  twistLimitDegrees=30
  twistMotion=Locked
 }
 {
  bone1Name=dyn_hair_front_03_rt
  bone2Name=dyn_hair_front_02_rt
  swing1LimitDegrees=10
  swing2LimitDegrees=20
  twistLimitDegrees=5
  twistMotion=Locked
 }
 {
  bone1Name=dyn_hair_front_01_lf
  bone2Name=offset_hair_front_lf
  swing1LimitDegrees=5
  swing2LimitDegrees=30
  twistLimitDegrees=45
  twistMotion=Locked
 }
 {
  bone1Name=dyn_hair_front_02_lf
  bone2Name=dyn_hair_front_01_lf
  swing1LimitDegrees=10
  swing2LimitDegrees=15
  twistLimitDegrees=45
  twistMotion=Locked
 }
 {
  bone1Name=dyn_hair_front_03_lf
  bone2Name=dyn_hair_front_02_lf
  swing1LimitDegrees=10
  swing2LimitDegrees=20
  twistLimitDegrees=45
  twistMotion=Locked
 }
 {
  bone1Name=dyn_hair_front_04_lf
  bone2Name=dyn_hair_front_03_lf
  swing1LimitDegrees=20
  swing2LimitDegrees=20
  twistLimitDegrees=5
  twistMotion=Locked
 }
]
