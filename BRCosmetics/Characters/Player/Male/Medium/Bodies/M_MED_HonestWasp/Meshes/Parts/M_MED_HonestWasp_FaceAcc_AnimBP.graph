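# Anim Blueprint for the HonestWasp face accessory, animating hair and pigtails as secondary motion on the character.
# Parent pose, blended with Body_Mesh clavicles and upper arms, feeds a Control Rig and bone offsets.
# Front-end or gameplay RigidBody simulation then feeds chained AnimDynamics pigtail bones, with spherical limits on front strands.
# When a backbling or cape is worn, hair bones are constrained to spine_05 before output.
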AnimBlueprint /BRCosmetics/Characters/Player/Male/Medium/Bodies/M_MED_HonestWasp/Meshes/Parts/M_MED_HonestWasp_FaceAcc_AnimBP
target_skeleton=&/M_MED_HonestWasp_FaceAcc_Skeleton
graphs [
 AnimGraph {nodes [
   Anim Dynamics - Bone: dyn_pigtail_01_cr(Alpha=1, ExternalForce="0,0,0") {
    angularDampingOverride=1
    angularSpringConstant=200
    bAngularSpring=true
    bOverrideAngularDamping=true
    bOverrideLinearDamping=true
    bUseGravityOverride=true
    linearDampingOverride=1
    linearSpringConstant=50
    simulationSpace=RootRelative
    alphaBoolBlend {blendOption=Linear}
    boundBone {boneName=dyn_pigtail_01_cr}
    componentAppliedLinearAccClamp=<3000 3000 3000>
    componentLinearAccScale=<0.4 0.4 0.4>
    componentLinearVelScale=<0.4 0.4 0.4>
    gravityOverride {z=-980}
    physicsBodyDefinitions [{
      boundBone {boneName=dyn_pigtail_01_cr}
      boxExtents=<10 20 10>
      constraintSetup {angularLimitsMax=<0 2 5>, angularLimitsMin=<0 -15 -5>, angularTarget=<1 0 0>}
      localJointOffset=<3 0 0>
     }]
   }
   Output Pose(Result=Use cached pose 'backpackCapeCache'()) {layerGroup=None}
   n1 = Layered blend per bone(
    BasePose=Copy Pose From Mesh() {bUseAttachedParent=true}
    BlendPoses_0=Copy Pose From Mesh(SourceMeshComponent=Get Body_Mesh()) {bUseAttachedParent=true}
   ) {layerSetup [{branchFilters [{boneName=clavicle_l}, {boneName=clavicle_r}, {boneName=upperarm_r}, {boneName=upperarm_l}]}]}
   n2 = Control Rig!(
    Source=@n1
    Alpha
    swingPickaxe:bool=Get bIsPlayingMeleeAnim()
    IsTargeting:bool=Get bIsTargeting()
    IsFrontEnd:bool=Get bIsInFrontEnd()
    isWearingBackblingCape:bool=Get is Wearing cape and backpack() -> (is Wearing cape and backpack:bool)
   ) -> (Pose) {controlRigAssetReference {blueprintRigClass=&/M_MED_HonestWasp_FaceAcc_CtrlRig.M_MED_HonestWasp_FaceAcc_CtrlRig_C}}
   Save cached pose 'controlRigCache'(Pose=@n2) {cacheName=controlRigCache}
   n3 = RigidBody!(
    ComponentPose=Local To Component(LocalPose=Use cached pose 'PreDynamicsPose'())
    Alpha
    OverrideWorldGravity:Vector="0,0,-980"
    SimSpaceSettings:Sim Space Settings Structure=(WorldAlpha=1,VelocityScaleZ=1,DampingAlpha=0,MaxLinearVelocity=350,MaxAngularVelocity=4,MaxLinearAcceleration=1500,MaxAngularAcceleration=9,ExternalLinearDragV=(X=1.1,Y=1.1,Z=0),ExternalLinearVelocity=(X=0,Y=0,Z=0),ExternalAngularVelocity=(X=0,Y=0,Z=0))
   ) -> (Pose) {
    bClampLinearTranslationLimitToRefPose=true
    bOverrideWorldGravity=true
    overridePhysicsAsset=&/M_MED_HonestWasp_FaceAcc_Physics_FE
    simulationSpace=BaseBoneSpace
    alphaBoolBlend {blendOption=Linear}
    baseBoneRef {boneName=neck_02}
    overrideWorldGravity {z=-980}
    simSpaceSettings {
     dampingAlpha=0
     maxAngularAcceleration=9
     maxAngularVelocity=4
     maxLinearAcceleration=1500
     maxLinearVelocity=350
     worldAlpha=1
     externalLinearDragV=<1.1 1.1>
    }
   }
   n4 = Component To Local(ComponentPose=@n3)
   n5 = RigidBody(
    ComponentPose=Local To Component(LocalPose=Use cached pose 'PreDynamicsPose'())
    OverrideWorldGravity="0,0,-980"
    SimSpaceSettings=(WorldAlpha=1,VelocityScaleZ=1,DampingAlpha=0,MaxLinearVelocity=350,MaxAngularVelocity=4,MaxLinearAcceleration=1500,MaxAngularAcceleration=9,ExternalLinearDragV=(X=1.1,Y=1.1,Z=0),ExternalLinearVelocity=(X=0,Y=0,Z=0),ExternalAngularVelocity=(X=0,Y=0,Z=0))
   ) {
    bClampLinearTranslationLimitToRefPose=true
    bOverrideWorldGravity=true
    overridePhysicsAsset=&/M_MED_HonestWasp_FaceAcc_Physics1
    simulationSpace=BaseBoneSpace
    alphaBoolBlend {blendOption=Linear}
    baseBoneRef {boneName=neck_02}
    overrideWorldGravity {z=-980}
    simSpaceSettings {
     dampingAlpha=0
     maxAngularAcceleration=9
     maxAngularVelocity=4
     maxLinearAcceleration=1500
     maxLinearVelocity=350
     worldAlpha=1
     externalLinearDragV=<1.1 1.1>
    }
   }
   n6 = Component To Local(ComponentPose=@n5)
   n7 = Get bIsInFrontEnd()
   n8 = "Transform (Modify) Bone - Bone: dyn_hair_front_02_rt"!(
    ComponentPose=Local To Component(LocalPose=Blend Poses by bool(BlendPose_0=@n4, BlendPose_1=@n6, bActiveValue=@n7) {blendType=Linear})
    bAlphaBoolEnabled:bool=True
    Translation="0,0,0"
    Rotation="-5,0,0"
    Scale="1,1,1"
   ) -> (Pose) {
    alphaInputType=Bool
    rotationMode=BMM_Additive
    rotationSpace=BCS_BoneSpace
    translationMode=BMM_Additive
    translationSpace=BCS_BoneSpace
    boneToModify {boneName=dyn_hair_front_02_rt}
    rotation {pitch=-5}
   }
   n9 = Anim Dynamics - Bone: dyn_pigtail_front_01_rt(ComponentPose=@n8, Alpha=1, ExternalForce="0,0,0") {
    angularDampingOverride=1
    angularSpringConstant=200
    bAngularSpring=true
    bOverrideAngularDamping=true
    bOverrideLinearDamping=true
    bUseGravityOverride=true
    bUseSphericalLimits=true
    linearDampingOverride=1
    linearSpringConstant=50
    simulationSpace=RootRelative
    alphaBoolBlend {blendOption=Linear}
    boundBone {boneName=dyn_pigtail_front_01_rt}
    componentAppliedLinearAccClamp=<3000 3000 3000>
    componentLinearAccScale=<0.4 0.4 0.4>
    componentLinearVelScale=<0.4 0.4 0.4>
    gravityOverride {z=-980}
    physicsBodyDefinitions [{
      boundBone {boneName=dyn_pigtail_front_01_rt}
      boxExtents=<10 10 10>
      constraintSetup {angularLimitsMax=<0 5 10>, angularLimitsMin=<0 -5 -10>, angularTarget=<1 0 0>}
      localJointOffset=<5 0 0>
     }]
    sphericalLimits [{limitRadius=25, drivingBone {boneName=spine_05}, sphereLocalOffset=<-10 -10 0>}]
   }
   n10 = Anim Dynamics - Bone: dyn_pigtail_front_02_rt(ComponentPose=@n9, Alpha=1, ExternalForce="0,0,0") {
    angularDampingOverride=1
    angularSpringConstant=200
    bAngularSpring=true
    bOverrideAngularDamping=true
    bOverrideLinearDamping=true
    bUseGravityOverride=true
    bUseSphericalLimits=true
    linearDampingOverride=1
    linearSpringConstant=50
    simulationSpace=RootRelative
    alphaBoolBlend {blendOption=Linear}
    boundBone {boneName=dyn_pigtail_front_02_rt}
    componentAppliedLinearAccClamp=<3000 3000 3000>
    componentLinearAccScale=<0.4 0.4 0.4>
    componentLinearVelScale=<0.4 0.4 0.4>
    gravityOverride {z=-980}
    physicsBodyDefinitions [{
      boundBone {boneName=dyn_pigtail_front_02_rt}
      boxExtents=<10 10 10>
      constraintSetup {angularLimitsMax=<0 5 10>, angularLimitsMin=<0 -5 -10>, angularTarget=<1 0 0>}
      localJointOffset=<5 0 0>
     }]
    sphericalLimits [{limitRadius=25, drivingBone {boneName=spine_05}, sphereLocalOffset=<-10 -10 0>}]
   }
   n11 = Anim Dynamics - Bone: dyn_pigtail_front_03_rt(ComponentPose=@n10, Alpha=1, ExternalForce="0,0,0") {
    angularDampingOverride=1
    angularSpringConstant=200
    bAngularSpring=true
    bOverrideAngularDamping=true
    bOverrideLinearDamping=true
    bUseGravityOverride=true
    bUseSphericalLimits=true
    linearDampingOverride=1
    linearSpringConstant=50
    simulationSpace=RootRelative
    alphaBoolBlend {blendOption=Linear}
    boundBone {boneName=dyn_pigtail_front_03_rt}
    componentAppliedLinearAccClamp=<3000 3000 3000>
    componentLinearAccScale=<0.4 0.4 0.4>
    componentLinearVelScale=<0.4 0.4 0.4>
    gravityOverride {z=-980}
    physicsBodyDefinitions [{
      boundBone {boneName=dyn_pigtail_front_03_rt}
      boxExtents=<10 10 10>
      constraintSetup {angularLimitsMax=<0 5 10>, angularLimitsMin=<0 -5 -10>, angularTarget=<1 0 0>}
      localJointOffset=<5 0 0>
     }]
    sphericalLimits [{limitRadius=25, drivingBone {boneName=spine_05}, sphereLocalOffset=<-10 -10 0>}]
   }
   n12 = Anim Dynamics - Bone: dyn_pigtail_front_04_rt(ComponentPose=@n11, Alpha=1, ExternalForce="0,0,0") {
    angularDampingOverride=1
    angularSpringConstant=200
    bAngularSpring=true
    bOverrideAngularDamping=true
    bOverrideLinearDamping=true
    bUseGravityOverride=true
    bUseSphericalLimits=true
    linearDampingOverride=1
    linearSpringConstant=50
    simulationSpace=RootRelative
    alphaBoolBlend {blendOption=Linear}
    boundBone {boneName=dyn_pigtail_front_04_rt}
    componentAppliedLinearAccClamp=<3000 3000 3000>
    componentLinearAccScale=<0.4 0.4 0.4>
    componentLinearVelScale=<0.4 0.4 0.4>
    gravityOverride {z=-980}
    physicsBodyDefinitions [{
      boundBone {boneName=dyn_pigtail_front_04_rt}
      boxExtents=<10 10 10>
      constraintSetup {angularLimitsMax=<0 5 10>, angularLimitsMin=<0 -10 -2>, angularTarget=<1 0 0>}
      localJointOffset=<5 0 0>
     }]
    sphericalLimits [{limitRadius=25, drivingBone {boneName=spine_05}, sphereLocalOffset=<-10 -10 0>}]
   }
   n13 = Anim Dynamics - Bone: dyn_pigtail_front_05_rt(ComponentPose=@n12, Alpha=1, ExternalForce="0,0,0") {
    angularDampingOverride=1
    angularSpringConstant=200
    bAngularSpring=true
    bOverrideAngularDamping=true
    bOverrideLinearDamping=true
    bUseGravityOverride=true
    bUseSphericalLimits=true
    linearDampingOverride=1
    linearSpringConstant=50
    simulationSpace=RootRelative
    alphaBoolBlend {blendOption=Linear}
    boundBone {boneName=dyn_pigtail_front_05_rt}
    componentAppliedLinearAccClamp=<3000 3000 3000>
    componentLinearAccScale=<0.4 0.4 0.4>
    componentLinearVelScale=<0.4 0.4 0.4>
    gravityOverride {z=-980}
    physicsBodyDefinitions [{
      boundBone {boneName=dyn_pigtail_front_05_rt}
      boxExtents=<10 10 10>
      constraintSetup {angularLimitsMax=<0 5 10>, angularLimitsMin=<0 -10 -2>, angularTarget=<1 0 0>}
      localJointOffset=<5 0 0>
     }]
    sphericalLimits [{limitRadius=25, drivingBone {boneName=spine_05}, sphereLocalOffset=<-10 -10 0>}]
   }
   n14 = Anim Dynamics - Bone: dyn_pigtail_front_01_lf(ComponentPose=@n13, Alpha=1, ExternalForce="0,0,0") {
    angularDampingOverride=1
    angularSpringConstant=200
    bAngularSpring=true
    bOverrideAngularDamping=true
    bOverrideLinearDamping=true
    bUseGravityOverride=true
    bUseSphericalLimits=true
    linearDampingOverride=1
    linearSpringConstant=50
    simulationSpace=RootRelative
    alphaBoolBlend {blendOption=Linear}
    boundBone {boneName=dyn_pigtail_front_01_lf}
    componentAppliedLinearAccClamp=<3000 3000 3000>
    componentLinearAccScale=<0.4 0.4 0.4>
    componentLinearVelScale=<0.4 0.4 0.4>
    gravityOverride {z=-980}
    physicsBodyDefinitions [{
      boundBone {boneName=dyn_pigtail_front_01_lf}
      boxExtents=<10 10 10>
      constraintSetup {angularLimitsMax=<0 -2 0>, angularLimitsMin=<0 -20 -10>, angularTarget=<1 0 0>}
      localJointOffset=<5 0 0>
     }]
    sphericalLimits [{limitRadius=25, drivingBone {boneName=spine_05}, sphereLocalOffset=<-10 -10 0>}]
   }
   n15 = Anim Dynamics - Bone: dyn_pigtail_front_02_lf(ComponentPose=@n14, Alpha=1, ExternalForce="0,0,0") {
    angularDampingOverride=1
    angularSpringConstant=200
    bAngularSpring=true
    bOverrideAngularDamping=true
    bOverrideLinearDamping=true
    bUseGravityOverride=true
    bUseSphericalLimits=true
    linearDampingOverride=1
    linearSpringConstant=50
    simulationSpace=RootRelative
    alphaBoolBlend {blendOption=Linear}
    boundBone {boneName=dyn_pigtail_front_02_lf}
    componentAppliedLinearAccClamp=<3000 3000 3000>
    componentLinearAccScale=<0.4 0.4 0.4>
    componentLinearVelScale=<0.4 0.4 0.4>
    gravityOverride {z=-980}
    physicsBodyDefinitions [{
      boundBone {boneName=dyn_pigtail_front_02_lf}
      boxExtents=<10 10 10>
      constraintSetup {angularLimitsMax=<0 -2 0>, angularLimitsMin=<0 -20 -10>, angularTarget=<1 0 0>}
      localJointOffset=<5 0 0>
     }]
    sphericalLimits [{limitRadius=25, drivingBone {boneName=spine_05}, sphereLocalOffset=<-10 -10 0>}]
   }
   n16 = Anim Dynamics - Bone: dyn_pigtail_front_03_lf(ComponentPose=@n15, Alpha=1, ExternalForce="0,0,0") {
    angularDampingOverride=1
    angularSpringConstant=200
    bAngularSpring=true
    bOverrideAngularDamping=true
    bOverrideLinearDamping=true
    bUseGravityOverride=true
    bUseSphericalLimits=true
    linearDampingOverride=1
    linearSpringConstant=50
    simulationSpace=RootRelative
    alphaBoolBlend {blendOption=Linear}
    boundBone {boneName=dyn_pigtail_front_03_lf}
    componentAppliedLinearAccClamp=<3000 3000 3000>
    componentLinearAccScale=<0.4 0.4 0.4>
    componentLinearVelScale=<0.4 0.4 0.4>
    gravityOverride {z=-980}
    physicsBodyDefinitions [{
      boundBone {boneName=dyn_pigtail_front_03_lf}
      boxExtents=<10 10 10>
      constraintSetup {angularLimitsMax=<0 -2 0>, angularLimitsMin=<0 -20 -10>, angularTarget=<1 0 0>}
      localJointOffset=<5 0 0>
     }]
    sphericalLimits [{limitRadius=25, drivingBone {boneName=spine_05}, sphereLocalOffset=<-10 -10 0>}]
   }
   n17 = Anim Dynamics - Bone: dyn_pigtail_front_04_lf(ComponentPose=@n16, Alpha=1, ExternalForce="0,0,0") {
    angularDampingOverride=1
    angularSpringConstant=200
    bAngularSpring=true
    bOverrideAngularDamping=true
    bOverrideLinearDamping=true
    bUseGravityOverride=true
    bUseSphericalLimits=true
    linearDampingOverride=1
    linearSpringConstant=50
    simulationSpace=RootRelative
    alphaBoolBlend {blendOption=Linear}
    boundBone {boneName=dyn_pigtail_front_04_lf}
    componentAppliedLinearAccClamp=<3000 3000 3000>
    componentLinearAccScale=<0.4 0.4 0.4>
    componentLinearVelScale=<0.4 0.4 0.4>
    gravityOverride {z=-980}
    physicsBodyDefinitions [{
      boundBone {boneName=dyn_pigtail_front_04_lf}
      boxExtents=<10 10 10>
      constraintSetup {angularLimitsMax=<0 -5 0>, angularLimitsMin=<0 -10 -10>, angularTarget=<1 0 0>}
      localJointOffset=<5 0 0>
     }]
    sphericalLimits [{limitRadius=25, drivingBone {boneName=spine_05}, sphereLocalOffset=<-10 -10 0>}]
   }
   n18 = Anim Dynamics - Bone: dyn_pigtail_front_05_lf(ComponentPose=@n17, Alpha=1, ExternalForce="0,0,0") {
    angularDampingOverride=1
    angularSpringConstant=200
    bAngularSpring=true
    bOverrideAngularDamping=true
    bOverrideLinearDamping=true
    bUseGravityOverride=true
    bUseSphericalLimits=true
    linearDampingOverride=1
    linearSpringConstant=50
    simulationSpace=RootRelative
    alphaBoolBlend {blendOption=Linear}
    boundBone {boneName=dyn_pigtail_front_05_lf}
    componentAppliedLinearAccClamp=<3000 3000 3000>
    componentLinearAccScale=<0.4 0.4 0.4>
    componentLinearVelScale=<0.4 0.4 0.4>
    gravityOverride {z=-980}
    physicsBodyDefinitions [{
      boundBone {boneName=dyn_pigtail_front_05_lf}
      boxExtents=<10 10 10>
      constraintSetup {angularLimitsMax=<0 -5 0>, angularLimitsMin=<0 -10 -10>, angularTarget=<1 0 0>}
      localJointOffset=<5 0 0>
     }]
    sphericalLimits [{limitRadius=25, drivingBone {boneName=spine_05}, sphereLocalOffset=<-10 -10 0>}]
   }
   n19 = Anim Dynamics - Bone: dyn_pigtail_02_cr(ComponentPose=@n18, Alpha=1, ExternalForce="0,0,0") {
    angularDampingOverride=1
    angularSpringConstant=200
    bAngularSpring=true
    bOverrideAngularDamping=true
    bOverrideLinearDamping=true
    bUseGravityOverride=true
    linearDampingOverride=1
    linearSpringConstant=50
    simulationSpace=RootRelative
    alphaBoolBlend {blendOption=Linear}
    boundBone {boneName=dyn_pigtail_02_cr}
    componentAppliedLinearAccClamp=<3000 3000 3000>
    componentLinearAccScale=<0.4 0.4 0.4>
    componentLinearVelScale=<0.4 0.4 0.4>
    gravityOverride {z=-980}
    physicsBodyDefinitions [{
      boundBone {boneName=dyn_pigtail_02_cr}
      boxExtents=<10 20 10>
      constraintSetup {angularLimitsMax=<0 2 5>, angularLimitsMin=<0 -10 -5>, angularTarget=<1 0 0>}
      localJointOffset=<3 0 0>
     }]
   }
   n20 = Anim Dynamics - Bone: dyn_pigtail_03_cr(ComponentPose=@n19, Alpha=1, ExternalForce="0,0,0") {
    angularDampingOverride=1
    angularSpringConstant=200
    bAngularSpring=true
    bOverrideAngularDamping=true
    bOverrideLinearDamping=true
    bUseGravityOverride=true
    linearDampingOverride=1
    linearSpringConstant=50
    simulationSpace=RootRelative
    alphaBoolBlend {blendOption=Linear}
    boundBone {boneName=dyn_pigtail_03_cr}
    componentAppliedLinearAccClamp=<3000 3000 3000>
    componentLinearAccScale=<0.4 0.4 0.4>
    componentLinearVelScale=<0.4 0.4 0.4>
    gravityOverride {z=-980}
    physicsBodyDefinitions [{
      boundBone {boneName=dyn_pigtail_03_cr}
      boxExtents=<10 20 10>
      constraintSetup {angularLimitsMax=<0 2 5>, angularLimitsMin=<0 -10 -5>, angularTarget=<1 0 0>}
      localJointOffset=<3 0 0>
     }]
   }
   n21 = Anim Dynamics - Bone: dyn_pigtail_01_B_lf(ComponentPose=@n20, Alpha=1, ExternalForce="0,0,0") {
    angularDampingOverride=1
    angularSpringConstant=200
    bAngularSpring=true
    bOverrideAngularDamping=true
    bOverrideLinearDamping=true
    bUseGravityOverride=true
    linearDampingOverride=1
    linearSpringConstant=50
    simulationSpace=RootRelative
    alphaBoolBlend {blendOption=Linear}
    boundBone {boneName=dyn_pigtail_01_B_lf}
    componentAppliedLinearAccClamp=<3000 3000 3000>
    componentLinearAccScale=<0.4 0.4 0.4>
    componentLinearVelScale=<0.4 0.4 0.4>
    gravityOverride {z=-980}
    physicsBodyDefinitions [{
      boundBone {boneName=dyn_pigtail_01_B_lf}
      boxExtents=<10 20 10>
      constraintSetup {angularLimitsMax=<0 5 5>, angularLimitsMin=<0 -5 -5>, angularTarget=<1 0 0>}
      localJointOffset=<3 0 0>
     }]
   }
   n22 = Anim Dynamics - Bone: dyn_pigtail_02_B_lf(ComponentPose=@n21, Alpha=1, ExternalForce="0,0,0") {
    angularDampingOverride=1
    angularSpringConstant=200
    bAngularSpring=true
    bOverrideAngularDamping=true
    bOverrideLinearDamping=true
    bUseGravityOverride=true
    linearDampingOverride=1
    linearSpringConstant=50
    simulationSpace=RootRelative
    alphaBoolBlend {blendOption=Linear}
    boundBone {boneName=dyn_pigtail_02_B_lf}
    componentAppliedLinearAccClamp=<3000 3000 3000>
    componentLinearAccScale=<0.4 0.4 0.4>
    componentLinearVelScale=<0.4 0.4 0.4>
    gravityOverride {z=-980}
    physicsBodyDefinitions [{
      boundBone {boneName=dyn_pigtail_02_B_lf}
      boxExtents=<10 20 10>
      constraintSetup {angularLimitsMax=<0 5 5>, angularLimitsMin=<0 -5 -5>, angularTarget=<1 0 0>}
      localJointOffset=<3 0 0>
     }]
   }
   n23 = Anim Dynamics - Bone: dyn_pigtail_03_B_lf(ComponentPose=@n22, Alpha=1, ExternalForce="0,0,0") {
    angularDampingOverride=1
    angularSpringConstant=200
    bAngularSpring=true
    bOverrideAngularDamping=true
    bOverrideLinearDamping=true
    bUseGravityOverride=true
    linearDampingOverride=1
    linearSpringConstant=50
    simulationSpace=RootRelative
    alphaBoolBlend {blendOption=Linear}
    boundBone {boneName=dyn_pigtail_03_B_lf}
    componentAppliedLinearAccClamp=<3000 3000 3000>
    componentLinearAccScale=<0.4 0.4 0.4>
    componentLinearVelScale=<0.4 0.4 0.4>
    gravityOverride {z=-980}
    physicsBodyDefinitions [{
      boundBone {boneName=dyn_pigtail_03_B_lf}
      boxExtents=<10 20 10>
      constraintSetup {angularLimitsMax=<0 5 5>, angularLimitsMin=<0 -5 -5>, angularTarget=<1 0 0>}
      localJointOffset=<3 0 0>
     }]
   }
   n24 = Anim Dynamics - Bone: dyn_pigtail_01_A_lf(ComponentPose=@n23, Alpha=1, ExternalForce="0,0,0") {
    angularDampingOverride=1
    angularSpringConstant=200
    bAngularSpring=true
    bOverrideAngularDamping=true
    bOverrideLinearDamping=true
    bUseGravityOverride=true
    linearDampingOverride=1
    linearSpringConstant=50
    simulationSpace=RootRelative
    alphaBoolBlend {blendOption=Linear}
    boundBone {boneName=dyn_pigtail_01_A_lf}
    componentAppliedLinearAccClamp=<3000 3000 3000>
    componentLinearAccScale=<0.4 0.4 0.4>
    componentLinearVelScale=<0.4 0.4 0.4>
    gravityOverride {z=-980}
    physicsBodyDefinitions [{
      boundBone {boneName=dyn_pigtail_01_A_lf}
      boxExtents=<10 20 10>
      constraintSetup {angularLimitsMax=<0 5 5>, angularLimitsMin=<0 -5 -5>, angularTarget=<1 0 0>}
      localJointOffset=<3 0 0>
     }]
   }
   n25 = Anim Dynamics - Bone: dyn_pigtail_02_A_lf(ComponentPose=@n24, Alpha=1, ExternalForce="0,0,0") {
    angularDampingOverride=1
    angularSpringConstant=200
    bAngularSpring=true
    bOverrideAngularDamping=true
    bOverrideLinearDamping=true
    bUseGravityOverride=true
    linearDampingOverride=1
    linearSpringConstant=50
    simulationSpace=RootRelative
    alphaBoolBlend {blendOption=Linear}
    boundBone {boneName=dyn_pigtail_02_A_lf}
    componentAppliedLinearAccClamp=<3000 3000 3000>
    componentLinearAccScale=<0.4 0.4 0.4>
    componentLinearVelScale=<0.4 0.4 0.4>
    gravityOverride {z=-980}
    physicsBodyDefinitions [{
      boundBone {boneName=dyn_pigtail_02_A_lf}
      boxExtents=<10 20 10>
      constraintSetup {angularLimitsMax=<0 5 5>, angularLimitsMin=<0 -5 -5>, angularTarget=<1 0 0>}
      localJointOffset=<3 0 0>
     }]
   }
   n26 = Anim Dynamics - Bone: dyn_pigtail_03_A_lf(ComponentPose=@n25, Alpha=1, ExternalForce="0,0,0") {
    angularDampingOverride=1
    angularSpringConstant=200
    bAngularSpring=true
    bOverrideAngularDamping=true
    bOverrideLinearDamping=true
    bUseGravityOverride=true
    linearDampingOverride=1
    linearSpringConstant=50
    simulationSpace=RootRelative
    alphaBoolBlend {blendOption=Linear}
    boundBone {boneName=dyn_pigtail_03_A_lf}
    componentAppliedLinearAccClamp=<3000 3000 3000>
    componentLinearAccScale=<0.4 0.4 0.4>
    componentLinearVelScale=<0.4 0.4 0.4>
    gravityOverride {z=-980}
    physicsBodyDefinitions [{
      boundBone {boneName=dyn_pigtail_03_A_lf}
      boxExtents=<10 20 10>
      constraintSetup {angularLimitsMax=<0 5 5>, angularLimitsMin=<0 -5 -5>, angularTarget=<1 0 0>}
      localJointOffset=<3 0 0>
     }]
   }
   Save cached pose 'dynamicsCache'(Pose=Component To Local(ComponentPose=@n26)) {cacheName=dynamicsCache}
   n27 = "Transform (Modify) Bone - Bone: offset_hair_front_rt"!(
    ComponentPose=Local To Component(LocalPose=Use cached pose 'controlRigCache'())
    bAlphaBoolEnabled:bool=True
    Translation="0,0.5,0"
    Rotation="0,5,0"
    Scale="1,1,1"
   ) -> (Pose) {
    alphaInputType=Bool
    rotationMode=BMM_Additive
    rotationSpace=BCS_BoneSpace
    translationMode=BMM_Additive
    translationSpace=BCS_BoneSpace
    alphaBoolBlend {blendInTime=0.5, blendOutTime=0.5}
    boneToModify {boneName=offset_hair_front_rt}
    rotation {yaw=5}
    translation {y=0.5}
   }
   n28 = "Transform (Modify) Bone - Bone: offset_hair_front_rt"(ComponentPose=@n27, bAlphaBoolEnabled=True, Translation="0,0,0", Rotation="-10,5,0", Scale="1,1,1") {
    alphaInputType=Bool
    rotationMode=BMM_Additive
    rotationSpace=BCS_BoneSpace
    translationMode=BMM_Additive
    translationSpace=BCS_BoneSpace
    alphaBoolBlend {blendInTime=0.5, blendOutTime=0.5}
    boneToModify {boneName=offset_hair_front_rt}
    rotation {pitch=-10, yaw=5}
   }
   n29 = "Transform (Modify) Bone - Bone: offset_hair_rt"!(ComponentPose=@n28, bAlphaBoolEnabled:bool=True, Translation="0,0,0", Rotation="-10,5,0", Scale="1,1,1") -> (Pose) {
    alphaInputType=Bool
    rotationMode=BMM_Additive
    rotationSpace=BCS_BoneSpace
    translationMode=BMM_Additive
    translationSpace=BCS_BoneSpace
    alphaBoolBlend {blendInTime=0.5, blendOutTime=0.5}
    boneToModify {boneName=offset_hair_rt}
    rotation {pitch=-10, yaw=5}
   }
   n30 = "Transform (Modify) Bone - Bone: offset_hair_front_rt"!(
    ComponentPose="Transform (Modify) Bone - Bone: skn_hair_front_02_lf"(
     ComponentPose="Transform (Modify) Bone - Bone: skn_hair_front_02_rt"(ComponentPose=@n29, Alpha=1, Translation="0,0,0", Rotation="0,0,0", Scale="1,0.5,1") {scaleMode=BMM_Additive, scaleSpace=BCS_BoneSpace, boneToModify {boneName=skn_hair_front_02_rt}, scale {y=0.5}}
     Alpha=1
     Translation="0,0,0"
     Rotation="0,0,0"
     Scale="1,0.5,1"
    ) {scaleMode=BMM_Additive, scaleSpace=BCS_BoneSpace, boneToModify {boneName=skn_hair_front_02_lf}, scale {y=0.5}}
    Alpha:float=1
    Translation="0,0,0"
    Rotation="0,0,0"
    Scale="1,1,1"
   ) -> (Pose) {
    rotationMode=BMM_Additive
    rotationSpace=BCS_BoneSpace
    scaleMode=BMM_Additive
    scaleSpace=BCS_BoneSpace
    boneToModify {boneName=offset_hair_front_rt}
   }
   n31 = "Transform (Modify) Bone - Bone: offset_pigtail_cr"(ComponentPose=@n30, Alpha=1, Translation="-2,0,0", Rotation="0,0,0", Scale="1,1,1") {
    rotationMode=BMM_Additive
    rotationSpace=BCS_BoneSpace
    scaleMode=BMM_Additive
    scaleSpace=BCS_BoneSpace
    translationMode=BMM_Additive
    translationSpace=BCS_BoneSpace
    boneToModify {boneName=offset_pigtail_cr}
    translation {x=-2}
   }
   n32 = "Transform (Modify) Bone - Bone: base_hair_cr"!(ComponentPose=@n31, bAlphaBoolEnabled:bool=True, Translation="0,0,0", Scale="1,1,1") -> (Pose) {
    alphaInputType=Bool
    rotationMode=BMM_Additive
    rotationSpace=BCS_BoneSpace
    translationMode=BMM_Additive
    translationSpace=BCS_BoneSpace
    alphaBoolBlend {blendInTime=0.5, blendOutTime=0.5}
    boneToModify {boneName=base_hair_cr}
    rotation {pitch=-10}
   }
   n33 = "Transform (Modify) Bone - Bone: base_hair_lf"!(ComponentPose=@n32, bAlphaBoolEnabled:bool=True, Translation="0,0,0", Scale="1,1,1") -> (Pose) {
    alphaInputType=Bool
    rotationMode=BMM_Additive
    rotationSpace=BCS_BoneSpace
    translationMode=BMM_Additive
    translationSpace=BCS_BoneSpace
    alphaBoolBlend {blendInTime=0.5, blendOutTime=0.5}
    boneToModify {boneName=base_hair_lf}
    rotation {pitch=-10}
   }
   n34 = "Transform (Modify) Bone - Bone: base_hair_rt"!(ComponentPose=@n33, bAlphaBoolEnabled:bool=True, Translation="0,0,0", Scale="1,1,1") -> (Pose) {
    alphaInputType=Bool
    rotationMode=BMM_Additive
    rotationSpace=BCS_BoneSpace
    translationMode=BMM_Additive
    translationSpace=BCS_BoneSpace
    alphaBoolBlend {blendInTime=0.5, blendOutTime=0.5}
    boneToModify {boneName=base_hair_rt}
    rotation {pitch=-10}
   }
   Save cached pose 'PreDynamicsPose'(Pose=Component To Local(ComponentPose=@n34)) {cacheName=PreDynamicsPose}
   n35 = "Constraint - dyn_hair_02_lf (P)"!(ComponentPose=Local To Component(LocalPose=Use cached pose 'dynamicsCache'()), bAlphaBoolEnabled:bool=True, ConstraintWeights_0=1) -> (Pose) {alphaInputType=Bool, boneToModify {boneName=dyn_hair_02_lf}, constraintSetup [{targetBone {boneName=spine_05}}], constraintWeights [1]}
   n36 = "Constraint - dyn_hair_03_lf (P)"!(ComponentPose=@n35, bAlphaBoolEnabled:bool=True, ConstraintWeights_0=1) -> (Pose) {alphaInputType=Bool, boneToModify {boneName=dyn_hair_03_lf}, constraintSetup [{targetBone {boneName=spine_05}}], constraintWeights [1]}
   n37 = "Constraint - dyn_hair_04_lf (P)"!(ComponentPose=@n36, bAlphaBoolEnabled:bool=True, ConstraintWeights_0=1) -> (Pose) {alphaInputType=Bool, boneToModify {boneName=dyn_hair_04_lf}, constraintSetup [{targetBone {boneName=spine_05}}], constraintWeights [1]}
   n38 = "Constraint - dyn_hair_05_lf (P)"!(ComponentPose=@n37, bAlphaBoolEnabled:bool=True, ConstraintWeights_0=1) -> (Pose) {alphaInputType=Bool, boneToModify {boneName=dyn_hair_05_lf}, constraintSetup [{targetBone {boneName=spine_05}}], constraintWeights [1]}
   n39 = "Constraint - dyn_hair_02_cr (P)"!(ComponentPose=@n38, bAlphaBoolEnabled:bool=True, ConstraintWeights_0=1) -> (Pose) {alphaInputType=Bool, boneToModify {boneName=dyn_hair_02_cr}, constraintSetup [{targetBone {boneName=spine_05}}], constraintWeights [1]}
   n40 = "Constraint - dyn_hair_03_cr (P)"!(ComponentPose=@n39, bAlphaBoolEnabled:bool=True, ConstraintWeights_0=1) -> (Pose) {alphaInputType=Bool, boneToModify {boneName=dyn_hair_03_cr}, constraintSetup [{targetBone {boneName=spine_05}}], constraintWeights [1]}
   n41 = "Constraint - dyn_hair_04_cr (P)"!(ComponentPose=@n40, bAlphaBoolEnabled:bool=True, ConstraintWeights_0=1) -> (Pose) {alphaInputType=Bool, boneToModify {boneName=dyn_hair_04_cr}, constraintSetup [{targetBone {boneName=spine_05}}], constraintWeights [1]}
   n42 = "Constraint - dyn_hair_05_cr (P)"!(ComponentPose=@n41, bAlphaBoolEnabled:bool=True, ConstraintWeights_0=1) -> (Pose) {alphaInputType=Bool, boneToModify {boneName=dyn_hair_05_cr}, constraintSetup [{targetBone {boneName=spine_05}}], constraintWeights [1]}
   n43 = "Constraint - dyn_hair_02_rt (P)"!(ComponentPose=@n42, bAlphaBoolEnabled:bool=True, ConstraintWeights_0=1) -> (Pose) {alphaInputType=Bool, boneToModify {boneName=dyn_hair_02_rt}, constraintSetup [{targetBone {boneName=spine_05}}], constraintWeights [1]}
   n44 = "Constraint - dyn_hair_03_rt (P)"!(ComponentPose=@n43, bAlphaBoolEnabled:bool=True, ConstraintWeights_0=1) -> (Pose) {alphaInputType=Bool, boneToModify {boneName=dyn_hair_03_rt}, constraintSetup [{targetBone {boneName=spine_05}}], constraintWeights [1]}
   n45 = "Constraint - dyn_hair_04_rt (P)"!(ComponentPose=@n44, bAlphaBoolEnabled:bool=True, ConstraintWeights_0=1) -> (Pose) {alphaInputType=Bool, boneToModify {boneName=dyn_hair_04_rt}, constraintSetup [{targetBone {boneName=spine_05}}], constraintWeights [1]}
   n46 = "Constraint - dyn_hair_04_rt (P)"(ComponentPose=@n45, bAlphaBoolEnabled=True, ConstraintWeights_0=1) {alphaInputType=Bool, boneToModify {boneName=dyn_hair_04_rt}, constraintSetup [{targetBone {boneName=spine_05}}], constraintWeights [1]}
   Save cached pose 'backpackCapeCache'(Pose=Blend Poses by bool(
     BlendPose_0=Component To Local(ComponentPose=@n46)
     BlendPose_1=Use cached pose 'dynamicsCache'()
     bActiveValue=Get is Wearing cape and backpack()
    ) {blendType=Linear}) {cacheName=backpackCapeCache}
  ]}
 EventGraph {nodes [
   Event BlueprintUpdateAnimation()
   > n1 = Sequence!(execute) -> (then_0, then_1, then_2:Exec)
   > n2 = "Gravity Override State Function (Parameter Asset) (Nativized)"(
    CharacterPartAnimInstance=Self-Reference()
    GameTimeInSeconds=GetGameTimeInSeconds()
    Parameters=&/M_MED_HonestWasp_FaceAcc_Parameters
   )
   > Set GravityOverride(GravityOverride=@n2)
   > Set RBAN Settings Out(RBAN Settings Out=@n2.OutRigidBodyAnimNodeParameters)
   Set IsCrouchTarget(execute=@n1.then_1, IsCrouchTarget:bool=AND Boolean(A=Get bIsCrouching(), B=Get bIsTargeting())) -> (Output_Get:bool)
   > n3 = Branch(Condition=Get is Wearing cape and backpack())
   > Set Alpha BP(Alpha BP:double=0) -> (Output_Get:double)
   Set Alpha BP(execute=@n3.else, Alpha BP=1)
   n4 = GetAnimInstance(self=GetOwnerSkeletalMeshComponent())
   Is Valid(exec=@n1.then_2, InputObject=@n4)
   > Set IsFrontEnd IDLE(IsFrontEnd IDLE:bool=Montage_IsActive!(
     self:Anim Instance Object Reference=@n4
     Montage:Anim Montage Object Reference=/BRCosmetics/Animation/Game/MainPlayer/Menu/BR/CustomSkins/_S41/HonestWasp_M/HonestWasp_M_HonestWaspLobbyIdle_Idle_Montage
    ) -> (ReturnValue:bool)) -> (Output_Get:bool)
   Event OnCharacterCustomizationLoaded()
   > n5 = Cast To PlayerPawn_Athena(Object=TryGetPawnOwner())
   > Set Body_Mesh(Body_Mesh=GetSkeletalMeshForPartType(self=@n5))
   n6 = CheckForBackbling()
   > n7 = Set isWearingBackbling(isWearingBackbling=AND Boolean!(A=@n6, B=NOT Boolean(A=@n6.bIsBackblingACape), C:bool=NOT Boolean(A=@n6.bIsBackblingLong)) -> (ReturnValue))
   > n8 = Set Is Wearing Cape(Is Wearing Cape=OR Boolean(A=@n6.bIsBackblingACape, B=@n6.bIsBackblingLong))
   > Set is Wearing cape and backpack(is Wearing cape and backpack:bool=OR Boolean(A=@n7, B=@n8)) -> (Output_Get:bool)
   "To String (Name)"(InName=@n2.OutCurrentState)
  ]}
]
variables [
 Alpha BP {type=number}
 Body_Mesh {type=SkeletalMeshComponent}
 GravityOverride {type=Vector}
 Is Wearing Cape {type=boolean}
 is Wearing cape and backpack {type=boolean}
 IsCrouchTarget {type=boolean}
 IsFrontEnd IDLE {type=boolean}
 isWearingBackbling {type=boolean}
 RBAN Settings Out {type=FortRigidBodyAnimNodeParameters}
]
